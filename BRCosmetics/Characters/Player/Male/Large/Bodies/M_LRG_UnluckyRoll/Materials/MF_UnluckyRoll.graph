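# Material function for the UnluckyRoll character body that layers animated, tinted smoke over incoming material attributes, with a selectable tech-art mask channel.
# Smoke layers sample one noise map with panning, flow-map distortion and BumpOffset depth, then combine with Fresnel shading.
# An optional featured layer crossfades two time phases to hide looping.
# The result alters base color, roughness, specular, normal, AO and subsurface, with switchable parallax-distorted emissive light-up.
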MaterialFunction /BRCosmetics/Characters/Player/Male/Large/Bodies/M_LRG_UnluckyRoll/Materials/MF_UnluckyRoll
graph {nodes [
  n1 = GetMaterialAttributes(None=FunctionInput() {inputName=Material In, inputType=FunctionInput_MaterialAttributes}) {attributeGetTypes [
    #1
    #2
    #3
    #4
    #5
    #6
    #7
   ]}
  n2 = Multiply(A=@n1.BaseColor, B=ScalarParameter() {defaultValue=0.25, group=UnluckyRoll, parameterName=BaseColorScale})
  n3 = NamedRerouteUsage()
  n4 = NamedRerouteUsage()
  n5 = ComponentMask(None=@n4) {r=true}
  n6 = VectorParameter() {group=Smoke_Lightup, parameterName=Smoke_Emissive, defaultValue=<rgba 0.3802 0.3802 0.3802 1>}
  n7 = Multiply(
   A=StaticSwitchParameter(True=Multiply(A=@n5, B=@n6.RGB), False=@n4) {group=Smoke_Lightup, parameterName=LightUpSmoke?}
   B=NamedRerouteUsage()
  )
  n8 = ComponentMask(None=@n7) {r=true}
  n9 = Power(Base=@n8) {constExponent=0.2}
  n10 = NamedRerouteUsage()
  n11 = Multiply(A=@n8, B=Multiply(A=ComponentMask(None=@n10) {desc=TA Mask Red Channel, r=true}, B=NamedRerouteUsage()))
  n12 = NamedRerouteUsage()
  n13 = Constant()
  n14 = SetMaterialAttributes!(
   MaterialAttributes=@n1.MaterialAttributes
   BaseColor=Add(A=LinearInterpolate(A=@n1.BaseColor, B=@n2, Alpha=@n3), B=@n7) {desc=BC}
   Roughness=LinearInterpolate(A=@n1.Roughness, B=Constant() {r=1}, Alpha=@n3)
   SubsurfaceColor=Multiply(A=@n1.SubsurfaceColor, B=@n8) {desc=SSS}
   AmbientOcclusion=Add(A=@n1.AmbientOcclusion, B=Multiply(A=@n9) {constB=-1}) {desc=AO}
   EmissiveColor=StaticSwitchParameter(True=Add(A=@n1.EmissiveColor, B=@n11) {desc=Emissive}, False=@n1.EmissiveColor) {group=Smoke_Lightup, parameterName=LightUpSmoke?}
   Specular=LinearInterpolate(A=@n1.Specular, B=Constant() {r=0.15}, Alpha=@n3) {desc=Spec}
   Normal=LinearInterpolate(
    A=@n1.Normal
    B=MaterialFunctionCall(Normal=@n12, Flatness=@n13) {materialFunction=&FlattenNormal}.Result
    Alpha=@n3
   ) {desc=Normal}
  ) {attributeSetTypes [
    #1
    #2
    #3
    #4
    #5
    #6
    #7
   ]}
  FunctionOutput(None=@n14) {outputName=Material Out}
  NamedRerouteDeclaration(None=TextureCoordinate()) {name=CachedUV, nodeColor=<rgba 0.3359 0 1 1>}
  n15 = Time()
  n16 = Constant4Vector() {constant=<rgba -1 1 0.5 1>}
  n17 = MaterialFunctionCall(Input=Sine(None=Multiply(A=@n15) {constB=0.5}) {period=6.283}, Input Low=@n16.R, Input High=@n16.R, Target Low=@n16.R, Target High=@n16.R) {materialFunction=&RemapValueRange}
  n18 = StaticSwitchParameter(True=LinearInterpolate(A=NamedRerouteUsage(), B=NamedRerouteUsage(), Alpha=@n17.Result), False=NamedRerouteUsage()) {group=UnluckyRoll, parameterName=EnableFeaturedSmoke}
  n19 = VectorParameter() {group=UnluckyRoll, parameterName=smokecolor_01, defaultValue=<rgba 0.3802 0.3049 0.3614 1>}
  n20 = NamedRerouteUsage()
  n21 = VectorParameter() {group=UnluckyRoll, parameterName=smokecolor_02, defaultValue=<rgba 0.2859 0.3701 0.6615 1>}
  n22 = LinearInterpolate(A=NamedRerouteUsage(), B=NamedRerouteUsage(), Alpha=@n17.Result)
  n23 = NamedRerouteUsage()
  n24 = Add(
   A=Add(A=Multiply(A=@n18, B=@n19.RGB), B=Multiply(A=@n20, B=@n21.RGB))
   B=Multiply(
    A=StaticSwitchParameter(True=@n22, False=@n23) {group=UnluckyRoll, parameterName=EnableFeaturedSmoke}
    B=VectorParameter() {group=UnluckyRoll, parameterName=smokecolor_03, defaultValue=<rgba 0.6378 0.6376 0.7604 1>}.RGB
   )
  )
  n25 = ScalarParameter() {defaultValue=1.75, group=UnluckyRoll, parameterName=SmokeStrength}
  n26 = NamedRerouteUsage()
  n27 = Multiply(
   A=Fresnel(ExponentIn=ScalarParameter() {defaultValue=2, group=UnluckyRoll, parameterName=ExponentIn}, Normal=Transform(None=@n26)) {baseReflectFraction=0}
   B=ScalarParameter() {defaultValue=0.15, group=UnluckyRoll, parameterName=FresnelStrength}
  )
  n28 = Subtract(
   A=Add(A=Multiply(A=@n24, B=@n25), B=VectorParameter() {group=UnluckyRoll, parameterName=BaseColor, defaultValue=<rgba 0.026 0.026 0.026 1>}.RGB)
   B=@n27
  )
  n29 = ScalarParameter() {defaultValue=1, group=UnluckyRoll, parameterName=SmokeStrengthInDark}
  NamedRerouteDeclaration(None=LinearInterpolate(A=Saturate(None=@n28), B=Multiply(A=@n24, B=@n29) {constB=10}, Alpha=@n27) {constA=1, constB=0}) {name=Smoke, nodeColor=<rgba 0 1 0.9766 1>}
  n30 = NamedRerouteUsage()
  n31 = VectorParameter() {group=UnluckyRoll, parameterName=Smoke3_UV, channelNames {a=Offset Y, b=Offset X, g=Tilling Y, r=Tilling X}, defaultValue=<rgba 1.5 1.5 0 1>}
  n32 = MaterialFunctionCall(X=@n31.Tilling X, Y=@n31.Tilling X) {materialFunction=&MakeFloat2}
  n33 = Panner(
   Coordinate=Add(
    A=Multiply(A=@n30, B=@n32.Result)
    B=MaterialFunctionCall(X=@n31.Offset X, Y=@n31.Offset X) {materialFunction=&MakeFloat2}.Result
   )
   Speed=ComponentMask(None=VectorParameter() {group=UnluckyRoll, parameterName=Smoke3_Speed, defaultValue=<rgba 0 0.125 0 1>}.RGB) {g=true, r=true}
  )
  n34 = Multiply(A=ScalarParameter() {defaultValue=0.35, group=UnluckyRoll, parameterName=FlowMapStrength3}, B=NamedRerouteUsage())
  n35 = TextureSampleParameter2D(UVs=BumpOffset(Coordinate=Add(A=@n33, B=@n34), Height=ScalarParameter() {defaultValue=-3, group=UnluckyRoll, parameterName=Smoke3_Height})) {
   group=UnluckyRoll
   parameterName=SmokeMap
   samplerSource=SSM_Wrap_WorldGroupSettings
   samplerType=SAMPLERTYPE_Grayscale
   texture=&LowResBlurredNoise
  }
  NamedRerouteDeclaration(None=@n35.R) {name=Smoke3, nodeColor=<rgba 1 0 0.7266 1>}
  n36 = VectorParameter() {
   group=UnluckyRoll
   parameterName=Smoke_DistortUV2
   channelNames {a=Offset.Y, b=Offset.X, g=Tilling.Y, r=Tilling.X}
   defaultValue=<rgba 1 1 2 0>
  }
  n37 = Multiply(
   A=NamedRerouteUsage()
   B=MaterialFunctionCall(X=@n36.Tilling.X, Y=@n36.Tilling.X) {materialFunction=&MakeFloat2}.Result
  )
  n38 = MaterialFunctionCall(X=@n36.Offset.X, Y=@n36.Offset.X) {materialFunction=&MakeFloat2}
  n39 = VectorParameter() {group=UnluckyRoll, parameterName=Smoke_Flowmap_OffsetSpeed2, sortPriority=31, defaultValue=<rgba 0 0.05 0 0>}
  n40 = TextureSampleParameter2D(UVs=Panner(
    Coordinate=Add(A=@n37, B=@n38.Result)
    Speed=MaterialFunctionCall(X=@n39.R, Y=@n39.R) {materialFunction=&MakeFloat2}.Result
   )) {
   desc=Only RG channel works
   group=UnluckyRoll
   parameterName=Smoke_DistortUV_Flowmap
   samplerSource=SSM_Wrap_WorldGroupSettings
   samplerType=SAMPLERTYPE_LinearColor
   texture=/Game/Global/Textures/Default/Flowmaps/T_TorusFlowmap
  }
  n41 = ComponentMask(None=@n40.RGB) {g=true, r=true}
  NamedRerouteDeclaration(None=Subtract(A=Multiply(A=@n41) {constB=2})) {name=FlowMap_Secondary, nodeColor=<rgba 0 1 0.2969 1>}
  n42 = NamedRerouteUsage()
  n43 = NamedRerouteUsage()
  NamedRerouteDeclaration(None=MaterialFunctionCall!(XY vector=Multiply(A=@n42, B=@n43), Z sign, Default Normalization vector) {materialFunction=/Engine/Functions/Engine_MaterialFunctions02/Math/DeriveNormalZ_Function}.Result) {name=Normal, nodeColor=<rgba 0.3828 1 0 1>}
  n44 = Multiply(A=CameraVectorWS()) {constB=-1}
  n45 = VertexNormalWS()
  n46 = MaterialFunctionCall(Float3=Subtract(
    A=Transform(None=@n44) {transformSourceType=TRANSFORMSOURCE_World, transformType=TRANSFORM_Tangent}
    B=Transform(None=@n45) {transformSourceType=TRANSFORMSOURCE_World, transformType=TRANSFORM_Tangent}
   )) {materialFunction=&BreakOutFloat3Components}
  n47 = Multiply(
   A=MaterialFunctionCall(X=@n46.R, Y=@n46.R) {materialFunction=&MakeFloat2}.Result
   B=ScalarParameter() {group=Smoke_Lightup, parameterName=ParallaxDepth}
  )
  n48 = Multiply(A=@n46.B) {constB=-1}
  n49 = TextureSampleParameter2D(UVs=Multiply(A=TextureCoordinate(), B=ScalarParameter() {defaultValue=1, group=Smoke_Lightup, parameterName=Emissive_Distort_TitleSize})) {
   desc=Only RG channel works
   group=UnluckyRoll
   parameterName=Smoke_DistortUV_Flowmap
   samplerSource=SSM_Wrap_WorldGroupSettings
   samplerType=SAMPLERTYPE_LinearColor
   texture=&T_TorusFlowmap
  }
  n50 = MaterialFunctionCall(X=@n49.R, Y=@n49.R) {materialFunction=&MakeFloat2}
  n51 = ScalarParameter() {defaultValue=0.5, group=Smoke_Lightup, parameterName=Emissive_Distort_Strength}
  n52 = VectorParameter() {group=Smoke_Lightup, parameterName=Emissive_UV, defaultValue=<rgba 1 1 0 0.5>}
  n53 = Add(
   A=Multiply(
    A=Multiply(A=@n50.Result, B=@n51) {constB=0.1}
    B=MaterialFunctionCall(X=@n52.R, Y=@n52.R) {materialFunction=&MakeFloat2}.Result
   )
   B=MaterialFunctionCall(X=@n52.B, Y=@n52.B) {materialFunction=&MakeFloat2}.Result
  )
  n54 = VectorParameter() {group=Smoke_Lightup, parameterName=Emissive_Speed, defaultValue=<rgba 0 0.25 0 1>}
  n55 = TextureSampleParameter2D(UVs=Add(A=Divide(A=@n47, B=@n48), B=Panner(Coordinate=@n53, Speed=@n54.RGB))) {
   group=Smoke_Lightup
   parameterName=SmokeLightup_EmissiveMap
   samplerSource=SSM_Wrap_WorldGroupSettings
   samplerType=SAMPLERTYPE_Grayscale
   texture=&LowResBlurredNoise
  }
  n56 = Add(A=@n55.R) {constB=0.08}
  n57 = ScalarParameter() {defaultValue=1, group=Smoke_Lightup, parameterName=Emissive_Intensity}
  NamedRerouteDeclaration(None=Multiply(
    A=Multiply(A=@n56, B=@n57) {constB=2}
    B=VectorParameter() {group=Smoke_Lightup, parameterName=Smoke_Emissive, defaultValue=<rgba 0.3802 0.3802 0.3802 1>}.RGB
   )) {name=EmissiveFactor, nodeColor=<rgba 1 0.4219 0 1>}
  n58 = NamedRerouteUsage()
  n59 = VectorParameter() {
   group=UnluckyRoll
   parameterName=Smoke_DistortUV0
   channelNames {a=Offset.Y, b=Offset.X, g=Tilling.Y, r=Tilling.X}
   defaultValue=<rgba 1 1 2 0>
  }
  n60 = MaterialFunctionCall(X=@n59.Tilling.X, Y=@n59.Tilling.X) {materialFunction=&MakeFloat2}
  n61 = TextureSampleParameter2D(UVs=Add(
    A=Multiply(A=@n58, B=@n60.Result)
    B=MaterialFunctionCall(X=@n59.Offset.X, Y=@n59.Offset.X) {materialFunction=&MakeFloat2}.Result
   )) {
   desc=Only RG channel works
   group=UnluckyRoll
   parameterName=Smoke_DistortUV_Flowmap
   samplerSource=SSM_Wrap_WorldGroupSettings
   samplerType=SAMPLERTYPE_LinearColor
   texture=&T_TorusFlowmap
  }
  n62 = ComponentMask(None=@n61.RGB) {g=true, r=true}
  NamedRerouteDeclaration(None=Subtract(A=Multiply(A=@n62) {constB=2})) {name=Flowmap_Featured, nodeColor=<rgba 0 1 0.1797 1>}
  n63 = VectorParameter() {group=UnluckyRoll, parameterName=Smoke0_UV, channelNames {a=Offset Y, b=Offset X, g=Tilling Y, r=Tilling X}, defaultValue=<rgba 2 2 0 1>}
  n64 = Add(
   A=Multiply(
    A=NamedRerouteUsage()
    B=MaterialFunctionCall(X=@n63.Tilling X, Y=@n63.Tilling X) {materialFunction=&MakeFloat2}.Result
   )
   B=ComponentMask(None=MaterialFunctionCall(X=@n63.Offset X, Y=@n63.Offset X) {materialFunction=&MakeFloat2}.Result) {g=true, r=true}
  )
  n65 = Time()
  n66 = ScalarParameter() {defaultValue=0.2, group=UnluckyRoll, parameterName=Smoke_DistortTimeScale}
  n67 = Frac(None=Multiply(A=@n65, B=@n66) {constB=0.5})
  n68 = VectorParameter() {group=UnluckyRoll, parameterName=Smoke_Flowmap_OffsetSpeed0, sortPriority=31, defaultValue=<rgba 0 0.05 0 0>}
  n69 = MaterialFunctionCall(X=@n68.R, Y=@n68.R) {materialFunction=&MakeFloat2}
  n70 = Panner(Coordinate=@n64, Time=@n67, Speed=@n69.Result)
  n71 = Multiply(A=ScalarParameter() {defaultValue=0.1, group=UnluckyRoll, parameterName=Featured_FlowMap_Strength}, B=NamedRerouteUsage())
  n72 = Constant() {r=-0.5}
  n73 = TextureSampleParameter2D(UVs=BumpOffset(Coordinate=Add(A=@n70, B=@n71), Height=@n72)) {
   group=UnluckyRoll
   parameterName=SmokeMap
   samplerSource=SSM_Wrap_WorldGroupSettings
   samplerType=SAMPLERTYPE_Grayscale
   texture=&LowResBlurredNoise
  }
  n74 = Multiply(A=@n65, B=@n66) {constB=0.5}
  n75 = Panner(Coordinate=@n64, Time=Frac(None=Add(A=@n74) {constB=0.5}), Speed=@n69.Result)
  n76 = TextureSampleParameter2D(UVs=BumpOffset(Coordinate=Add(A=@n75, B=@n71), Height=@n72)) {
   group=UnluckyRoll
   parameterName=SmokeMap
   samplerSource=SSM_Wrap_WorldGroupSettings
   samplerType=SAMPLERTYPE_Grayscale
   texture=&LowResBlurredNoise
  }
  n77 = Abs(None=Subtract(A=@n67) {constB=0.5})
  NamedRerouteDeclaration(None=LinearInterpolate(A=@n73.R, B=@n76.R, Alpha=Multiply(A=@n77) {constB=2})) {name=Smoke0, nodeColor=<rgba 0.3828 1 0 1>}
  n78 = TextureSample() {samplerType=SAMPLERTYPE_Masks, texture=&T_WhiteMask}
  NamedRerouteDeclaration(None=StaticComponentMaskParameter(None=FunctionInput(Preview=@n78.RGB) {bUsePreviewValueAsDefault=true, inputName=TechArtMaskInput}) {defaultR=true, desc=This Tech Art Mask Channel Could Have a Gradient, group=UnluckyRoll, parameterName=UnluckyRollChannel}) {name=TechArtMask, nodeColor=<rgba 0.6641 1 0 1>}
  n79 = VectorParameter() {
   group=UnluckyRoll
   parameterName=Smoke_DistortUV
   channelNames {a=Offset.Y, b=Offset.X, g=Tilling.Y, r=Tilling.X}
   defaultValue=<rgba 1 1 0 0>
  }
  n80 = Add(
   A=Multiply(
    A=NamedRerouteUsage()
    B=MaterialFunctionCall(X=@n79.Tilling.X, Y=@n79.Tilling.X) {materialFunction=&MakeFloat2}.Result
   )
   B=MaterialFunctionCall(X=@n79.Offset.X, Y=@n79.Offset.X) {materialFunction=&MakeFloat2}.Result
  )
  n81 = VectorParameter() {group=UnluckyRoll, parameterName=Smoke_Flowmap_OffsetSpeed, sortPriority=31}
  n82 = MaterialFunctionCall(X=@n81.R, Y=@n81.R) {materialFunction=&MakeFloat2}
  n83 = NamedRerouteUsage()
  n84 = ScalarParameter() {group=UnluckyRoll, parameterName=FlowMap_Secondary_Scale}
  n85 = TextureSampleParameter2D(UVs=Add(A=Panner(Coordinate=@n80, Speed=@n82.Result), B=Multiply(A=@n83, B=@n84))) {
   desc=Only RG channel works
   group=UnluckyRoll
   parameterName=Smoke_DistortUV_Flowmap
   samplerSource=SSM_Wrap_WorldGroupSettings
   samplerType=SAMPLERTYPE_LinearColor
   texture=&T_TorusFlowmap
  }
  n86 = ComponentMask(None=@n85.RGB) {g=true, r=true}
  NamedRerouteDeclaration(None=Subtract(A=Multiply(A=@n86) {constB=2})) {name=FlowMap_Main, nodeColor=<rgba 0 1 0.1797 1>}
  n87 = NamedRerouteUsage()
  n88 = VectorParameter() {group=UnluckyRoll, parameterName=Smoke1_UV, channelNames {a=Offset Y, b=Offset X, g=Tilling Y, r=Tilling X}, defaultValue=<rgba 2 2 0 1>}
  n89 = MaterialFunctionCall(X=@n88.Tilling X, Y=@n88.Tilling X) {materialFunction=&MakeFloat2}
  n90 = MaterialFunctionCall(X=@n88.Offset X, Y=@n88.Offset X) {materialFunction=&MakeFloat2}
  n91 = Panner(
   Coordinate=Add(A=Multiply(A=@n87, B=@n89.Result), B=ComponentMask(None=@n90.Result) {g=true, r=true})
   Speed=ComponentMask(None=VectorParameter() {group=UnluckyRoll, parameterName=Smoke1_Speed, defaultValue=<rgba 0 0.1 0 0>}.RGB) {g=true, r=true}
  )
  n92 = Multiply(A=ScalarParameter() {defaultValue=0.1, group=UnluckyRoll, parameterName=FlowMapStrength1}, B=NamedRerouteUsage())
  n93 = TextureSampleParameter2D(UVs=BumpOffset(Coordinate=Add(A=@n91, B=@n92), Height=ScalarParameter() {group=UnluckyRoll, parameterName=Smoke1_Height})) {
   group=UnluckyRoll
   parameterName=SmokeMap
   samplerSource=SSM_Wrap_WorldGroupSettings
   samplerType=SAMPLERTYPE_Grayscale
   texture=&LowResBlurredNoise
  }
  NamedRerouteDeclaration(None=@n93.RGB) {name=Smoke1, nodeColor=<rgba 0.3828 1 0 1>}
  n94 = NamedRerouteUsage()
  n95 = VectorParameter() {
   group=UnluckyRoll
   parameterName=Smoke2_UV
   channelNames {a=Offset Y, b=Offset X, g=Tilling Y, r=Tilling X}
   defaultValue=<rgba 1.25 1.25 0 1>
  }
  n96 = MaterialFunctionCall(X=@n95.Tilling X, Y=@n95.Tilling X) {materialFunction=&MakeFloat2}
  n97 = Panner(
   Coordinate=Add(
    A=Multiply(A=@n94, B=@n96.Result)
    B=MaterialFunctionCall(X=@n95.Offset X, Y=@n95.Offset X) {materialFunction=&MakeFloat2}.Result
   )
   Speed=ComponentMask(None=VectorParameter() {group=UnluckyRoll, parameterName=Smoke2_Speed, defaultValue=<rgba 0 0.15 0 1>}.RGB) {g=true, r=true}
  )
  n98 = Multiply(A=ScalarParameter() {defaultValue=0.33, group=UnluckyRoll, parameterName=FlowMapStrength}, B=NamedRerouteUsage())
  n99 = TextureSampleParameter2D(UVs=BumpOffset(Coordinate=Add(A=@n97, B=@n98), Height=ScalarParameter() {defaultValue=-1.5, group=UnluckyRoll, parameterName=Smoke2_Height})) {
   group=UnluckyRoll
   parameterName=SmokeMap
   samplerSource=SSM_Wrap_WorldGroupSettings
   samplerType=SAMPLERTYPE_Grayscale
   texture=&LowResBlurredNoise
  }
  NamedRerouteDeclaration(None=@n99.R) {name=Smoke2, nodeColor=<rgba 1 0 0.7031 1>}
 ]}
