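# Physics asset of cape colliders for the UnluckyRoll large male player body cosmetic, giving simple collision volumes for, by its name, the cape.
# Capsules cover the pelvis, spine, clavicles and neck, with a sphere on the head, while arm, leg and foot bodies have no shapes.
# More capsules sit on the back armor plate and canister bones, including the dyn_ ones.
# It has no constraints, so it defines collision shapes only, with no jointed ragdoll.
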PhysicsAsset /BRCosmetics/Characters/Player/Male/Large/Bodies/M_LRG_UnluckyRoll/Meshes/M_LRG_UnluckyRoll_CapeColliders
bodies {
 base_armor_plate_back_l [{shapeType=Capsule, length=7.5419, radius=0.505, shapeName=base_armor_plate_back_l_capsule, rotation=<rot 0 -90 0>}]
 base_armor_plate_back_r [{shapeType=Capsule, length=0.505, radius=0.505, shapeName=base_armor_plate_back_r_capsule, rotation=<rot 0 -90 0>}]
 calf_l []
 calf_r []
 clavicle_l [{shapeType=Capsule, length=9.0192, radius=9.8871, center=<9.4266 -1.1806 -2.4977>, rotation=<rot 69.3297 1.4476 176.9462>}]
 clavicle_r [{shapeType=Capsule, length=9.0192, radius=9.8871, center=<-9.4266 1.1806 2.4977>, rotation=<rot 69.3297 1.4476 -3.0538>}]
 dyn_armor_plate_back_01 [{
   shapeType=Capsule
   length=22.4242
   radius=5.0725
   shapeName=dyn_armor_plate_back_01_capsule
   center=<7.507 1.5478 -0.0012>
   rotation=<rot 0 -90 0>
  }]
 dyn_armor_plate_back_01_l [{
   shapeType=Capsule
   length=12.9183
   radius=7.7142
   shapeName=dyn_armor_plate_back_01_l_capsule
   center=<2.3574 -2.1939 1.5952>
   rotation=<rot 14.7697 -90 -3.5138>
  }]
 dyn_armor_plate_back_01_r [{
   shapeType=Capsule
   length=12.9183
   radius=7.7142
   shapeName=dyn_armor_plate_back_01_l_capsule
   center=<2.7466 -4.3117 -1.7678>
   rotation=<rot -0.3358 0.1376 -22.2369>
  }]
 dyn_armor_plate_back_02 [{shapeType=Capsule, length=0.505, radius=0.505, shapeName=dyn_armor_plate_back_02_capsule, rotation=<rot 0 -90 0>}]
 dyn_armor_plate_back_02_l [{
   shapeType=Capsule
   length=13.7509
   radius=7.5294
   shapeName=dyn_armor_plate_back_02_l_capsule
   center=<2.302 -3.226 2.4874>
   rotation=<rot 15.4937 -90 -2.3242>
  }]
 dyn_armor_plate_back_02_r [{
   shapeType=Capsule
   length=13.7509
   radius=7.5294
   shapeName=dyn_armor_plate_back_02_l_capsule
   center=<2.7466 -4.8318 -1.9804>
   rotation=<rot -0.3358 0.1376 -22.2369>
  }]
 dyn_armor_plate_back_03 [{
   shapeType=Capsule
   length=27.5469
   radius=4.0383
   shapeName=dyn_armor_plate_back_03_capsule
   center=<3.2536 0.246 -0.0012>
   rotation=<rot 0 -90 0>
  }]
 dyn_armor_plate_back_03_l [{
   shapeType=Capsule
   length=14.659
   radius=7.6443
   shapeName=dyn_armor_plate_back_03_l_capsule
   center=<1.7223 -4.0438 3.3836>
   rotation=<rot 15.3886 -90 -2.3767>
  }]
 dyn_armor_plate_back_03_r [{
   shapeType=Capsule
   length=14.659
   radius=7.6443
   shapeName=dyn_armor_plate_back_03_l_capsule
   center=<2.7466 -5.8544 -2.3985>
   rotation=<rot -0.3358 0.1376 -22.2369>
  }]
 dyn_armor_plate_back_04_l [{
   shapeType=Capsule
   length=16.7535
   radius=7.705
   shapeName=dyn_armor_plate_back_04_l_capsule
   center=<1.6256 -4.8182 4.4468>
   rotation=<rot 16.1496 -90 -3.173>
  }]
 dyn_armor_plate_back_04_r [{
   shapeType=Capsule
   length=16.7535
   radius=7.705
   shapeName=dyn_armor_plate_back_04_l_capsule
   center=<2.7466 -6.2101 -2.5439>
   rotation=<rot -0.3358 0.1376 -22.2369>
  }]
 dyn_canister_l [{
   shapeType=Capsule
   length=6.398
   radius=6.2171
   shapeName=dyn_canister_l_capsule
   center=<5.9103 1.3405 2.2082>
   rotation=<rot 19.8537 -90 -12.4394>
  }]
 dyn_canister_r [{
   shapeType=Capsule
   length=6.398
   radius=6.2171
   shapeName=dyn_canister_l_capsule
   center=<-5.9103 -1.3405 -2.2082>
   rotation=<rot 19.8537 -90 167.5606>
  }]
 foot_l []
 foot_r []
 head [{shapeType=Sphere, radius=17.3125, center=<11.028 -0.174 0>}]
 lowerarm_l []
 lowerarm_r []
 neck_02 [{shapeType=Capsule, length=13.4375, radius=7.3175, center=<-0.2642 -0.3408 0>, rotation=<rot 90 -171.1193 -175.3978>}]
 pelvis [{shapeType=Capsule, length=8.6456, radius=14.5388, center=<-2.4887 -0.9654 0>}]
 spine_01 [{shapeType=Capsule, length=7.0841, radius=14.4182, center=<-1.7965 -0.8636 0>}]
 spine_02 [
  {shapeType=Capsule, length=7.3264, radius=13.6322, center=<-1.711 -0.1851 0.0001>}
  {shapeType=Capsule, length=25.9589, radius=15.0173, center=<2.5024 -0.9735 3.9825>, rotation=<rot 90 0.0099 0.0099>}
  {shapeType=Capsule, length=26.4383, radius=14.538, center=<2.4704 -0.9736 -3.9825>, rotation=<rot 90 -0.0099 -0.0099>}
 ]
 spine_03 [{shapeType=Capsule, length=7.1249, radius=15.7589, center=<0.5825 -7.152 0.0001>, rotation=<rot 0 10.2925 0>}]
 spine_04 [{shapeType=Capsule, length=9.3816, radius=18.0949, center=<2.8724 -3.4383 0.0002>, rotation=<rot 0 7.5929 0>}]
 spine_05 [
  {shapeType=Capsule, length=8.0226, radius=15.695, center=<-2.2795 -0.6089 0>}
  {shapeType=Capsule, length=14.8869, radius=16.0569, center=<-0.2533 -4.2066 0>, rotation=<rot 0 2.0732 0>}
 ]
 thigh_l []
 thigh_r []
 upperarm_l []
 upperarm_r []
}
constraints []
