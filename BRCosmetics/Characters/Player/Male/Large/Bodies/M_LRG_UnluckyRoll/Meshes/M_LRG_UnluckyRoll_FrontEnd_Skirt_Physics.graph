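# Physics asset for the front-end skirt of the large male UnluckyRoll body, simulating secondary motion of armor plates, pant sway and ropes.
# Pelvis, leg, foot and katana bodies, mostly capsules, form the body volumes alongside base anchor bones.
# Dynamic chains hang from those anchors: a branching back plate with mostly single-axis swing, limited side plates, short and long ropes.
# Pant sway bones swing from anchors locked to the calves.
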PhysicsAsset /BRCosmetics/Characters/Player/Male/Large/Bodies/M_LRG_UnluckyRoll/Meshes/M_LRG_UnluckyRoll_FrontEnd_Skirt_Physics
bodies {
 base_armor_plate_area_l [{shapeType=Capsule, length=9.167, radius=4.8841, shapeName=base_armor_plate_area_l_capsule, rotation=<rot 0 0 90>}]
 base_armor_plate_area_r [{shapeType=Capsule, length=9.167, radius=4.8841, shapeName=base_armor_plate_area_r_capsule, rotation=<rot 0 0 90>}]
 base_armor_plate_back [{shapeType=Capsule, length=0.505, radius=0.505, shapeName=base_armor_plate_back_capsule, rotation=<rot 0 -90 0>}]
 base_armor_plate_large_l [{shapeType=Capsule, length=0.505, radius=0.505, shapeName=base_armor_plate_large_l_capsule, rotation=<rot 0 0 90>}]
 base_armor_plate_large_r [{shapeType=Capsule, length=0.505, radius=0.505, shapeName=base_armor_plate_large_r_capsule, rotation=<rot 0 0 90>}]
 base_katana [{
   shapeType=Capsule
   length=48.5669
   radius=11.121
   shapeName=base_katana_capsule
   center=<39.1676 -19.0954 -13.8892>
   rotation=<rot 77.57 -4.4117 9.381>
  }]
 base_pant_sway_back_l [{shapeType=Capsule, length=0.505, radius=0.505, shapeName=base_pant_sway_back_l_capsule, rotation=<rot 0 -90 0>}]
 base_pant_sway_back_r [{shapeType=Capsule, length=0.505, radius=0.505, shapeName=base_pant_sway_back_r_capsule, rotation=<rot 0 -90 0>}]
 base_pant_sway_front_l [{shapeType=Capsule, length=0.505, radius=0.505, shapeName=base_pant_sway_front_l_capsule, rotation=<rot 0 -90 0>}]
 base_pant_sway_front_r [{shapeType=Capsule, length=0.505, radius=0.505, shapeName=base_pant_sway_front_r_capsule, rotation=<rot 0 -90 0>}]
 base_pant_sway_inside_l [{shapeType=Capsule, length=0.505, radius=0.505, shapeName=base_pant_sway_inside_l_capsule, rotation=<rot 0 -90 0>}]
 base_pant_sway_inside_r [{shapeType=Capsule, length=0.505, radius=0.505, shapeName=base_pant_sway_inside_r_capsule, rotation=<rot 0 -90 0>}]
 base_pant_sway_out_l [{shapeType=Capsule, length=0.505, radius=0.505, shapeName=base_pant_sway_out_l_capsule, rotation=<rot 0 -90 0>}]
 base_pant_sway_out_r [{shapeType=Capsule, length=0.505, radius=0.505, shapeName=base_pant_sway_out_r_capsule, rotation=<rot 0 -90 0>}]
 base_pelvis_armor_back [{
   shapeType=Capsule
   length=9.7565
   radius=4.5309
   shapeName=base_pelvis_armor_back_capsule
   center=<2.48 -0.0012 1.8627>
   rotation=<rot 0 0 -90>
  }]
 base_rope_e_l [{shapeType=Capsule, length=0.505, radius=0.505, shapeName=base_rope_e_l_capsule, rotation=<rot 0 -90 0>}]
 base_rope_e_r [{shapeType=Capsule, length=0.505, radius=0.505, shapeName=base_rope_e_r_capsule, rotation=<rot 0 -90 0>}]
 calf_l [{
   shapeType=Capsule
   length=23.6493
   radius=16.6437
   shapeName=calf_l_capsule
   center=<-15.4915 -0.5059 0>
   rotation=<rot 90 176.4237 -167.2425>
  }]
 calf_r [{shapeType=Capsule, length=23.6493, radius=16.6437, shapeName=calf_l_capsule, center=<15.4915 0.5059 0>, rotation=<rot 83.1176 -106.3339 90>}]
 dyn_armor_plate_back_01 [{shapeType=Box, extentX=7.5, extentY=7.5, extentZ=31.5, shapeName=dyn_armor_plate_back_01_box, center=<0 -7.8 0>}]
 dyn_armor_plate_back_01_l [{shapeType=Capsule, length=12, radius=4, shapeName=dyn_armor_plate_back_01_l_capsule, center=<3.5657 -2.9825 0.9>}]
 dyn_armor_plate_back_01_r [{shapeType=Capsule, length=12, radius=4, shapeName=dyn_armor_plate_back_01_l_capsule, center=<3.5657 -2.9825 -0.9>}]
 dyn_armor_plate_back_02 [{shapeType=Capsule, length=23, radius=6.5, shapeName=dyn_armor_plate_back_02_capsule, center=<6.6658 -5.8682 0>}]
 dyn_armor_plate_back_02_l [{shapeType=Capsule, length=14, radius=4, shapeName=dyn_armor_plate_back_02_l_capsule, center=<3.5657 -3.3705 1.56>}]
 dyn_armor_plate_back_02_r [{shapeType=Capsule, length=14, radius=4, shapeName=dyn_armor_plate_back_02_l_capsule, center=<3.5657 -3.3705 -1.56>}]
 dyn_armor_plate_back_03_l [{shapeType=Capsule, length=16, radius=4, shapeName=dyn_armor_plate_back_03_l_capsule, center=<3.7277 -3.4425 3>, rotation=<rot 0 6.8825 0>}]
 dyn_armor_plate_back_03_r [{shapeType=Capsule, length=16, radius=4, shapeName=dyn_armor_plate_back_03_l_capsule, center=<3.7277 -3.4425 -3>, rotation=<rot 0 6.8825 0>}]
 dyn_armor_plate_back_04_l [{shapeType=Capsule, length=20, radius=4.5, shapeName=dyn_armor_plate_back_04_l_capsule, center=<6.1112 -5.0435 5>}]
 dyn_armor_plate_back_04_r [{shapeType=Capsule, length=20, radius=4.5, shapeName=dyn_armor_plate_back_04_l_capsule, center=<6.1112 -5.0435 -4.9>}]
 dyn_armor_plate_large_01_l [{shapeType=Capsule, length=12.2937, radius=7.6511, shapeName=dyn_armor_plate_large_01_l_capsule, rotation=<rot 0 0 90>}]
 dyn_armor_plate_large_01_r [{shapeType=Capsule, length=12.3567, radius=7.8398, shapeName=dyn_armor_plate_large_01_r_capsule, rotation=<rot 0 0 90>}]
 dyn_armor_plate_large_02_l [
  {
   shapeType=Capsule
   length=18.9425
   radius=7.0021
   shapeName=dyn_armor_plate_large_02_l_capsule
   center=<16.7352 10.5934 -0.927>
   rotation=<rot 85.4105 -0.0003 -0.0005>
  }
  {
   shapeType=Capsule
   length=20.5023
   radius=5.4422
   shapeName=dyn_armor_plate_large_02_l_capsule
   center=<12.1096 0 4.5603>
   rotation=<rot 85.5546 -2.9162 -0.0004>
  }
  {
   shapeType=Capsule
   length=20.5023
   radius=5.8063
   shapeName=dyn_armor_plate_large_02_l_capsule
   center=<12.1729 -10.5406 -0.195>
   rotation=<rot 83.1061 -6.6537 -0.0003>
  }
 ]
 dyn_armor_plate_large_02_r [
  {
   shapeType=Capsule
   length=18.9425
   radius=7.0021
   shapeName=dyn_armor_plate_large_02_l_capsule
   center=<-15.3797 -10.0988 3.4126>
   rotation=<rot 84.7711 -0.0005 179.9993>
  }
  {
   shapeType=Capsule
   length=20.5023
   radius=5.4422
   shapeName=dyn_armor_plate_large_02_l_capsule
   center=<-12.0085 1.3765 -3.3992>
   rotation=<rot 90 180 4.2011>
  }
  {
   shapeType=Capsule
   length=20.5023
   radius=5.8063
   shapeName=dyn_armor_plate_large_02_l_capsule
   center=<-11.735 12.6558 1.7658>
   rotation=<rot 89.366 -6.6476 179.9984>
  }
 ]
 dyn_armor_plate_large_03_l [{shapeType=Capsule, length=14.874, radius=8.0813, shapeName=dyn_armor_plate_large_03_l_capsule, center=<10.1667 0 0>, rotation=<rot 0 0 90>}]
 dyn_armor_plate_large_03_r [{shapeType=Capsule, length=14.874, radius=8.0813, shapeName=dyn_armor_plate_large_03_l_capsule, center=<-10.1667 0 0>, rotation=<rot 0 0 -90>}]
 dyn_pant_sway_back_l [{shapeType=Capsule, length=18.639, radius=3.7366, shapeName=dyn_pant_sway_back_r_capsule, center=<0 0 -8.7232>, rotation=<rot 0 0 -180>}]
 dyn_pant_sway_back_r [{shapeType=Capsule, length=18.639, radius=3.7366, shapeName=dyn_pant_sway_back_r_capsule, center=<0 0 8.7232>}]
 dyn_pant_sway_front_l [{shapeType=Capsule, length=18.639, radius=3.7366, shapeName=dyn_pant_sway_front_r_capsule, center=<0 0 -8.7232>, rotation=<rot 0 0 -180>}]
 dyn_pant_sway_front_r [{shapeType=Capsule, length=18.639, radius=3.7366, shapeName=dyn_pant_sway_front_r_capsule, center=<0 0 8.7232>}]
 dyn_pant_sway_inside_l [{shapeType=Capsule, length=18.639, radius=3.7366, shapeName=dyn_pant_sway_inside_r_capsule, center=<0 0 -8.7232>, rotation=<rot 0 0 -180>}]
 dyn_pant_sway_inside_r [{shapeType=Capsule, length=18.639, radius=3.7366, shapeName=dyn_pant_sway_inside_r_capsule, center=<0 0 8.7232>}]
 dyn_pant_sway_out_l [{shapeType=Capsule, length=18.639, radius=3.7366, shapeName=dyn_pant_sway_out_r_capsule, center=<0 0 -8.7232>, rotation=<rot 0 0 -180>}]
 dyn_pant_sway_out_r [{shapeType=Capsule, length=18.639, radius=3.7366, shapeName=dyn_pant_sway_out_r_capsule, center=<0 0 8.7232>}]
 dyn_ropeLong_e_01_l [{
   shapeType=Capsule
   length=0.505
   radius=0.505
   shapeName=dyn_ropeLong_e_01_l_capsule
   center=<0.7684 -0.1082 -0.1465>
   rotation=<rot 0 -95.7029 -72.4524>
  }]
 dyn_ropeLong_e_01_r [{
   shapeType=Capsule
   length=0.505
   radius=0.505
   shapeName=dyn_ropeLong_e_01_r_capsule
   center=<-0.7684 0.106 0.1455>
   rotation=<rot 0 -95.7029 -72.4524>
  }]
 dyn_ropeLong_e_02_l [{
   shapeType=Capsule
   length=0.505
   radius=0.505
   shapeName=dyn_ropeLong_e_02_l_capsule
   center=<1.2119 -0.1144 0.0289>
   rotation=<rot 0 93.6067 -86.8383>
  }]
 dyn_ropeLong_e_02_r [{
   shapeType=Capsule
   length=0.505
   radius=0.505
   shapeName=dyn_ropeLong_e_02_r_capsule
   center=<-1.2118 0.1122 -0.0298>
   rotation=<rot 0 93.6067 -86.8383>
  }]
 dyn_ropeLong_e_03_l [{
   shapeType=Capsule
   length=1.3788
   radius=0.6903
   shapeName=dyn_ropeLong_e_03_l_capsule
   center=<1.2637 0.004 0.1745>
   rotation=<rot 0 86.5433 -86.327>
  }]
 dyn_ropeLong_e_03_r [{
   shapeType=Capsule
   length=1.3788
   radius=0.6903
   shapeName=dyn_ropeLong_e_03_r_capsule
   center=<-1.2635 -0.0062 -0.1754>
   rotation=<rot 0 86.5433 -86.327>
  }]
 dyn_ropeShort_e_01_l [{
   shapeType=Capsule
   length=0.505
   radius=0.505
   shapeName=dyn_ropeShort_e_01_l_capsule
   center=<0.5575 0.0453 -0.137>
   rotation=<rot -37.793 -90 23.0821>
  }]
 dyn_ropeShort_e_01_r [{
   shapeType=Capsule
   length=0.505
   radius=0.505
   shapeName=dyn_ropeShort_e_01_r_capsule
   center=<-0.5575 -0.0476 0.1362>
   rotation=<rot -37.793 -90 23.0821>
  }]
 dyn_ropeShort_e_02_l [{
   shapeType=Capsule
   length=0.505
   radius=0.505
   shapeName=dyn_ropeShort_e_02_l_capsule
   center=<0.3654 -0.1175 0.0136>
   rotation=<rot 0 83.9367 -86.0705>
  }]
 dyn_ropeShort_e_02_r [{
   shapeType=Capsule
   length=0.505
   radius=0.505
   shapeName=dyn_ropeShort_e_02_r_capsule
   center=<-0.3651 0.1152 -0.0143>
   rotation=<rot 0 83.9367 -86.0705>
  }]
 dyn_ropeShort_e_03_l [{
   shapeType=Capsule
   length=1.3025
   radius=0.5949
   shapeName=dyn_ropeShort_e_03_l_capsule
   center=<0.9113 -0.2054 0.0545>
   rotation=<rot 0 -89.9611 -85.3007>
  }]
 dyn_ropeShort_e_03_r [{
   shapeType=Capsule
   length=1.3025
   radius=0.5949
   shapeName=dyn_ropeShort_e_03_r_capsule
   center=<-0.9109 0.2031 -0.0552>
   rotation=<rot 0 -89.9611 -85.3007>
  }]
 foot_l [{shapeType=Capsule, length=15, radius=12.6946, shapeName=foot_l_capsule, center=<-4.2129 10.8158 0>, rotation=<rot 0 0 90>}]
 foot_r [{shapeType=Capsule, length=15, radius=12.6946, shapeName=foot_l_capsule, center=<4.2129 -10.8158 0>, rotation=<rot 6.8824 0 -90>}]
 pelvis [{shapeType=Capsule, length=19.2882, radius=13, shapeName=pelvis_capsule, center=<-0.3169 -1.0513 0>, rotation=<rot 90 0 0>}]
 thigh_l [{
   shapeType=Capsule
   length=20.4523
   radius=14.9574
   shapeName=thigh_l_capsule
   center=<-22.0161 -2.8541 7.1849>
   rotation=<rot 82.6781 179.9836 176.6905>
  }]
 thigh_r [{
   shapeType=Capsule
   length=20.4523
   radius=14.9574
   shapeName=thigh_l_capsule
   center=<22.0485 2.2864 -7.1807>
   rotation=<rot 82.6781 179.9836 -3.3095>
  }]
}
constraints [
 {
  bone1Name=dyn_armor_plate_back_01
  bone2Name=base_armor_plate_back
  swing1LimitDegrees=2.5
  swing2LimitDegrees=2
  swing2Motion=Locked
  twistLimitDegrees=45
  twistMotion=Locked
 }
 {
  bone1Name=dyn_armor_plate_back_02
  bone2Name=dyn_armor_plate_back_01
  swing1LimitDegrees=20
  swing2LimitDegrees=2
  swing2Motion=Locked
  twistLimitDegrees=45
  twistMotion=Locked
 }
 {
  bone1Name=dyn_armor_plate_back_01_l
  bone2Name=dyn_armor_plate_back_02
  swing1LimitDegrees=30
  swing2LimitDegrees=3
  swing2Motion=Locked
  twistLimitDegrees=45
  twistMotion=Locked
 }
 {
  bone1Name=dyn_armor_plate_back_02_l
  bone2Name=dyn_armor_plate_back_01_l
  swing1LimitDegrees=30
  swing2LimitDegrees=3
  swing2Motion=Locked
  twistLimitDegrees=10
  twistMotion=Locked
 }
 {
  bone1Name=dyn_armor_plate_back_03_l
  bone2Name=dyn_armor_plate_back_02_l
  swing1LimitDegrees=30
  swing2LimitDegrees=3
  swing2Motion=Locked
  twistLimitDegrees=45
  twistMotion=Locked
 }
 {
  bone1Name=dyn_armor_plate_back_04_l
  bone2Name=dyn_armor_plate_back_03_l
  swing1LimitDegrees=30
  swing2LimitDegrees=3
  swing2Motion=Locked
  twistLimitDegrees=10
  twistMotion=Locked
 }
 {
  bone1Name=foot_l
  bone2Name=calf_l
  swing1LimitDegrees=45
  swing1Motion=Free
  swing2LimitDegrees=45
  swing2Motion=Free
  twistLimitDegrees=45
  twistMotion=Free
 }
 {
  bone1Name=dyn_armor_plate_back_01_r
  bone2Name=dyn_armor_plate_back_02
  swing1LimitDegrees=20
  swing2LimitDegrees=3
  swing2Motion=Locked
  twistLimitDegrees=45
  twistMotion=Locked
 }
 {
  bone1Name=dyn_armor_plate_back_02_r
  bone2Name=dyn_armor_plate_back_01_r
  swing1LimitDegrees=20
  swing2LimitDegrees=3
  swing2Motion=Locked
  twistLimitDegrees=45
  twistMotion=Locked
 }
 {
  bone1Name=dyn_armor_plate_back_03_r
  bone2Name=dyn_armor_plate_back_02_r
  swing1LimitDegrees=20
  swing2LimitDegrees=3
  swing2Motion=Locked
  twistLimitDegrees=45
  twistMotion=Locked
 }
 {
  bone1Name=dyn_armor_plate_back_04_r
  bone2Name=dyn_armor_plate_back_03_r
  swing1LimitDegrees=30
  swing2LimitDegrees=3
  swing2Motion=Locked
  twistLimitDegrees=45
  twistMotion=Locked
 }
 {
  bone1Name=calf_l
  bone2Name=thigh_l
  swing1LimitDegrees=45
  swing1Motion=Free
  swing2LimitDegrees=45
  swing2Motion=Free
  twistLimitDegrees=45
  twistMotion=Free
 }
 {
  bone1Name=dyn_armor_plate_large_01_l
  bone2Name=base_armor_plate_large_l
  swing1LimitDegrees=45
  swing1Motion=Free
  swing2LimitDegrees=45
  swing2Motion=Free
  twistLimitDegrees=45
  twistMotion=Free
 }
 {
  bone1Name=base_armor_plate_area_l
  bone2Name=pelvis
  swing1LimitDegrees=45
  swing1Motion=Free
  swing2LimitDegrees=45
  swing2Motion=Free
  twistLimitDegrees=45
  twistMotion=Free
 }
 {
  bone1Name=base_armor_plate_large_l
  bone2Name=base_armor_plate_area_l
  swing1LimitDegrees=45
  swing1Motion=Free
  swing2LimitDegrees=45
  swing2Motion=Free
  twistLimitDegrees=45
  twistMotion=Free
 }
 {
  bone1Name=base_armor_plate_area_r
  bone2Name=pelvis
  swing1LimitDegrees=45
  swing1Motion=Free
  swing2LimitDegrees=45
  swing2Motion=Free
  twistLimitDegrees=45
  twistMotion=Free
 }
 {
  bone1Name=base_armor_plate_large_r
  bone2Name=base_armor_plate_area_r
  swing1LimitDegrees=45
  swing1Motion=Free
  swing2LimitDegrees=45
  swing2Motion=Free
  twistLimitDegrees=45
  twistMotion=Free
 }
 {
  bone1Name=dyn_armor_plate_large_01_r
  bone2Name=base_armor_plate_large_r
  swing1LimitDegrees=45
  swing1Motion=Free
  swing2LimitDegrees=45
  swing2Motion=Free
  twistLimitDegrees=45
  twistMotion=Free
 }
 {
  bone1Name=dyn_armor_plate_large_03_l
  bone2Name=dyn_armor_plate_large_02_l
  swing1LimitDegrees=20
  swing2LimitDegrees=10
  twistLimitDegrees=5
 }
 {
  bone1Name=base_pelvis_armor_back
  bone2Name=pelvis
  swing1LimitDegrees=45
  swing1Motion=Free
  swing2LimitDegrees=45
  swing2Motion=Free
  twistLimitDegrees=45
  twistMotion=Free
 }
 {
  bone1Name=base_armor_plate_back
  bone2Name=base_pelvis_armor_back
  swing1LimitDegrees=45
  swing1Motion=Free
  swing2LimitDegrees=45
  swing2Motion=Free
  twistLimitDegrees=45
  twistMotion=Free
 }
 {
  bone1Name=foot_r
  bone2Name=calf_r
  swing1LimitDegrees=45
  swing1Motion=Free
  swing2LimitDegrees=45
  swing2Motion=Free
  twistLimitDegrees=45
  twistMotion=Free
 }
 {bone1Name=dyn_armor_plate_large_03_r, bone2Name=dyn_armor_plate_large_02_r, swing1LimitDegrees=20, swing2LimitDegrees=5, twistLimitDegrees=0}
 {
  bone1Name=thigh_l
  bone2Name=pelvis
  swing1LimitDegrees=20
  swing2LimitDegrees=3
  swing2Motion=Locked
  twistLimitDegrees=45
  twistMotion=Locked
 }
 {
  bone1Name=dyn_armor_plate_large_02_l
  bone2Name=dyn_armor_plate_large_01_l
  swing1LimitDegrees=30
  swing2LimitDegrees=10
  twistLimitDegrees=15
 }
 {bone1Name=thigh_r, bone2Name=pelvis, swing1LimitDegrees=30, swing2LimitDegrees=10, twistLimitDegrees=15}
 {
  bone1Name=calf_r
  bone2Name=thigh_r
  swing1LimitDegrees=45
  swing1Motion=Free
  swing2LimitDegrees=45
  swing2Motion=Free
  twistLimitDegrees=45
  twistMotion=Free
 }
 {
  bone1Name=dyn_armor_plate_large_02_r
  bone2Name=dyn_armor_plate_large_01_r
  swing1LimitDegrees=30
  swing2LimitDegrees=10
  twistLimitDegrees=15
 }
 {
  bone1Name=base_katana
  bone2Name=pelvis
  swing1LimitDegrees=45
  swing1Motion=Free
  swing2LimitDegrees=45
  swing2Motion=Free
  twistLimitDegrees=45
  twistMotion=Free
 }
 {bone1Name=dyn_ropeShort_e_03_r, bone2Name=dyn_ropeShort_e_02_r, swing1LimitDegrees=20, swing2LimitDegrees=20, twistLimitDegrees=0}
 {bone1Name=dyn_ropeShort_e_02_r, bone2Name=dyn_ropeShort_e_01_r, swing1LimitDegrees=20, swing2LimitDegrees=20, twistLimitDegrees=0}
 {bone1Name=dyn_ropeShort_e_01_r, bone2Name=base_rope_e_r, swing1LimitDegrees=20, swing2LimitDegrees=20, twistLimitDegrees=0}
 {bone1Name=dyn_ropeLong_e_03_r, bone2Name=dyn_ropeLong_e_02_r, swing1LimitDegrees=20, swing2LimitDegrees=20, twistLimitDegrees=0}
 {bone1Name=dyn_ropeLong_e_02_r, bone2Name=dyn_ropeLong_e_01_r, swing1LimitDegrees=20, swing2LimitDegrees=20, twistLimitDegrees=0}
 {bone1Name=dyn_ropeLong_e_01_r, bone2Name=base_rope_e_r, swing1LimitDegrees=20, swing2LimitDegrees=20, twistLimitDegrees=0}
 {bone1Name=dyn_ropeShort_e_03_l, bone2Name=dyn_ropeShort_e_02_l, swing1LimitDegrees=20, swing2LimitDegrees=20, twistLimitDegrees=0}
 {bone1Name=dyn_ropeShort_e_02_l, bone2Name=dyn_ropeShort_e_01_l, swing1LimitDegrees=20, swing2LimitDegrees=20, twistLimitDegrees=0}
 {bone1Name=dyn_ropeShort_e_01_l, bone2Name=base_rope_e_l, swing1LimitDegrees=20, swing2LimitDegrees=20, twistLimitDegrees=0}
 {bone1Name=dyn_ropeLong_e_03_l, bone2Name=dyn_ropeLong_e_02_l, swing1LimitDegrees=20, swing2LimitDegrees=20, twistLimitDegrees=0}
 {bone1Name=dyn_ropeLong_e_01_l, bone2Name=base_rope_e_l, swing1LimitDegrees=20, swing2LimitDegrees=20, twistLimitDegrees=0}
 {bone1Name=dyn_ropeLong_e_02_l, bone2Name=dyn_ropeLong_e_01_l, swing1LimitDegrees=20, swing2LimitDegrees=20, twistLimitDegrees=0}
 {
  bone1Name=base_rope_e_l
  bone2Name=base_pelvis_armor_back
  swing1LimitDegrees=45
  swing1Motion=Locked
  swing2LimitDegrees=45
  swing2Motion=Locked
  twistLimitDegrees=45
  twistMotion=Locked
 }
 {
  bone1Name=base_rope_e_r
  bone2Name=base_pelvis_armor_back
  swing1LimitDegrees=45
  swing1Motion=Locked
  swing2LimitDegrees=45
  swing2Motion=Locked
  twistLimitDegrees=45
  twistMotion=Locked
 }
 {
  bone1Name=base_pant_sway_back_l
  bone2Name=calf_l
  swing1LimitDegrees=45
  swing1Motion=Locked
  swing2LimitDegrees=45
  swing2Motion=Locked
  twistLimitDegrees=45
  twistMotion=Locked
 }
 {
  bone1Name=base_pant_sway_front_l
  bone2Name=calf_l
  swing1LimitDegrees=45
  swing1Motion=Locked
  swing2LimitDegrees=45
  swing2Motion=Locked
  twistLimitDegrees=45
  twistMotion=Locked
 }
 {
  bone1Name=base_pant_sway_inside_l
  bone2Name=calf_l
  swing1LimitDegrees=45
  swing1Motion=Locked
  swing2LimitDegrees=45
  swing2Motion=Locked
  twistLimitDegrees=45
  twistMotion=Locked
 }
 {
  bone1Name=base_pant_sway_out_l
  bone2Name=calf_l
  swing1LimitDegrees=45
  swing1Motion=Locked
  swing2LimitDegrees=45
  swing2Motion=Locked
  twistLimitDegrees=45
  twistMotion=Locked
 }
 {
  bone1Name=base_pant_sway_back_r
  bone2Name=calf_r
  swing1LimitDegrees=45
  swing1Motion=Locked
  swing2LimitDegrees=45
  swing2Motion=Locked
  twistLimitDegrees=45
  twistMotion=Locked
 }
 {bone1Name=dyn_pant_sway_back_r, bone2Name=base_pant_sway_back_r, swing1LimitDegrees=20, swing2LimitDegrees=20, twistLimitDegrees=0}
 {
  bone1Name=base_pant_sway_front_r
  bone2Name=calf_r
  swing1LimitDegrees=45
  swing1Motion=Locked
  swing2LimitDegrees=45
  swing2Motion=Locked
  twistLimitDegrees=45
  twistMotion=Locked
 }
 {bone1Name=dyn_pant_sway_front_r, bone2Name=base_pant_sway_front_r, swing1LimitDegrees=20, swing2LimitDegrees=20, twistLimitDegrees=0}
 {
  bone1Name=base_pant_sway_inside_r
  bone2Name=calf_r
  swing1LimitDegrees=45
  swing1Motion=Locked
  swing2LimitDegrees=45
  swing2Motion=Locked
  twistLimitDegrees=45
  twistMotion=Locked
 }
 {bone1Name=dyn_pant_sway_inside_r, bone2Name=base_pant_sway_inside_r, swing1LimitDegrees=20, swing2LimitDegrees=20, twistLimitDegrees=0}
 {
  bone1Name=base_pant_sway_out_r
  bone2Name=calf_r
  swing1LimitDegrees=45
  swing1Motion=Locked
  swing2LimitDegrees=45
  swing2Motion=Locked
  twistLimitDegrees=45
  twistMotion=Locked
 }
 {bone1Name=dyn_pant_sway_out_r, bone2Name=base_pant_sway_out_r, swing1LimitDegrees=20, swing2LimitDegrees=20, twistLimitDegrees=0}
 {bone1Name=dyn_pant_sway_back_l, bone2Name=base_pant_sway_back_l, swing1LimitDegrees=20, swing2LimitDegrees=20, twistLimitDegrees=0}
 {bone1Name=dyn_pant_sway_front_l, bone2Name=base_pant_sway_front_l, swing1LimitDegrees=20, swing2LimitDegrees=20, twistLimitDegrees=0}
 {bone1Name=dyn_pant_sway_out_l, bone2Name=base_pant_sway_out_l, swing1LimitDegrees=20, swing2LimitDegrees=20, twistLimitDegrees=0}
 {bone1Name=dyn_pant_sway_inside_l, bone2Name=base_pant_sway_inside_l, swing1LimitDegrees=20, swing2LimitDegrees=20, twistLimitDegrees=0}
]
