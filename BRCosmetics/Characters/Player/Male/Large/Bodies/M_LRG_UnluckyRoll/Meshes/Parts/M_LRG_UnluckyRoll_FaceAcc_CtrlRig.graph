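# Control Rig for the UnluckyRoll large male's face accessory part, correcting hair and hat bones against head, arm and gameplay pose.
# Head and upper-arm rotation deltas are remapped into look and aim variables.
# These combine with state flags (crouch, targeting, emote) and spherical pose readers to weight additive bone transforms.
# Back hair strands use parent-constrained control nulls feeding spline IK and twist bones across neck and spine.
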ControlRigBlueprint /BRCosmetics/Characters/Player/Male/Large/Bodies/M_LRG_UnluckyRoll/Meshes/Parts/M_LRG_UnluckyRoll_FaceAcc_CtrlRig
graphs [
 Multiply_1_1_1_ContainedGraph {nodes [
   n1 = RigVMFunctionEntryNode()
   n2 = &RigVMFunction_MathDoubleMul(A=@n1.A, B=@n1.B)
   RigVMFunctionReturnNode(Result=&RigVMFunction_MathDoubleMul(B=@n1.F, A=&RigVMFunction_MathDoubleMul(B=@n1.E, A=@n2.Result).Result).Result)
  ]}
 Multiply_1_1_ContainedGraph {nodes [
   n1 = RigVMFunctionEntryNode()
   n2 = &RigVMFunction_MathDoubleMul(B=@n1.C, A=&RigVMFunction_MathDoubleMul(A=@n1.A, B=@n1.B).Result)
   RigVMFunctionReturnNode(Result=&RigVMFunction_MathDoubleMul(B=@n1.E, A=&RigVMFunction_MathDoubleMul(B=@n1.D, A=@n2.Result).Result).Result)
  ]}
 Multiply_1_2_ContainedGraph {nodes [
   n1 = RigVMFunctionEntryNode()
   RigVMFunctionReturnNode(Result=&RigVMFunction_MathDoubleMul(B=@n1.C, A=&RigVMFunction_MathDoubleMul(A=@n1.A, B=@n1.B).Result).Result)
  ]}
 Multiply_1_ContainedGraph {nodes [
   n1 = RigVMFunctionEntryNode()
   Multiply = &RigVMFunction_MathDoubleMul(A=@n1.A, B=@n1.B)
   RigVMFunctionReturnNode(Result=&RigVMFunction_MathDoubleMul(B=@n1.C, A=@Multiply.Result).Result)
  ]}
 Multiply_2_1_ContainedGraph {nodes [
   n1 = RigVMFunctionEntryNode()
   n2 = &RigVMFunction_MathDoubleMul(A=@n1.A, B=@n1.B)
   RigVMFunctionReturnNode(Result=&RigVMFunction_MathDoubleMul(B=@n1.D, A=&RigVMFunction_MathDoubleMul(B=@n1.C, A=@n2.Result).Result).Result)
  ]}
 Multiply_3_1_ContainedGraph {nodes [
   n1 = RigVMFunctionEntryNode()
   RigVMFunctionReturnNode(Result=&RigVMFunction_MathDoubleMul(B=@n1.C, A=&RigVMFunction_MathDoubleMul(A=@n1.A, B=@n1.B).Result).Result)
  ]}
 Multiply_4_1_ContainedGraph {nodes [
   n1 = RigVMFunctionEntryNode()
   n2 = &RigVMFunction_MathDoubleMul(A=@n1.A, B=@n1.B)
   RigVMFunctionReturnNode(Result=&RigVMFunction_MathDoubleMul(B=@n1.D, A=&RigVMFunction_MathDoubleMul(B=@n1.C, A=@n2.Result).Result).Result)
  ]}
 Multiply_5_1_ContainedGraph {nodes [
   n1 = RigVMFunctionEntryNode()
   n2 = &RigVMFunction_MathDoubleMul(B=@n1.C, A=&RigVMFunction_MathDoubleMul(A=@n1.A, B=@n1.B).Result)
   RigVMFunctionReturnNode(Result=&RigVMFunction_MathDoubleMul(B=@n1.E, A=&RigVMFunction_MathDoubleMul(B=@n1.D, A=@n2.Result).Result).Result)
  ]}
 New Function_ContainedGraph {nodes [
   n1 = RigVMFunctionEntryNode()
   RigVMFunction_ControlFlowBranch = &RigVMFunction_ControlFlowBranch(ExecuteContext=@n1.Execute Content, Condition=@n1.debug)
   RigUnit_GetTransform = &RigUnit_GetTransform(Item=(Type=Control,Name="head_ctrl"), bInitial=False)
   &RigUnit_SetTransform(
    Item=(Type=Bone,Name="head")
    Weight=1
    bInitial=False
    Value=@RigUnit_GetTransform.Transform
    ExecutePin=@RigVMFunction_ControlFlowBranch.True
   )
   RigVMFunctionReturnNode(Execute Content=@RigVMFunction_ControlFlowBranch.Completed)
  ]}
 New Function_ContainedGraph {nodes [
   n1 = RigVMFunctionEntryNode()
   RigVMFunction_MathQuaternionToEuler = &RigVMFunction_MathQuaternionToEuler(
    RotationOrder=XYZ
    Value=&RigUnit_GetRelativeTransformForItem(
     Child=(Type=Bone,Name="hand_l")
     Parent=(Type=Null,Name="lowerarm_l")
     bChildInitial=False
     bParentInitial=False
     Child.Name=@n1.Child
     Parent.Name=@n1.Parent
    ).RelativeTransform.Rotation
   )
   Remap = &RigVMFunction_MathDoubleRemap(
    TargetMaximum=1
    TargetMinimum=0
    Value=@RigVMFunction_MathQuaternionToEuler.Result.X
    SourceMaximum=@n1.Source Maximum
    SourceMinimum=@n1.Source Minimum
   )
   > RigVMFunctionReturnNode(Delta=@Remap.Result)
  ]}
 New Function_ContainedGraph {nodes [
   n1 = RigVMFunctionEntryNode()
   RigVMFunction_MathQuaternionToEuler = &RigVMFunction_MathQuaternionToEuler(
    RotationOrder=ZYX
    Value=&RigUnit_GetRelativeTransformForItem(
     Child=(Type=Bone,Name="hand_l")
     Parent=(Type=Null,Name="lowerarm_l")
     bChildInitial=False
     bParentInitial=False
     Child.Name=@n1.Child
     Parent.Name=@n1.Parent
    ).RelativeTransform.Rotation
   )
   Remap = &RigVMFunction_MathDoubleRemap(
    TargetMaximum=1
    TargetMinimum=0
    Value=@RigVMFunction_MathQuaternionToEuler.Result.Z
    SourceMaximum=@n1.Source Maximum
    SourceMinimum=@n1.Source Minimum
   )
   > RigVMFunctionReturnNode(Delta=@Remap.Result)
  ]}
 New Function_ContainedGraph {nodes [
   RigVMFunctionEntryNode()
   > RigVMFunction_Sequence = RigVMAggregateNode()
   &RigUnit_ParentConstraint(
    AdvancedSettings=(InterpolationType=Average,RotationOrderForFilter=XZY)
    Child=(Type=Null,Name="dyn_hairB_C_01_ctrl_null")
    Filter=(TranslationFilter=(bX=True,bY=True,bZ=True),RotationFilter=(bX=false,bY=false,bZ=false),ScaleFilter=(bX=True,bY=True,bZ=True))
    Parents=((Item=(Type=Bone,Name="head"),Weight=0.9),(Item=(Type=Bone,Name="spine_05"),Weight=0.1))
    Weight=1
    ExecutePin=@RigVMFunction_Sequence.A
   )
   > &RigUnit_ParentConstraint(
    AdvancedSettings=(InterpolationType=Average,RotationOrderForFilter=XZY)
    Child=(Type=Null,Name="dyn_hairB_C_02_ctrl_null")
    Filter=(TranslationFilter=(bX=True,bY=True,bZ=True),RotationFilter=(bX=false,bY=false,bZ=false),ScaleFilter=(bX=True,bY=True,bZ=True))
    Parents=((Item=(Type=Bone,Name="neck_02"),Weight=0.9),(Item=(Type=Bone,Name="spine_05"),Weight=0.1))
    Weight=1
   )
   > &RigUnit_ParentConstraint(
    AdvancedSettings=(InterpolationType=Average,RotationOrderForFilter=XZY)
    Child=(Type=Null,Name="dyn_hairB_C_03_ctrl_null")
    Filter=(TranslationFilter=(bX=True,bY=True,bZ=True),RotationFilter=(bX=False,bY=false,bZ=false),ScaleFilter=(bX=True,bY=True,bZ=True))
    Parents=((Item=(Type=Bone,Name="spine_05"),Weight=0.4),(Item=(Type=Bone,Name="neck_02"),Weight=0.6))
    Weight=1
   )
   > &RigUnit_ParentConstraint(
    AdvancedSettings=(InterpolationType=Average,RotationOrderForFilter=XZY)
    Child=(Type=Null,Name="dyn_hairB_C_04_ctrl_null")
    Filter=(TranslationFilter=(bX=True,bY=True,bZ=True),RotationFilter=(bX=False,bY=false,bZ=false),ScaleFilter=(bX=True,bY=True,bZ=True))
    Parents=((Item=(Type=Bone,Name="spine_05"),Weight=1))
    Weight=1
   )
   > &RigUnit_ParentConstraint(
    AdvancedSettings=(InterpolationType=Average,RotationOrderForFilter=XZY)
    Child=(Type=Null,Name="dyn_hairB_C_05_ctrl_null")
    Filter=(TranslationFilter=(bX=True,bY=True,bZ=True),RotationFilter=(bX=False,bY=false,bZ=false),ScaleFilter=(bX=True,bY=True,bZ=True))
    Parents=((Item=(Type=Bone,Name="spine_04"),Weight=1))
    Weight=1
   )
   Make_Array = RigVMDispatchNode(Values=((Type=Bone,Name="dyn_hairB_C_01"),((),Type=Bone,Name="dyn_hairB_C_02"),((),Type=Bone,Name="dyn_hairB_C_03"),((),Type=Bone,Name="dyn_hairB_C_04"),((),Type=Bone,Name="dyn_hairB_C_05"))) {template="DISPATCH_RigVMDispatch_ArrayMake(in Values,out Array)"}
   > SplineIK = RigVMFunctionReferenceNode(
    Primary Axis=(X=1,Y=0,Z=0)
    Samples Per Segment=1
    Secondary Spline Direction=(X=0,Y=0,Z=0)
    Spline Mode=BSpline
    Stretch=false
    Up Axis=(X=0,Y=0,Z=0)
    debug=false
    Bones=@Make_Array.Array
    Controls=RigVMVariableNode() {variable=HairBack_CTL}.Value
   )
   > TwistBones = RigVMFunctionReferenceNode(
    Axis to Twist=(bX=true,bY=true,bZ=true)
    Bones to Twist=((Type=Bone,Name="dyn_hairB_C_04"),(Type=Bone,Name="dyn_hairB_C_05"))
    Debug=false
    End=(Type=Bone,Name="spine_05")
    Start=(Type=Bone,Name="spine_03")
   )
   RigVMFunctionReturnNode(ExecuteContext=@RigVMFunction_Sequence.D)
   &RigUnit_ParentConstraint(
    AdvancedSettings=(InterpolationType=Average,RotationOrderForFilter=XZY)
    Child=(Type=Null,Name="dyn_hairB_L_02_ctrl_null")
    Filter=(TranslationFilter=(bX=True,bY=True,bZ=True),RotationFilter=(bX=false,bY=false,bZ=false),ScaleFilter=(bX=True,bY=True,bZ=True))
    Parents=((Item=(Type=Bone,Name="neck_02"),Weight=1),(Item=(Type=Bone,Name="spine_05"),Weight=0))
    Weight=1
    ExecutePin=@RigVMFunction_Sequence.C
   )
   > &RigUnit_ParentConstraint(
    AdvancedSettings=(InterpolationType=Average,RotationOrderForFilter=XZY)
    Child=(Type=Null,Name="dyn_hairB_L_03_ctrl_null")
    Filter=(TranslationFilter=(bX=True,bY=True,bZ=True),RotationFilter=(bX=False,bY=false,bZ=false),ScaleFilter=(bX=True,bY=True,bZ=True))
    Parents=((Item=(Type=Bone,Name="spine_05"),Weight=0.4),(Item=(Type=Bone,Name="neck_02"),Weight=0.6))
    Weight=1
   )
   > &RigUnit_ParentConstraint(
    AdvancedSettings=(InterpolationType=Average,RotationOrderForFilter=XZY)
    Child=(Type=Null,Name="dyn_hairB_L_04_ctrl_null")
    Filter=(TranslationFilter=(bX=True,bY=True,bZ=True),RotationFilter=(bX=False,bY=false,bZ=false),ScaleFilter=(bX=True,bY=True,bZ=True))
    Parents=((Item=(Type=Bone,Name="spine_05"),Weight=1))
    Weight=1
   )
   > &RigUnit_ParentConstraint(
    AdvancedSettings=(InterpolationType=Average,RotationOrderForFilter=XZY)
    Child=(Type=Null,Name="dyn_hairB_L_05_ctrl_null")
    Filter=(TranslationFilter=(bX=True,bY=True,bZ=True),RotationFilter=(bX=False,bY=false,bZ=false),ScaleFilter=(bX=True,bY=True,bZ=True))
    Parents=((Item=(Type=Bone,Name="spine_05"),Weight=1))
    Weight=1
   )
   Make_Array_1_1 = RigVMDispatchNode(Values=(((),Type=Bone,Name="dyn_hairB_L_02"),((),Type=Bone,Name="dyn_hairB_L_03"),((),Type=Bone,Name="dyn_hairB_L_04"),((),Type=Bone,Name="dyn_hairB_L_05"))) {template="DISPATCH_RigVMDispatch_ArrayMake(in Values,out Array)"}
   > SplineIK_1_1 = RigVMFunctionReferenceNode(
    Primary Axis=(X=1,Y=0,Z=0)
    Samples Per Segment=1
    Secondary Spline Direction=(X=-1,Y=-1,Z=-1)
    Spline Mode=BSpline
    Stretch=false
    Up Axis=(X=0,Y=0,Z=0)
    debug=false
    Bones=@Make_Array_1_1.Array
    Controls=RigVMVariableNode() {variable=HairBack_CTL_L}.Value
   )
   > TwistBones_1_1 = RigVMFunctionReferenceNode(
    Axis to Twist=(bX=true,bY=true,bZ=true)
    Bones to Twist=((Type=Bone,Name="dyn_hairB_L_04"),(Type=Bone,Name="dyn_hairB_L_05"))
    End=(Type=Bone,Name="spine_05")
    Start=(Type=Bone,Name="spine_05")
   )
   &RigUnit_ParentConstraint(
    AdvancedSettings=(InterpolationType=Average,RotationOrderForFilter=XZY)
    Child=(Type=Null,Name="dyn_hairB_R_02_ctrl_null")
    Filter=(TranslationFilter=(bX=True,bY=True,bZ=True),RotationFilter=(bX=false,bY=false,bZ=false),ScaleFilter=(bX=True,bY=True,bZ=True))
    Parents=((Item=(Type=Bone,Name="neck_02"),Weight=0.9),(Item=(Type=Bone,Name="spine_05"),Weight=0.1))
    Weight=1
    ExecutePin=@RigVMFunction_Sequence.B
   )
   > &RigUnit_ParentConstraint(
    AdvancedSettings=(InterpolationType=Average,RotationOrderForFilter=XZY)
    Child=(Type=Null,Name="dyn_hairB_R_03_ctrl_null")
    Filter=(TranslationFilter=(bX=True,bY=True,bZ=True),RotationFilter=(bX=False,bY=false,bZ=false),ScaleFilter=(bX=True,bY=True,bZ=True))
    Parents=((Item=(Type=Bone,Name="spine_05"),Weight=0.4),(Item=(Type=Bone,Name="neck_02"),Weight=0.6))
    Weight=1
   )
   > &RigUnit_ParentConstraint(
    AdvancedSettings=(InterpolationType=Average,RotationOrderForFilter=XZY)
    Child=(Type=Null,Name="dyn_hairB_R_04_ctrl_null")
    Filter=(TranslationFilter=(bX=True,bY=True,bZ=True),RotationFilter=(bX=False,bY=false,bZ=false),ScaleFilter=(bX=True,bY=True,bZ=True))
    Parents=((Item=(Type=Bone,Name="spine_05"),Weight=1))
    Weight=1
   )
   > &RigUnit_ParentConstraint(
    AdvancedSettings=(InterpolationType=Average,RotationOrderForFilter=XZY)
    Child=(Type=Null,Name="dyn_hairB_R_05_ctrl_null")
    Filter=(TranslationFilter=(bX=True,bY=True,bZ=True),RotationFilter=(bX=False,bY=false,bZ=false),ScaleFilter=(bX=True,bY=True,bZ=True))
    Parents=((Item=(Type=Bone,Name="spine_05"),Weight=1))
    Weight=1
   )
   Make_Array_1 = RigVMDispatchNode(Values=(((),Type=Bone,Name="dyn_hairB_R_02"),((),Type=Bone,Name="dyn_hairB_R_03"),((),Type=Bone,Name="dyn_hairB_R_04"),((),Type=Bone,Name="dyn_hairB_R_05"))) {template="DISPATCH_RigVMDispatch_ArrayMake(in Values,out Array)"}
   > SplineIK_1 = RigVMFunctionReferenceNode(
    Primary Axis=(X=1,Y=0,Z=0)
    Samples Per Segment=1
    Secondary Spline Direction=(X=0,Y=0,Z=0)
    Spline Mode=BSpline
    Stretch=false
    Up Axis=(X=0,Y=0,Z=0)
    debug=false
    Bones=@Make_Array_1.Array
    Controls=RigVMVariableNode() {variable=HairBack_CTL_R}.Value
   )
   > TwistBones_1 = RigVMFunctionReferenceNode(
    Axis to Twist=(bX=True,bY=true,bZ=true)
    Bones to Twist=((Type=Bone,Name="dyn_hairB_R_04"),(Type=Bone,Name="dyn_hairB_R_05"))
    End=(Type=Bone,Name="spine_05")
    Start=(Type=Bone,Name="spine_05")
   )
   &RigUnit_ParentConstraint(
    AdvancedSettings=(InterpolationType=Average,RotationOrderForFilter=XZY)
    Child=(Type=Bone,Name="dyn_hairB_R_01")
    ExecutePin=()
    Parents=((Item=(Type=Bone,Name="dyn_hairB_C_01"),Weight=1))
    Weight=1
   )
   > &RigUnit_ParentConstraint(
    AdvancedSettings=(InterpolationType=Average,RotationOrderForFilter=XZY)
    Child=(Type=Bone,Name="dyn_hairB_R_02")
    Parents=((Item=(Type=Bone,Name="dyn_hairB_C_02"),Weight=1))
    Weight=1
   )
   > &RigUnit_ParentConstraint(
    AdvancedSettings=(InterpolationType=Average,RotationOrderForFilter=XZY)
    Child=(Type=Bone,Name="dyn_hairB_R_03")
    Parents=((Item=(Type=Bone,Name="dyn_hairB_C_03"),Weight=1))
    Weight=1
   )
   > &RigUnit_ParentConstraint(
    AdvancedSettings=(InterpolationType=Average,RotationOrderForFilter=XZY)
    Child=(Type=Bone,Name="dyn_hairB_R_04")
    Parents=((Item=(Type=Bone,Name="dyn_hairB_C_04"),Weight=1))
    Weight=1
   )
   > &RigUnit_ParentConstraint(
    AdvancedSettings=(InterpolationType=Average,RotationOrderForFilter=XZY)
    Child=(Type=Bone,Name="dyn_hairB_R_05")
    Parents=((Item=(Type=Bone,Name="dyn_hairB_C_05"),Weight=1))
    Weight=1
   )
   &RigUnit_ParentConstraint(
    AdvancedSettings=(InterpolationType=Average,RotationOrderForFilter=XZY)
    Child=(Type=Bone,Name="dyn_hairB_L_01")
    ExecutePin=()
    Filter=(TranslationFilter=(bX=true,bY=true,bZ=true),RotationFilter=(bX=true,bY=True,bZ=True),ScaleFilter=(bX=True,bY=True,bZ=True))
    Parents=((Item=(Type=Bone,Name="dyn_hairB_C_01"),Weight=1))
    Weight=1
   )
   > &RigUnit_ParentConstraint(
    AdvancedSettings=(InterpolationType=Average,RotationOrderForFilter=XZY)
    Child=(Type=Bone,Name="dyn_hairB_L_02")
    Parents=((Item=(Type=Bone,Name="dyn_hairB_C_02"),Weight=1))
    Weight=1
   )
   > &RigUnit_ParentConstraint(
    AdvancedSettings=(InterpolationType=Average,RotationOrderForFilter=XZY)
    Child=(Type=Bone,Name="dyn_hairB_L_03")
    Parents=((Item=(Type=Bone,Name="dyn_hairB_C_03"),Weight=1))
    Weight=1
   )
   > &RigUnit_ParentConstraint(
    AdvancedSettings=(InterpolationType=Average,RotationOrderForFilter=XZY)
    Child=(Type=Bone,Name="dyn_hairB_L_04")
    Parents=((Item=(Type=Bone,Name="dyn_hairB_C_04"),Weight=1))
    Weight=1
   )
   > &RigUnit_ParentConstraint(
    AdvancedSettings=(InterpolationType=Average,RotationOrderForFilter=XZY)
    Child=(Type=Bone,Name="dyn_hairB_L_05")
    Parents=((Item=(Type=Bone,Name="dyn_hairB_C_05"),Weight=1))
    Weight=1
   )
   &RigUnit_ParentConstraint(
    AdvancedSettings=(InterpolationType=Average,RotationOrderForFilter=XZY)
    Child=(Type=Null,Name="dyn_hairB_R_01_ctrl_null")
    ExecutePin=()
    Filter=(TranslationFilter=(bX=True,bY=True,bZ=True),RotationFilter=(bX=false,bY=false,bZ=false),ScaleFilter=(bX=True,bY=True,bZ=True))
    Parents=((Item=(Type=Bone,Name="head"),Weight=0.9),(Item=(Type=Bone,Name="spine_05"),Weight=0.1))
    Weight=0
   )
   &RigUnit_ParentConstraint(
    AdvancedSettings=(InterpolationType=Average,RotationOrderForFilter=XZY)
    Child=(Type=Null,Name="dyn_hairB_L_01_ctrl_null")
    ExecutePin=()
    Filter=(TranslationFilter=(bX=True,bY=True,bZ=True),RotationFilter=(bX=false,bY=false,bZ=false),ScaleFilter=(bX=True,bY=True,bZ=True))
    Parents=((Item=(Type=Bone,Name="head"),Weight=1),(Item=(Type=Bone,Name="spine_05"),Weight=0))
    Weight=1
   )
  ]}
 New Function_ContainedGraph {nodes [
   n1 = RigVMFunctionEntryNode()
   > DISPATCH_RigVMDispatch_ArrayIterator = RigVMDispatchNode(Array=@n1.transform) {template="DISPATCH_RigVMDispatch_ArrayIterator(in Array,out Element,out Index,out Count,out Ratio)"}
   Multiply = &RigVMFunction_MathTransformMul(
    A=@DISPATCH_RigVMDispatch_ArrayIterator.Element.Transform
    B=&RigUnit_GetTransform(
     Item=(Type=Bone,Name="dyn_L_hair_3")
     Space=LocalSpace
     bInitial=False
     Item.Name=@DISPATCH_RigVMDispatch_ArrayIterator.Element.Bone
     Item.Type=@n1.type
    ).Transform
   ) {template="Multiply::Execute(in A,in B,out Result)"}
   > &RigUnit_SetTransform(
    Item=(Type=Bone,Name="dyn_L_hair_3")
    Space=LocalSpace
    bInitial=false
    Value=@Multiply.Result
    bPropagateToChildren=@n1.inheritsTransform
    Weight=@n1.weight
    Item.Name=@DISPATCH_RigVMDispatch_ArrayIterator.Element.Bone
    Item.Type=@n1.type
   )
   RigVMFunctionReturnNode(ExecuteContext=@DISPATCH_RigVMDispatch_ArrayIterator.Completed)
  ]}
 New Function_ContainedGraph {nodes [
   RigVMFunctionEntryNode()
   > RigVMFunction_Sequence = RigVMAggregateNode()
   AlterTransform = RigVMFunctionReferenceNode(
    type=Bone
    transform=((Bone="dyn_hairB_R_01",Transform=(Rotation=(X=0,Y=0,Z=0,W=1),Translation=(X=-2,Y=-5,Z=0),Scale3D=(X=1,Y=1,Z=1))),(Bone="dyn_hairB_R_02",Transform=(Rotation=(X=0,Y=0,Z=0,W=1),Translation=(X=0,Y=-4,Z=0),Scale3D=(X=1,Y=1,Z=1))))
    ExecuteContext=@RigVMFunction_Sequence.A
    weight=RigVMVariableNode() {variable=HeadLookUp}.Value
   )
   > AlterTransform_1 = RigVMFunctionReferenceNode(
    type=Bone
    transform=((Bone="dyn_hairB_R_03",Transform=(Rotation=(X=0,Y=0,Z=0,W=1),Translation=(X=0,Y=-1,Z=1),Scale3D=(X=1,Y=1,Z=1))),(Bone="dyn_hairB_R_02",Transform=(Rotation=(X=0,Y=0,Z=0,W=1),Translation=(X=0,Y=-5,Z=2),Scale3D=(X=1,Y=1,Z=1))))
    weight=RigVMVariableNode() {variable=HeadLookLeft}.Value
   )
   > AlterTransform_1_1 = RigVMFunctionReferenceNode(
    type=Bone
    transform=((Bone="dyn_hairB_L_03",Transform=(Rotation=(X=0,Y=0,Z=0,W=1),Translation=(X=0,Y=0,Z=-2),Scale3D=(X=1,Y=1,Z=1))),(Bone="dyn_hairB_L_02",Transform=(Rotation=(X=0,Y=0,Z=0,W=1),Translation=(X=0,Y=-2,Z=2),Scale3D=(X=1,Y=1,Z=1))))
    weight=RigVMVariableNode() {variable=HeadLookRight}.Value
   )
   > AlterTransform_1_1_1 = RigVMFunctionReferenceNode(
    type=Bone
    transform=((Bone="dyn_hairB_R_03",Transform=(Rotation=(X=0,Y=0,Z=0,W=1),Translation=(X=-3,Y=-2,Z=0),Scale3D=(X=1,Y=1,Z=1))),(Bone="dyn_hairB_L_03",Transform=(Rotation=(X=0,Y=0,Z=0,W=1),Translation=(X=0,Y=-1.5,Z=0),Scale3D=(X=1,Y=1,Z=1))))
    weight=RigVMVariableNode() {variable=idlePose}.Value
   )
   Multiply = &RigVMFunction_MathDoubleMul(B=RigVMVariableNode() {variable=Pistol}.Value, A=RigVMVariableNode() {variable=NotMeele}.Value)
   Remap = &RigVMFunction_MathDoubleRemap(SourceMaximum=1, SourceMinimum=0, TargetMaximum=0, TargetMinimum=1, bClamp=False, Value=RigVMVariableNode() {variable=HeadLookDown}.Value)
   Multiply_1 = RigVMAggregateNode(
    B=@Multiply.Result
    C=@Remap.Result
    D=&RigVMFunction_MathDoubleRemap(SourceMaximum=1, SourceMinimum=0, TargetMaximum=0, TargetMinimum=1, bClamp=False, Value=RigVMVariableNode() {variable=HairMovement}.Value).Result
    E=&RigVMFunction_MathDoubleRemap(SourceMaximum=1, SourceMinimum=0, TargetMaximum=0, TargetMinimum=1, bClamp=False, Value=RigVMVariableNode() {variable=isEmoting}.Value).Result
    A=RigVMVariableNode() {variable=ArmAimUp}.Value
   )
   > &RigUnit_ModifyTransforms(
    ItemToModify="((Item=(Type=Bone,Name=\"dyn_hairB_R_02\"),Transform=(Rotation=(X=0,Y=0,Z=-0.0788,W=0.9969),Translation=(X=-4.7581,Y=-2.3585,Z=1.4915),Scale3D=(X=1,Y=1,Z=1))),(Item=(Type=Bone,Name=\"dyn_hairB_C_01\"),Transform=(Rotation=(X=-0.0017,Y=-0.0151,Z=-0.1137,W=0.9934 ...(truncated)"
    WeightMinimum=0
    Weight=@Multiply_1.Result
   )
   > &RigUnit_ModifyTransforms(
    ItemToModify="((Item=(Type=Bone,Name=\"dyn_hair_side_01_l\"),Transform=(Rotation=(X=0.0281,Y=-0.1619,Z=0.2988,W=0.9401),Translation=(X=0,Y=0,Z=0),Scale3D=(X=1,Y=1,Z=1))),(Item=(Type=Bone,Name=\"dyn_hair_front_01_l\"),Transform=(Rotation=(X=0,Y=0,Z=0.2558,W=0.9667),Translati ...(truncated)"
    WeightMinimum=0
    Weight=RigVMVariableNode() {variable=HeadLookDown}.Value
   )
   RigVMFunction_ControlFlowBranch = &RigVMFunction_ControlFlowBranch(ExecuteContext=@RigVMFunction_Sequence.B, Condition=RigVMVariableNode() {variable=EmotingBool}.Value)
   &RigUnit_ModifyTransforms(
    ItemToModify="((Item=(Type=Bone,Name=\"dyn_hairB_C_03\"),Transform=(Rotation=(X=0,Y=0,Z=0,W=1),Translation=(X=0,Y=-2,Z=0),Scale3D=(X=1,Y=1,Z=1))),(Item=(Type=Bone,Name=\"dyn_hairB_C_04\"),Transform=(Rotation=(X=0,Y=0,Z=0,W=1),Translation=(X=0,Y=-2,Z=0),Scale3D=(X=1,Y=1,Z=1) ...(truncated)"
    Weight=1
    WeightMinimum=0
    ExecutePin=@RigVMFunction_ControlFlowBranch.True
   )
   > n1 = &RigUnit_SphericalPoseReader(
    ActiveRegionScaleFactors=(PositiveWidth=1,NegativeWidth=1,PositiveHeight=1.2,NegativeHeight=1)
    ActiveRegionSize=0.1
    Debug=(bDrawDebug=false,bDraw2D=False,bDrawLocalAxes=False,DebugScale=25,DebugSegments=20,DebugThickness=0.25)
    DriverAxis=(X=0,Y=1,Z=0)
    DriverItem=(Type=Bone,Name="head")
    FalloffRegionScaleFactors=(PositiveWidth=1,NegativeWidth=1,PositiveHeight=1,NegativeHeight=1)
    FalloffSize=0.25
    FlipHeightScaling=False
    FlipWidthScaling=False
    OptionalParentItem=(Type=Bone,Name="None")
    RotationOffset=(X=40,Y=0,Z=0)
   )
   n2 = &RigVMFunction_MathDoubleMul(A=@n1.OutputParam, B=RigVMVariableNode() {variable=isCustomEmotePlaying}.Value)
   > &RigUnit_ModifyTransforms(
    ItemToModify="((Item=(Type=Bone,Name=\"dyn_hairB_L_01\"),Transform=(Rotation=(X=0,Y=0,Z=-0.0436,W=0.999),Translation=(X=0,Y=-1,Z=0),Scale3D=(X=1,Y=1,Z=1))),(Item=(Type=Bone,Name=\"dyn_hairB_L_02\"),Transform=(Rotation=(X=0,Y=0,Z=-0.0436,W=0.999),Translation=(X=0,Y=-1,Z=0),S ...(truncated)"
    WeightMinimum=0
    Weight=@n2.Result
   )
   > &RigUnit_ModifyTransforms(
    ItemToModify="((Item=(Type=Bone,Name=\"dyn_hairB_C_01\"),Transform=(Rotation=(X=0,Y=0,Z=-0.0436,W=0.999),Translation=(X=0,Y=0,Z=0),Scale3D=(X=1,Y=1,Z=1))),(Item=(Type=Bone,Name=\"dyn_hairB_C_02\"),Transform=(Rotation=(X=0,Y=0,Z=-0.0436,W=0.999),Translation=(X=0,Y=0,Z=0),Sca ...(truncated)"
    WeightMinimum=0
    Weight=@n2.Result
   )
   > n3 = &RigUnit_SphericalPoseReader(
    ActiveRegionScaleFactors=(PositiveWidth=1,NegativeWidth=1,PositiveHeight=1,NegativeHeight=1.2)
    ActiveRegionSize=0.1
    Debug=(bDrawDebug=false,bDraw2D=False,bDrawLocalAxes=False,DebugScale=25,DebugSegments=20,DebugThickness=0.25)
    DriverAxis=(X=0,Y=1,Z=0)
    DriverItem=(Type=Bone,Name="head")
    FalloffRegionScaleFactors=(PositiveWidth=1,NegativeWidth=1,PositiveHeight=1,NegativeHeight=1)
    FalloffSize=0.2
    FlipHeightScaling=False
    FlipWidthScaling=false
    OptionalParentItem=(Type=Bone,Name="None")
    RotationOffset=(X=140,Y=0,Z=0)
   )
   n4 = &RigVMFunction_MathDoubleMul(A=@n3.OutputParam, B=RigVMVariableNode() {variable=isCustomEmotePlaying}.Value)
   > &RigUnit_ModifyTransforms(
    ItemToModify="((Item=(Type=Bone,Name=\"dyn_hairB_R_01\"),Transform=(Rotation=(X=0,Y=0,Z=-0.0262,W=0.9997),Translation=(X=0,Y=0,Z=0),Scale3D=(X=1,Y=1,Z=1))),(Item=(Type=Bone,Name=\"dyn_hairB_R_02\"),Transform=(Rotation=(X=0,Y=0,Z=-0.0262,W=0.9997),Translation=(X=0,Y=0,Z=0),S ...(truncated)"
    WeightMinimum=0
    Weight=@n4.Result
   )
   > &RigUnit_ModifyTransforms(
    ItemToModify="((Item=(Type=Bone,Name=\"dyn_hairB_C_01\"),Transform=(Rotation=(X=0,Y=0,Z=-0.0436,W=0.999),Translation=(X=0,Y=0,Z=0),Scale3D=(X=1,Y=1,Z=1))),(Item=(Type=Bone,Name=\"dyn_hairB_C_02\"),Transform=(Rotation=(X=0,Y=0,Z=-0.0436,W=0.999),Translation=(X=0,Y=0,Z=0),Sca ...(truncated)"
    WeightMinimum=0
    Weight=@n4.Result
   )
   RigVMFunctionReturnNode(ExecuteContext=@RigVMFunction_Sequence.C)
  ]}
 New Function_ContainedGraph {nodes [
   n1 = RigVMFunctionEntryNode()
   RigVMFunction_MathQuaternionToEuler = &RigVMFunction_MathQuaternionToEuler(
    RotationOrder=ZYX
    Value=&RigUnit_GetRelativeTransformForItem(
     Child=(Type=Bone,Name="hand_l")
     Parent=(Type=Bone,Name="lowerarm_l")
     bChildInitial=False
     bParentInitial=False
     Child.Name=@n1.Child
     Parent.Name=@n1.Parent
    ).RelativeTransform.Rotation
   )
   Remap = &RigVMFunction_MathDoubleRemap(
    TargetMaximum=1
    TargetMinimum=0
    Value=@RigVMFunction_MathQuaternionToEuler.Result.Z
    SourceMaximum=@n1.Source Maximum
    SourceMinimum=@n1.Source Minimum
   )
   > RigVMFunctionReturnNode(Delta Z=@Remap.Result)
  ]}
 New Function_ContainedGraph {nodes [
   n1 = RigVMFunctionEntryNode()
   RigVMFunction_MathQuaternionToEuler = &RigVMFunction_MathQuaternionToEuler(
    RotationOrder=ZYX
    Value=&RigUnit_GetRelativeTransformForItem(
     Child=(Type=Bone,Name="hand_l")
     Parent=(Type=Null,Name="lowerarm_l")
     bChildInitial=False
     bParentInitial=False
     Child.Name=@n1.Child
     Parent.Name=@n1.Parent
    ).RelativeTransform.Rotation
   )
   Remap = &RigVMFunction_MathDoubleRemap(
    TargetMaximum=1
    TargetMinimum=0
    Value=@RigVMFunction_MathQuaternionToEuler.Result.Y
    SourceMaximum=@n1.Source Maximum
    SourceMinimum=@n1.Source Minimum
   )
   > RigVMFunctionReturnNode(Delta=@Remap.Result)
  ]}
 New Function_ContainedGraph {nodes [
   n1 = RigVMFunctionEntryNode()
   RigVMFunction_Sequence = &RigVMFunction_Sequence(ExecuteContext=@n1.ExecuteContext)
   DeltaRotation_Z = RigVMFunctionReferenceNode(Child=head, Parent=head_null, Source Maximum=-35, Source Minimum=0, ExecuteContext=@RigVMFunction_Sequence.A)
   > RigVMVariableNode(Value=@DeltaRotation_Z.Delta) {variable=HeadLookUp}
   > DeltaRotation_Z_1 = RigVMFunctionReferenceNode(Child=head, Parent=head_null, Source Maximum=10, Source Minimum=-10)
   > RigVMVariableNode(Value=@DeltaRotation_Z_1.Delta) {variable=HeadLookDown}
   > DeltaRotation_X = RigVMFunctionReferenceNode(Child=head, Parent=head_null, Source Maximum=-15, Source Minimum=10)
   > RigVMVariableNode(Value=@DeltaRotation_X.Delta) {variable=HeadLookLeft}
   > DeltaRotation_X_1 = RigVMFunctionReferenceNode(Child=head, Parent=head_null, Source Maximum=30, Source Minimum=10)
   > RigVMVariableNode(Value=@DeltaRotation_X_1.Delta) {variable=HeadLookRight}
   > DeltaRotation_Z_1_1 = RigVMFunctionReferenceNode(Child=upperarm_l, Parent=upperarm_l_null, Source Maximum=45, Source Minimum=28)
   > RigVMVariableNode(Value=@DeltaRotation_Z_1_1.Delta) {variable=ArmAimUp}
   > ExtremeDelta = RigVMFunctionReferenceNode(Child=head, Parent=neck_01, Source Maximum=-85, Source Minimum=-30)
   > RigVMVariableNode(Value=@ExtremeDelta.Delta Z) {variable=isHeadUpExtreme}
   > DeltaRotation_Z_1_1_1 = RigVMFunctionReferenceNode(Child=head, Parent=head_null, Source Maximum=-12, Source Minimum=0)
   > RigVMVariableNode(Value=@DeltaRotation_Z_1_1_1.Delta) {variable=hairFix}
   RigVMFunctionReturnNode(ExecuteContext=@RigVMFunction_Sequence.B)
  ]}
 New Function_ContainedGraph {nodes [
   RigVMFunctionEntryNode()
   > RigVMFunction_Sequence = RigVMAggregateNode()
   Remap = &RigVMFunction_MathDoubleRemap(SourceMaximum=1, SourceMinimum=0, TargetMaximum=0, TargetMinimum=1, bClamp=False, Value=RigVMVariableNode() {variable=isEmoting}.Value)
   Multiply = RigVMAggregateNode(
    A=@Remap.Result
    C=&RigVMFunction_MathDoubleRemap(SourceMaximum=1, SourceMinimum=0, TargetMaximum=0, TargetMinimum=1, bClamp=False, Value=RigVMVariableNode() {variable=isCrouching}.Value).Result
    B=RigVMVariableNode() {variable=HeadLookUp}.Value
   )
   &RigUnit_ModifyTransforms(
    ItemToModify="((Item=(Type=Bone,Name=\"base_hairPieceE\"),Transform=(Rotation=(X=-0.0137,Y=-0.3557,Z=-0.2769,W=0.8925),Translation=(X=0,Y=0,Z=0),Scale3D=(X=1,Y=1,Z=1))),(Item=(Type=Bone,Name=\"base_hairPieceF\"),Transform=(Rotation=(X=0.1944,Y=-0.6486,Z=-0.279,W=0.6809),Tra ...(truncated)"
    WeightMinimum=0
    Weight=@Multiply.Result
    ExecutePin=@RigVMFunction_Sequence.A
   )
   Multiply_3 = RigVMAggregateNode(
    C=&RigVMFunction_MathDoubleRemap(SourceMaximum=1, SourceMinimum=0, TargetMaximum=0, TargetMinimum=1, bClamp=False, Value=RigVMVariableNode() {variable=isTargeting}.Value).Result
    A=RigVMVariableNode() {variable=HeadLookUp}.Value
    B=RigVMVariableNode() {variable=isCrouching}.Value
   )
   > &RigUnit_ModifyTransforms(
    ItemToModify="((Item=(Type=Bone,Name=\"base_hairPieceE\"),Transform=(Rotation=(X=0.2851,Y=-0.417,Z=-0.2341,W=0.8306),Translation=(X=0,Y=0,Z=0),Scale3D=(X=1,Y=1,Z=1))),(Item=(Type=Bone,Name=\"base_hairPieceF\"),Transform=(Rotation=(X=0.211,Y=-0.1614,Z=-0.016,W=0.9639),Transl ...(truncated)"
    WeightMinimum=0
    Weight=@Multiply_3.Result
   )
   Multiply_1 = RigVMAggregateNode(
    A=RigVMVariableNode() {variable=CloseGroundRemap}.Value
    B=RigVMVariableNode() {variable=isHeadUpExtreme}.Value
    C=RigVMVariableNode() {variable=isCrouching}.Value
   )
   > &RigUnit_ModifyTransforms(
    ItemToModify="((Item=(Type=Bone,Name=\"base_hairPieceF\"),Transform=(Rotation=(X=0,Y=-0.9381,Z=0,W=0.3462),Translation=(X=0,Y=0,Z=0),Scale3D=(X=1,Y=1,Z=1))),(Item=(Type=Bone,Name=\"base_hairPieceG\"),Transform=(Rotation=(X=-0.0212,Y=-0.9393,Z=0.0516,W=0.3386),Translation=(X ...(truncated)"
    WeightMinimum=0
    Weight=@Multiply_1.Result
   )
   Multiply_1_1 = RigVMAggregateNode(
    F=&RigVMFunction_MathDoubleRemap(
     SourceMaximum=0.5
     SourceMinimum=0
     TargetMaximum=1
     TargetMinimum=0
     bClamp=False
     Value=RigVMVariableNode() {variable=isHeadUpExtreme}.Value
    ).Result
    E=RigVMVariableNode() {variable=isTargeting}.Value
    B=RigVMVariableNode() {variable=isCrouching}.Value
    A=RigVMVariableNode() {variable=NotMeele}.Value
   )
   > &RigUnit_ModifyTransforms(
    ItemToModify="((Item=(Type=Bone,Name=\"base_hatMain\"),Transform=(Rotation=(X=-0.2821,Y=0.2473,Z=-0.0621,W=0.9249),Translation=(X=3.5598,Y=-0.4108,Z=-1.8648),Scale3D=(X=1,Y=1,Z=1))),(Item=(Type=Bone,Name=\"base_hairPieceG\"),Transform=(Rotation=(X=0,Y=0,Z=0,W=1),Translation ...(truncated)"
    WeightMinimum=0
    Weight=&RigVMFunction_MathDoubleRemap(SourceMaximum=1, SourceMinimum=0, TargetMaximum=1, TargetMinimum=0, Value=@Multiply_1_1.Result).Result
   )
   Multiply_2 = RigVMAggregateNode(
    A=&RigVMFunction_MathDoubleRemap(SourceMaximum=1, SourceMinimum=0.5, TargetMaximum=1, TargetMinimum=0, Value=RigVMVariableNode() {variable=CloseGroundRemap}.Value).Result
    C=RigVMVariableNode() {variable=isTargeting}.Value
    D=RigVMVariableNode() {variable=HeadLookUp}.Value
    B=RigVMVariableNode() {variable=isCrouching}.Value
   )
   > &RigUnit_ModifyTransforms(
    ItemToModify="((Item=(Type=Bone,Name=\"base_hatMain\"),Transform=(Rotation=(X=0.2173,Y=-0.3158,Z=0.1226,W=0.9154),Translation=(X=-4.2407,Y=-0.2309,Z=2.6532),Scale3D=(X=1,Y=1,Z=1))),(Item=(Type=Bone,Name=\"base_hairPieceG\"),Transform=(Rotation=(X=0.0018,Y=-0.0264,Z=-0.1508, ...(truncated)"
    WeightMinimum=0
    Weight=@Multiply_2.Result
   )
   RigVMFunction_MathBoolToFloat = &RigVMFunction_MathBoolToFloat(Value=RigVMVariableNode() {variable=SlyDive}.Value)
   > &RigUnit_ModifyTransforms(
    ItemToModify="((Item=(Type=Bone,Name=\"base_hairPieceF\"),Transform=(Rotation=(X=0,Y=0,Z=0,W=1),Translation=(X=0,Y=0,Z=0),Scale3D=(X=1,Y=1,Z=1))),(Item=(Type=Bone,Name=\"base_hatMain\"),Transform=(Rotation=(X=-0.1858,Y=0,Z=0,W=0.9826),Translation=(X=0,Y=0,Z=1.5552),Scale3D= ...(truncated)"
    WeightMinimum=0
    Weight=@RigVMFunction_MathBoolToFloat.Result
   )
   > n1 = &RigUnit_SphericalPoseReader(
    ActiveRegionScaleFactors=(PositiveWidth=2,NegativeWidth=0.4,PositiveHeight=1.5,NegativeHeight=1.5)
    ActiveRegionSize=0.15
    Debug=(bDrawDebug=false,bDraw2D=False,bDrawLocalAxes=False,DebugScale=25,DebugSegments=20,DebugThickness=0.25)
    DriverAxis=(X=0,Y=1,Z=0)
    DriverItem=(Type=Bone,Name="head")
    FalloffRegionScaleFactors=(PositiveWidth=1,NegativeWidth=0.3,PositiveHeight=1,NegativeHeight=1)
    FalloffSize=0.3
    FlipHeightScaling=False
    FlipWidthScaling=False
    OptionalParentItem=(Type=Bone,Name="None")
    RotationOffset=(X=90,Y=-7,Z=-25)
    FalloffRegionScaleFactors.NegativeWidth=RigVMDispatchNode(False=0.7, True=0.3, Condition=RigVMVariableNode() {variable=Aimfloat}.Value) {template="DISPATCH_RigVMDispatch_If(in Condition,in True,in False,out Result)"}.Result
    ActiveRegionScaleFactors.NegativeWidth=RigVMDispatchNode(False=0.4, True=0.2, Condition=RigVMVariableNode() {variable=Aimfloat}.Value) {template="DISPATCH_RigVMDispatch_If(in Condition,in True,in False,out Result)"}.Result
    RotationOffset.Z=&RigVMFunction_MathDoubleRemap(SourceMaximum=1, SourceMinimum=0, TargetMaximum=-40, TargetMinimum=-25, bClamp=False, Value=RigVMVariableNode() {variable=isCrouching}.Value).Result
   )
   Multiply_4 = RigVMAggregateNode(
    A=&RigVMFunction_MathDoubleRemap(SourceMaximum=1, SourceMinimum=0, TargetMaximum=0, TargetMinimum=1, bClamp=False, Value=RigVMVariableNode() {variable=isTargeting}.Value).Result
    B=&RigVMFunction_MathDoubleRemap(SourceMaximum=1, SourceMinimum=0, TargetMaximum=0, TargetMinimum=1, bClamp=False, Value=RigVMVariableNode() {variable=isCrouching}.Value).Result
    D=@n1.OutputParam
    C=RigVMVariableNode() {variable=idlePose}.Value
   )
   > &RigUnit_ModifyTransforms(
    ItemToModify="((Item=(Type=Bone,Name=\"dyn_hairB_L_03\"),Transform=(Rotation=(X=0,Y=0,Z=0.0333,W=0.9994),Translation=(X=2.4895,Y=0.948,Z=0),Scale3D=(X=1,Y=1,Z=1))),(Item=(Type=Bone,Name=\"dyn_hairB_R_03\"),Transform=(Rotation=(X=0,Y=0,Z=0.061,W=0.9981),Translation=(X=0,Y=0, ...(truncated)"
    WeightMinimum=0
    Weight=@Multiply_4.Result
   )
   Multiply_5 = RigVMAggregateNode(
    D=&RigVMFunction_MathDoubleRemap(SourceMaximum=1, SourceMinimum=0.6, TargetMaximum=0, TargetMinimum=1, Value=RigVMVariableNode() {variable=CloseGroundRemap}.Value).Result
    A=RigVMVariableNode() {variable=hairFix}.Value
    B=RigVMVariableNode() {variable=isCrouching}.Value
    C=RigVMVariableNode() {variable=isTargeting}.Value
    E=RigVMVariableNode() {variable=Pistol}.Value
   )
   > &RigUnit_ModifyTransforms(
    ItemToModify="((Item=(Type=Bone,Name=\"dyn_hairB_L_03\"),Transform=(Rotation=(X=0.0011,Y=-0.0479,Z=0.0237,W=0.9986),Translation=(X=0,Y=0,Z=0),Scale3D=(X=1,Y=1,Z=1))),(Item=(Type=Bone,Name=\"dyn_hairB_R_03\"),Transform=(Rotation=(X=0,Y=0,Z=0.0435,W=0.9991),Translation=(X=3.3 ...(truncated)"
    WeightMinimum=0
    Weight=@Multiply_5.Result
   )
   RigVMFunctionReturnNode(ExecuteContext=@RigVMFunction_Sequence.C)
   RigVMFunction_ControlFlowBranch = &RigVMFunction_ControlFlowBranch(ExecuteContext=(), Condition=RigVMVariableNode() {variable=EmotingBool}.Value)
   n2 = &RigUnit_SphericalPoseReader(
    ActiveRegionScaleFactors=(PositiveWidth=1.5,NegativeWidth=1.5,PositiveHeight=1.5,NegativeHeight=1)
    ActiveRegionSize=0.1
    Debug=(bDrawDebug=true,bDraw2D=False,bDrawLocalAxes=False,DebugScale=25,DebugSegments=20,DebugThickness=0.25)
    DriverAxis=(X=0,Y=1,Z=0)
    DriverItem=(Type=Bone,Name="head")
    FalloffRegionScaleFactors=(PositiveWidth=1,NegativeWidth=1,PositiveHeight=1,NegativeHeight=1)
    FalloffSize=0.3
    FlipHeightScaling=False
    FlipWidthScaling=False
    OptionalParentItem=(Type=Bone,Name="None")
    RotationOffset=(X=40,Y=0,Z=0)
    ExecutePin=@RigVMFunction_ControlFlowBranch.True
   )
   > &RigUnit_ModifyTransforms(
    ItemToModify="((Item=(Type=Bone,Name=\"dyn_hairB_L_01\"),Transform=(Rotation=(X=0,Y=0,Z=-0.0872,W=0.9962),Translation=(X=0,Y=0,Z=0),Scale3D=(X=1,Y=1,Z=1))),(Item=(Type=Bone,Name=\"dyn_hairB_L_02\"),Transform=(Rotation=(X=0,Y=0,Z=-0.0872,W=0.9962),Translation=(X=0,Y=0,Z=0),S ...(truncated)"
    WeightMinimum=0
    Weight=@n2.OutputParam
   )
   > &RigUnit_ModifyTransforms(
    ItemToModify="((Item=(Type=Bone,Name=\"dyn_hairB_C_01\"),Transform=(Rotation=(X=0,Y=0,Z=-0.0436,W=0.999),Translation=(X=0,Y=0,Z=0),Scale3D=(X=1,Y=1,Z=1))),(Item=(Type=Bone,Name=\"dyn_hairB_C_02\"),Transform=(Rotation=(X=0,Y=0,Z=-0.0436,W=0.999),Translation=(X=0,Y=0,Z=0),Sca ...(truncated)"
    WeightMinimum=0
    Weight=@n2.OutputParam
   )
   > n3 = &RigUnit_SphericalPoseReader(
    ActiveRegionScaleFactors=(PositiveWidth=1.5,NegativeWidth=1.5,PositiveHeight=1,NegativeHeight=1.5)
    ActiveRegionSize=0.1
    Debug=(bDrawDebug=false,bDraw2D=False,bDrawLocalAxes=False,DebugScale=25,DebugSegments=20,DebugThickness=0.25)
    DriverAxis=(X=0,Y=1,Z=0)
    DriverItem=(Type=Bone,Name="head")
    FalloffRegionScaleFactors=(PositiveWidth=1,NegativeWidth=1,PositiveHeight=1,NegativeHeight=1)
    FalloffSize=0.3
    FlipHeightScaling=False
    FlipWidthScaling=False
    OptionalParentItem=(Type=Bone,Name="None")
    RotationOffset=(X=180,Y=0,Z=0)
   )
   > &RigUnit_ModifyTransforms(
    ItemToModify="((Item=(Type=Bone,Name=\"dyn_hairB_R_01\"),Transform=(Rotation=(X=0,Y=0,Z=-0.0872,W=0.9962),Translation=(X=0,Y=0,Z=0),Scale3D=(X=1,Y=1,Z=1))),(Item=(Type=Bone,Name=\"dyn_hairB_R_02\"),Transform=(Rotation=(X=0,Y=0,Z=-0.0872,W=0.9962),Translation=(X=0,Y=0,Z=0),S ...(truncated)"
    WeightMinimum=0
    Weight=@n3.OutputParam
   )
   > &RigUnit_ModifyTransforms(
    ItemToModify="((Item=(Type=Bone,Name=\"dyn_hairB_C_01\"),Transform=(Rotation=(X=0,Y=0,Z=-0.0436,W=0.999),Translation=(X=0,Y=0,Z=0),Scale3D=(X=1,Y=1,Z=1))),(Item=(Type=Bone,Name=\"dyn_hairB_C_02\"),Transform=(Rotation=(X=0,Y=0,Z=-0.0436,W=0.999),Translation=(X=0,Y=0,Z=0),Sca ...(truncated)"
    WeightMinimum=0
    Weight=@n3.OutputParam
   )
  ]}
 RigVMFunctionLibrary {nodes [
   AlterTransform = RigVMCollapseNode(ExecuteContext=(), transform=())
   Constraints = RigVMCollapseNode(ExecuteContext=())
   Control Setup = RigVMCollapseNode(Execute Content=(), debug=False)
   DeltaRotation_X = RigVMCollapseNode(ExecuteContext=())
   DeltaRotation_Y = RigVMCollapseNode(ExecuteContext=())
   DeltaRotation_Z = RigVMCollapseNode(ExecuteContext=())
   ExtremeDelta = RigVMCollapseNode(ExecuteContext=())
   Hair Deformation System = RigVMCollapseNode(ExecuteContext=())
   Hair_Setup = RigVMCollapseNode(ExecuteContext=())
   Variables = RigVMCollapseNode(ExecuteContext=())
  ]}
 RigVMFunction_Sequence_1_ContainedGraph {nodes [
   n1 = RigVMFunctionEntryNode()
   RigVMFunction_Sequence = &RigVMFunction_Sequence(ExecuteContext=@n1.ExecuteContext)
   n2 = &RigVMFunction_Sequence(ExecuteContext=@RigVMFunction_Sequence.B)
   n3 = &RigVMFunction_Sequence(ExecuteContext=@n2.B)
   RigVMFunctionReturnNode(A=@RigVMFunction_Sequence.A, B=@n2.A, C=@n3.A, D=@n3.B)
  ]}
 RigVMFunction_Sequence_1_ContainedGraph {nodes [
   n1 = RigVMFunctionEntryNode()
   RigVMFunction_Sequence = &RigVMFunction_Sequence(ExecuteContext=@n1.ExecuteContext)
   n2 = &RigVMFunction_Sequence(ExecuteContext=@RigVMFunction_Sequence.B)
   RigVMFunctionReturnNode(A=@RigVMFunction_Sequence.A, B=@n2.A, C=@n2.B)
  ]}
 RigVMFunction_Sequence_1_ContainedGraph {nodes [
   n1 = RigVMFunctionEntryNode()
   RigVMFunction_Sequence = &RigVMFunction_Sequence(ExecuteContext=@n1.ExecuteContext)
   n2 = &RigVMFunction_Sequence(ExecuteContext=@RigVMFunction_Sequence.B)
   n3 = &RigVMFunction_Sequence(ExecuteContext=@n2.B)
   RigVMFunctionReturnNode(A=@RigVMFunction_Sequence.A, B=@n2.A, C=@n3.A, D=@n3.B)
  ]}
 RigVMFunction_Sequence_1_ContainedGraph {nodes [
   n1 = RigVMFunctionEntryNode()
   RigVMFunction_Sequence = &RigVMFunction_Sequence(ExecuteContext=@n1.ExecuteContext)
   n2 = &RigVMFunction_Sequence(ExecuteContext=@RigVMFunction_Sequence.B)
   n3 = &RigVMFunction_Sequence(ExecuteContext=@n2.B)
   RigVMFunctionReturnNode(A=@RigVMFunction_Sequence.A, B=@n2.A, C=@n3.A, D=@n3.B)
  ]}
 RigVMFunction_Sequence_1_ContainedGraph {nodes [
   n1 = RigVMFunctionEntryNode()
   RigVMFunction_Sequence = &RigVMFunction_Sequence(ExecuteContext=@n1.ExecuteContext)
   n2 = &RigVMFunction_Sequence(ExecuteContext=@RigVMFunction_Sequence.B)
   RigVMFunctionReturnNode(A=@RigVMFunction_Sequence.A, B=@n2.A, C=@n2.B)
  ]}
 RigVMModel {nodes [
   RigVMFunction_Sequence = RigVMAggregateNode(ExecuteContext=@RigUnit_BeginExecution.ExecutePin)
   Control Setup = RigVMFunctionReferenceNode(debug=false, Execute Content=@RigVMFunction_Sequence.A)
   Constraints = RigVMFunctionReferenceNode(ExecuteContext=@"Control Setup".Execute Content)
   RigUnit_BeginExecution = &RigUnit_BeginExecution()
   &RigUnit_ModifyTransforms(
    ItemToModify="((Item=(Type=Bone,Name=\"dyn_hairB_R_03\"),Transform=(Rotation=(X=0.0541,Y=-0.0018,Z=-0.034,W=0.998),Translation=(X=0.9659,Y=2.4977,Z=0.111),Scale3D=(X=1,Y=1,Z=1))),(Item=(Type=Bone,Name=\"dyn_hair_side_01_r\"),Transform=(Rotation=(X=0.1092,Y=-0.3193,Z=-0.01,W ...(truncated)"
    Weight=1
    WeightMinimum=0
    ExecutePin=@RigVMFunction_ControlFlowBranch.True
   )
   RigUnit_GetRelativeTransformForItem = &RigUnit_GetRelativeTransformForItem(Child=(Type=Bone,Name="pelvis"), Parent=(Type=Bone,Name="root"), bChildInitial=False, bParentInitial=False)
   n1 = &RigVMFunction_AnimEvalRichCurve(
    Curve="(EditorCurveData=(Keys=((InterpMode=RCIM_Cubic,TangentMode=RCTM_Break,TangentWeightMode=RCTWM_WeightedNone,Time=0,Value=0,ArriveTangent=0,ArriveTangentWeight=0,LeaveTangent=0,LeaveTangentWeight=0),(InterpMode=RCIM_Cubic,TangentMode=RCTM_Break,TangentWeight ...(truncated)"
    SourceMinimum=0
    TargetMinimum=0
    Value=&RigVMFunction_MathDoubleRemap(
     SourceMaximum=30
     SourceMinimum=60
     TargetMaximum=0
     TargetMinimum=1
     bClamp=False
     Value=@RigUnit_GetRelativeTransformForItem.RelativeTransform.Translation.Z
    ).Result
   )
   RigVMVariableNode(ExecuteContext=@Hair_Setup.ExecuteContext, Value=@n1.Result) {variable=CloseGroundRemap}
   Variables = RigVMFunctionReferenceNode(ExecuteContext=@RigVMFunction_Sequence.B)
   Hair Deformation System = RigVMFunctionReferenceNode(ExecuteContext=@RigVMFunction_Sequence.D)
   RigVMFunction_ControlFlowBranch = &RigVMFunction_ControlFlowBranch(Condition=RigVMVariableNode() {variable=FrontEnd}.Value, ExecuteContext=@"Hair Deformation System".ExecuteContext)
   Hair_Setup = RigVMFunctionReferenceNode(ExecuteContext=@RigVMFunction_Sequence.C)
  ]}
 RigVMModel Rig_1 {nodes [
   n1 = &RigUnit_PrepareForExecution()
   RigVMFunction_Sequence = RigVMAggregateNode(ExecuteContext=@n1.ExecutePin)
   Make_Array = RigVMDispatchNode(Values=((Type=Control,Name="dyn_hairB_C_01_ctrl"),((),Type=Control,Name="dyn_hairB_C_02_ctrl"),((),Type=Control,Name="dyn_hairB_C_03_ctrl"),((),Type=Control,Name="dyn_hairB_C_04_ctrl"),((),Type=Control,Name="dyn_hairB_C_05_ctrl"))) {template="DISPATCH_RigVMDispatch_ArrayMake(in Values,out Array)"}
   RigVMVariableNode(Value=@Make_Array.Array, ExecuteContext=@RigVMFunction_Sequence.A) {variable=HairBack_CTL}
   > RigVMVariableNode(Value=/Script/ControlRig.RigUnit_HierarchyGetPoseItemArray(ElementType=All, Initial=False, ItemsToGet=@Make_Array.Array) {is=struct RigVMUnitNode}.Pose) {variable=HairBack_Pose}
   Make_Array_1 = RigVMDispatchNode(Values=(((),Type=Control,Name="dyn_hairB_R_02_ctrl"),((),Type=Control,Name="dyn_hairB_R_03_ctrl"),((),Type=Control,Name="dyn_hairB_R_04_ctrl"),((),Type=Control,Name="dyn_hairB_R_05_ctrl"))) {template="DISPATCH_RigVMDispatch_ArrayMake(in Values,out Array)"}
   RigVMVariableNode(Value=@Make_Array_1.Array, ExecuteContext=@RigVMFunction_Sequence.B) {variable=HairBack_CTL_R}
   > RigVMVariableNode(Value=&RigUnit_HierarchyGetPoseItemArray(ElementType=All, Initial=False, ItemsToGet=@Make_Array_1.Array).Pose) {variable=HairBack_Pose_R}
   Make_Array_1_1 = RigVMDispatchNode(Values=(((),Type=Control,Name="dyn_hairB_L_02_ctrl"),((),Type=Control,Name="dyn_hairB_L_03_ctrl"),((),Type=Control,Name="dyn_hairB_L_04_ctrl"),((),Type=Control,Name="dyn_hairB_L_05_ctrl"))) {template="DISPATCH_RigVMDispatch_ArrayMake(in Values,out Array)"}
   RigVMVariableNode(Value=@Make_Array_1_1.Array, ExecuteContext=@RigVMFunction_Sequence.C) {variable=HairBack_CTL_L}
   > RigVMVariableNode(Value=&RigUnit_HierarchyGetPoseItemArray(ElementType=All, Initial=False, ItemsToGet=@Make_Array_1_1.Array).Pose) {variable=HairBack_Pose_L}
  ]}
]
hierarchy [
 root {loc=<0 0 0>, rot=<0 0 0 1>}
 pelvis {parent=root, loc=<0 -0.7321 92.4218>, rot=<0.0127 0.707 -0.0127 -0.707>}
 spine_01 {parent=pelvis, loc=<5.6002 -0.5128 0>, rot=<0 0 0.034 -0.9994>}
 spine_02 {parent=spine_01, loc=<8.0756 0.0139 0>, rot=<0 0 0.0238 -0.9997>}
 spine_03 {parent=spine_02, loc=<7.0395 0 0>, rot=<0 0 0.0149 -0.9999>}
 spine_04 {parent=spine_03, loc=<8.1023 0 0>, rot=<0 0 -0.0132 -0.9999>}
 spine_05 {parent=spine_04, loc=<10.0141 0 0>, rot=<0 0 -0.0015 -1>}
 clavicle_l {parent=spine_05, loc=<2.5856 -0.2866 -4.3777>, rot=<0.0465 -0.7289 -0.009 -0.6829>}
 upperarm_l {parent=clavicle_l, loc=<18.2264 0.1342 -0.001>, rot=<0.0253 -0.2693 0.0013 -0.9627>}
 base_armor_shoulder_l {parent=upperarm_l, loc=<0.0285 -2.5392 8.6797>, rot=<0 0 -0.0406 -0.9992>}
 armor_shoulder_l {parent=base_armor_shoulder_l, loc=<0 0 0>, rot=<0 0 0 1>}
 lowerarm_l {parent=upperarm_l, loc=<29.0071 0.1686 -1.239>, rot=<0.0391 -0.0099 -0.2453 -0.9686>}
 hand_l {parent=lowerarm_l, loc=<25.8607 0.222 1.0274>, rot=<-0.7193 -0.0988 0.1193 -0.6773>}
 thumb_01_l {parent=hand_l, loc=<1.3118 -0.3191 -2.459>, rot=<0.6546 -0.4042 0.0664 -0.6354>}
 thumb_02_l {parent=thumb_01_l, loc=<5.9336 -0.0448 0.0401>, rot=<-0.0037 0.0213 0.298 -0.9543>}
 thumb_03_l {parent=thumb_02_l, loc=<2.2223 0.3682 -0.0907>, rot=<0.0118 -0.013 0.0066 -0.9998>}
 index_metacarpal_l {parent=hand_l, loc=<1.6535 0.0183 -2.96>, rot=<0.0158 -0.0124 -0.0173 -0.9996>}
 index_01_l {parent=index_metacarpal_l, loc=<8.3008 0.0357 -0.0524>, rot=<-0.0044 0.0485 0.0924 -0.9945>}
 index_02_l {parent=index_01_l, loc=<3.3587 -0.3004 0.0452>, rot=<-0.002 0.0027 0.3225 -0.9466>}
 index_03_l {parent=index_02_l, loc=<2.5578 -0.123 -0.0252>, rot=<-0.002 0.0067 0.2558 -0.9667>}
 middle_metacarpal_l {parent=hand_l, loc=<2.1856 -0.1725 -0.5984>, rot=<0.006 0.0168 -0.0329 -0.9993>}
 middle_01_l {parent=middle_metacarpal_l, loc=<7.3275 -0.2447 0.0129>, rot=<-0.0208 0.0042 0.0858 -0.9961>}
 middle_02_l {parent=middle_01_l, loc=<4.0123 0.1401 0.1175>, rot=<-0.0218 0.0258 0.3512 -0.9357>}
 middle_03_l {parent=middle_02_l, loc=<2.6077 0.0534 0.1373>, rot=<0.0082 0.0188 0.1708 -0.9851>}
 ring_metacarpal_l {parent=hand_l, loc=<2.2074 -0.3625 1.3638>, rot=<-0.0284 0.0722 -0.0458 -0.9959>}
 ring_01_l {parent=ring_metacarpal_l, loc=<7.0295 -0.6197 -0.1446>, rot=<-0.0489 -0.0252 0.1892 -0.9804>}
 ring_02_l {parent=ring_01_l, loc=<3.6174 0.4647 -0.1027>, rot=<0 0 0.2806 -0.9598>}
 ring_03_l {parent=ring_02_l, loc=<2.4152 0.114 0.0199>, rot=<0 0 0.1189 -0.9929>}
 pinky_metacarpal_l {parent=hand_l, loc=<1.7616 -0.9753 3.3814>, rot=<-0.0397 0.0814 -0.069 -0.9935>}
 pinky_01_l {parent=pinky_metacarpal_l, loc=<7.1709 -0.4831 -0.1094>, rot=<-0.0654 -0.0035 0.2242 -0.9723>}
 pinky_02_l {parent=pinky_01_l, loc=<2.8274 0.0893 -0.1043>, rot=<0.0037 0.0011 0.2854 -0.9584>}
 pinky_03_l {parent=pinky_02_l, loc=<2.2046 0.0307 -0.0078>, rot=<-0.0002 0.0098 0.1445 -0.9895>}
 deform_wrist_dn_base_l {parent=hand_l, loc=<0 -3.4173 0>, rot=<0 0 0 -1>}
 deform_wrist_dn_l {parent=deform_wrist_dn_base_l, loc=<0 0 0>, rot=<0 0 0 1>}
 deform_wrist_up_base_l {parent=hand_l, loc=<0 2.7805 0>, rot=<0 0 0 -1>}
 deform_wrist_up_l {parent=deform_wrist_up_base_l, loc=<0 0 0>, rot=<0 0 0 1>}
 lowerarm_twist_02_l {parent=lowerarm_l, loc=<9.6199 0 0>, rot=<0 0 0 -1>}
 lowerarm_twist_01_l {parent=lowerarm_l, loc=<17.8549 0 0>, rot=<0 0 0 -1>}
 deform_elbow_in_base_l {parent=lowerarm_l, loc=<0.2112 5.645 0.8274>, rot=<0 0 0 1>}
 deform_elbow_in_l {parent=deform_elbow_in_base_l, loc=<0 0 0>, rot=<0 0 0 1>}
 deform_elbow_out_base_l {parent=lowerarm_l, loc=<0 -7.5129 0>, rot=<0 0 0 1>}
 deform_elbow_out_l {parent=deform_elbow_out_base_l, loc=<0 0 0>, rot=<0 0 0 1>}
 upperarm_twist_01_l {parent=upperarm_l, loc=<7.9796 0.0843 -0.6195>, rot=<0 0 0 -1>}
 deform_upperarm_bk_base_l {parent=upperarm_twist_01_l, loc=<-6.8983 -13.3826 -1.3258>, rot=<0 0 0 -1>}
 deform_upperarm_bk_l {parent=deform_upperarm_bk_base_l, loc=<0 0 0>, rot=<0 0 0 1>}
 deform_upperarm_fr_base_l {parent=upperarm_twist_01_l, loc=<-8.3487 10.7117 -2.1455>, rot=<0 0 0 -1>}
 deform_upperarm_fr_l {parent=deform_upperarm_fr_base_l, loc=<0 0 0>, rot=<0 0 0 1>}
 deform_upperarm_up_base_l {parent=upperarm_twist_01_l, loc=<-11.2012 -1.2234 8.7784>, rot=<0 0 0 -1>}
 deform_upperarm_up_l {parent=deform_upperarm_up_base_l, loc=<0 0 0>, rot=<0 0 0 1>}
 upperarm_twist_02_l {parent=upperarm_l, loc=<19.5059 0.0843 -0.6195>, rot=<0 0 0 -1>}
 deform_bicep_base_l {parent=upperarm_twist_02_l, loc=<0 8.0494 0>, rot=<0 0 0 1>}
 deform_bicep_l {parent=deform_bicep_base_l, loc=<0 0 0>, rot=<0 0 0 1>}
 dyn_shoulderPad_l_01 {parent=upperarm_l, loc=<0 0 0>, rot=<-0.0177 0.302 -0.0245 -0.9528>}
 upperarm_l_null {parent=clavicle_l, type=Null, loc=<18.2264 0.1342 -0.001>, rot=<0.0253 -0.2693 0.0013 -0.9627>}
 neck_01 {parent=spine_05, loc=<10.815 1.9196 0>, rot=<0 0 -0.1864 -0.9825>}
 neck_02 {parent=neck_01, loc=<4.7347 0.3045 0>, rot=<0 0 0.0363 -0.9993>}
 head {parent=neck_02, loc=<5.7632 0.5552 0>, rot=<0 0 0.115 -0.9934>}
 base_hatMain {parent=head, loc=<9.5908 -1.1364 -0.0001>, rot=<0.0029 -0.7071 -0.0029 -0.7071>}
 dyn_hatMain {parent=base_hatMain, loc=<0 0 0>, rot=<0 0 0 1>}
 base_hairPieceA {parent=dyn_hatMain, loc=<0.0057 18.965 -1.104>, rot=<-0.4987 -0.4986 0.5014 -0.5012>}
 dyn_hairPieceA_01 {parent=base_hairPieceA, loc=<0 0 0>, rot=<0 0 0 1>}
 base_hairPieceB {parent=dyn_hatMain, loc=<9.8169 16.4265 -0.9392>, rot=<-0.6134 -0.348 0.6166 -0.3499>}
 dyn_hairPieceB_01 {parent=base_hairPieceB, loc=<0 0 0>, rot=<0 0 0 1>}
 base_hairPieceC {parent=dyn_hatMain, loc=<16.8201 9.3167 -0.5583>, rot=<-0.6828 -0.1765 0.6864 -0.1774>}
 dyn_hairPieceC_01 {parent=base_hairPieceC, loc=<0 0 0>, rot=<0 0 0 1>}
 base_hairPieceD {parent=dyn_hatMain, loc=<19.3474 -0.3161 -0.1048>, rot=<-0.7052 0.0058 0.709 0.0058>}
 dyn_hairPieceD_01 {parent=base_hairPieceD, loc=<0 0 0>, rot=<0 0 0 1>}
 base_hairPieceE {parent=dyn_hatMain, loc=<16.611 -10.0692 0.4052>, rot=<-0.1898 -0.6792 0.1908 -0.6828>}
 dyn_hairPieceE_01 {parent=base_hairPieceE, loc=<0 0 0>, rot=<0 0 0 1>}
 base_hairPieceF {parent=dyn_hatMain, loc=<9.8007 -17.1529 0.6778>, rot=<-0.354 -0.61 0.3559 -0.6132>}
 dyn_hairPieceF_01 {parent=base_hairPieceF, loc=<0 0 0>, rot=<0 0 0 1>}
 base_hairPieceG {parent=dyn_hatMain, loc=<-0.0391 -19.6466 1.0285>, rot=<-0.4992 -0.4982 0.5018 -0.5008>}
 dyn_hairPieceG_01 {parent=base_hairPieceG, loc=<0 0 0>, rot=<0 0 0 1>}
 base_hairPieceH {parent=dyn_hatMain, loc=<-9.7996 -17.1287 0.6727>, rot=<-0.6101 -0.3538 0.6133 -0.3557>}
 dyn_hairPieceH_01 {parent=base_hairPieceH, loc=<0 0 0>, rot=<0 0 0 1>}
 base_hairPieceI {parent=dyn_hatMain, loc=<-16.6029 -10.0628 0.4011>, rot=<-0.6792 -0.1898 0.6828 -0.1908>}
 dyn_hairPieceI_01 {parent=base_hairPieceI, loc=<0 0 0>, rot=<0 0 0 1>}
 base_hairPieceJ {parent=dyn_hatMain, loc=<-19.3473 -0.3206 -0.1026>, rot=<0.0058 -0.7052 -0.0059 -0.709>}
 dyn_hairPieceJ_01 {parent=base_hairPieceJ, loc=<0 0 0>, rot=<0 0 0 1>}
 base_hairPieceK {parent=dyn_hatMain, loc=<-16.8502 9.2679 -0.5791>, rot=<-0.1755 -0.6831 0.1764 -0.6867>}
 dyn_hairPieceK_01 {parent=base_hairPieceK, loc=<0 0 0>, rot=<0 0 0 1>}
 base_hairPieceL {parent=dyn_hatMain, loc=<-9.8411 16.4229 -0.9357>, rot=<-0.3476 -0.6136 0.3495 -0.6169>}
 dyn_hairPieceL_01 {parent=base_hairPieceL, loc=<0 0 0>, rot=<0 0 0 1>}
 base_hair_side_l {parent=head, loc=<-1.269 -0.2129 -7.0831>, rot=<0.2285 0.6572 0.7179 -0.0203>}
 base_hair_side_01_l {parent=base_hair_side_l, loc=<0 0 0>, rot=<0 0 0 1>}
 base_hair_side_02_l {parent=base_hair_side_01_l, loc=<6.1321 0 0>, rot=<0 0 0 1>}
 dyn_hair_side_01_l {parent=base_hair_side_l, loc=<0 0 0>, rot=<-0.6088 0 0 -0.7934>}
 dyn_hair_side_02_l {parent=dyn_hair_side_01_l, loc=<6.1321 0 0>, rot=<0 0 0 1>}
 base_hair_front_l {parent=head, loc=<4.67 4.8701 -6.7876>, rot=<0.0041 -0.9975 0.0003 0.0702>}
 base_hair_front_01_l {parent=base_hair_front_l, loc=<0 0 0>, rot=<0 0 0 1>}
 base_hair_front_02_l {parent=base_hair_front_01_l, loc=<3.9818 0 0.0838>, rot=<0 -0.0497 0 -0.9988>}
 base_hair_front_03_l {parent=base_hair_front_02_l, loc=<4.7269 0 0>, rot=<0 0 0 1>}
 dyn_hair_front_01_l {parent=base_hair_front_l, loc=<0 0 0>, rot=<0 0 0 1>}
 dyn_hair_front_02_l {parent=dyn_hair_front_01_l, loc=<3.9818 0 0.0838>, rot=<0 -0.0497 0 -0.9988>}
 dyn_hair_front_03_l {parent=dyn_hair_front_02_l, loc=<4.7269 0 0>, rot=<0 0 0 1>}
 base_hair_back {parent=head, loc=<-1.7195 -12.378 0>, rot=<-0.2525 0.6605 0.6605 -0.2525>}
 base_hair_back_01 {parent=base_hair_back, loc=<0 0 0>, rot=<0 0 0 1>}
 base_hair_back_02 {parent=base_hair_back_01, loc=<9.5458 0 0>, rot=<0 0.2174 0 -0.9761>}
 base_hair_back_03 {parent=base_hair_back_02, loc=<11.5386 0 0>, rot=<0 0 0 1>}
 dyn_hair_back_01 {parent=base_hair_back, loc=<0 0 0>, rot=<0 0 0 1>}
 dyn_hair_back_02 {parent=dyn_hair_back_01, loc=<6.3184 0 -0.0001>, rot=<0 0 0 1>}
 dyn_hair_back_03 {parent=dyn_hair_back_02, loc=<7.2918 0 0.0001>, rot=<0 0.2115 0 -0.9774>}
 dyn_hair_back_04 {parent=dyn_hair_back_03, loc=<9.0717 0 0>, rot=<0 0 0 1>}
 base_hair_side_r {parent=head, loc=<-1.2695 -0.2129 7.0831>, rot=<0.2285 0.6572 0.7179 -0.0203>}
 base_hair_side_01_r {parent=base_hair_side_r, loc=<0 0 0>, rot=<0 0 0 -1>}
 base_hair_side_02_r {parent=base_hair_side_01_r, loc=<-6.1325 0.0001 0.0002>, rot=<0 0 0 -1>}
 dyn_hair_side_01_r {parent=base_hair_side_r, loc=<0 0 0>, rot=<-0.6088 0 0 -0.7934>}
 dyn_hair_side_02_r {parent=dyn_hair_side_01_r, loc=<4.588 -1.1851 3.8908>, rot=<0 0 0 1>}
 base_hair_front_r {parent=head, loc=<4.6697 4.8701 6.7876>, rot=<0.0041 -0.9975 0.0003 0.0702>}
 base_hair_front_01_r {parent=base_hair_front_r, loc=<0 0 0>, rot=<0 0 0 1>}
 base_hair_front_02_r {parent=base_hair_front_01_r, loc=<-3.9813 0 -0.0837>, rot=<0 -0.0497 0 -0.9988>}
 base_hair_front_03_r {parent=base_hair_front_02_r, loc=<-4.727 0 0>, rot=<0 0 0 -1>}
 dyn_hair_front_01_r {parent=base_hair_front_r, loc=<0 0 0>, rot=<0 0 0 1>}
 dyn_hair_front_02_r {parent=dyn_hair_front_01_r, loc=<3.8484 0 1.0246>, rot=<0 -0.0497 0 -0.9988>}
 dyn_hair_front_03_r {parent=dyn_hair_front_02_r, loc=<4.1923 0 2.1827>, rot=<0 0 0 1>}
 dyn_ropeHat_l {parent=head, loc=<-2.0882 2.1786 -6.9485>, rot=<0.0029 -0.7071 -0.0029 -0.7071>}
 dyn_ropeHat_r {parent=head, loc=<-2.0883 2.1786 6.9485>, rot=<0.0029 -0.7071 -0.0029 -0.7071>}
 base_chinRope_m {parent=head, loc=<-6.0714 4.0373 0.0158>, rot=<-0.0029 0.7071 0.7071 -0.0029>}
 dyn_chinRope_m {parent=base_chinRope_m, loc=<0 0 0>, rot=<0 0 0 1>}
 base_hairB_R {parent=head, loc=<6.9178 -8.738 6.6014>, rot=<0.1412 -0.99 0 0>}
 dyn_hairB_R_01 {parent=base_hairB_R, loc=<0 0 0>, rot=<0 0 0 1>}
 dyn_hairB_R_02 {parent=dyn_hairB_R_01, loc=<6.7665 0.9118 0>, rot=<0 0 0.1608 -0.987>}
 dyn_hairB_R_03 {parent=dyn_hairB_R_02, loc=<7.2405 0 0>, rot=<0 0 0.0357 -0.9994>}
 dyn_hairB_R_04 {parent=dyn_hairB_R_03, loc=<7.9383 0 0>, rot=<0 0 -0.1124 -0.9937>}
 dyn_hairB_R_05 {parent=dyn_hairB_R_04, loc=<7.4485 0 0>, rot=<0 0 0 1>}
 base_hairB_L {parent=head, loc=<6.9191 -8.9557 -7.019>, rot=<0.1412 -0.99 0 0>}
 dyn_hairB_L_01 {parent=base_hairB_L, loc=<0 0 0>, rot=<0 0 0 1>}
 dyn_hairB_L_02 {parent=dyn_hairB_L_01, loc=<6.6782 0.2665 0>, rot=<0 0 0.1785 -0.9839>}
 dyn_hairB_L_03 {parent=dyn_hairB_L_02, loc=<7.2954 0 0>, rot=<0 0 -0.0403 -0.9992>}
 dyn_hairB_L_04 {parent=dyn_hairB_L_03, loc=<7.4829 0 0>, rot=<0 0 -0.0977 -0.9952>}
 dyn_hairB_L_05 {parent=dyn_hairB_L_04, loc=<7.4822 0 0>, rot=<0 0 0 1>}
 base_hairB_C {parent=head, loc=<6.9219 -12.018 0>, rot=<0.0655 -0.9978 0 0>}
 dyn_hairB_C_01 {parent=base_hairB_C, loc=<0 0 0>, rot=<0 0 0 1>}
 dyn_hairB_C_02 {parent=dyn_hairB_C_01, loc=<6.6158 -0.1196 0>, rot=<0 0 0.208 -0.9781>}
 dyn_hairB_C_03 {parent=dyn_hairB_C_02, loc=<7.061 0 0>, rot=<0 0 -0.0168 -0.9999>}
 dyn_hairB_C_04 {parent=dyn_hairB_C_03, loc=<7.7305 0 0>, rot=<0 0 -0.1322 -0.9912>}
 dyn_hairB_C_05 {parent=dyn_hairB_C_04, loc=<7.2512 0 0>, rot=<0 0 0 1>}
 base_hoodTip_m {parent=head, loc=<12.6843 -0.0728 -0.0002>, rot=<0.5484 -0.446 0.4044 -0.5804>}
 dyn_hoodTip_m {parent=base_hoodTip_m, loc=<0 0 0>, rot=<0 0 0 1>}
 base_hoodback_m {parent=head, loc=<11.1918 -10.8833 0.0002>, rot=<0.0029 -0.7071 0.7071 -0.0029>}
 dyn_hoodback_m {parent=base_hoodback_m, loc=<0 0 0>, rot=<0 0 0 -1>}
 base_hood_l {parent=neck_02, loc=<5.1509 0.9515 -3.8042>, rot=<0.0661 -0.9976 -0.0015 -0.0221>}
 dyn_hood_l {parent=base_hood_l, loc=<0 0 0>, rot=<0 0 0 1>}
 base_hood_r {parent=neck_02, loc=<5.151 0.9515 3.8042>, rot=<0.0221 0.0015 0.9976 -0.0661>}
 dyn_hood_r {parent=base_hood_r, loc=<0 0 0>, rot=<0 0 0 -1>}
 head_null {parent=neck_02, type=Null, loc=<5.7632 0.5552 0>, rot=<0 0 0.115 -0.9934>}
 clavicle_r {parent=spine_05, loc=<2.5855 -0.2866 4.3777>, rot=<0.7289 0.0465 -0.6829 0.009>}
 upperarm_r {parent=clavicle_r, loc=<-18.2264 -0.1342 0.001>, rot=<0.0253 -0.2693 0.0013 -0.9627>}
 base_armor_shoulder_r {parent=upperarm_r, loc=<-0.0286 2.5392 -8.6796>, rot=<0 0 -0.0406 -0.9992>}
 armor_shoulder_r {parent=base_armor_shoulder_r, loc=<0 0 0>, rot=<0 0 0 1>}
 lowerarm_r {parent=upperarm_r, loc=<-29.0071 -0.1686 1.239>, rot=<0.0391 -0.0099 -0.2453 -0.9686>}
 hand_r {parent=lowerarm_r, loc=<-25.8607 -0.222 -1.0274>, rot=<-0.7193 -0.0988 0.1193 -0.6773>}
 thumb_01_r {parent=hand_r, loc=<-1.3118 0.3191 2.459>, rot=<0.6546 -0.4042 0.0664 -0.6354>}
 thumb_02_r {parent=thumb_01_r, loc=<-5.9336 0.0448 -0.0401>, rot=<-0.0037 0.0213 0.298 -0.9543>}
 thumb_03_r {parent=thumb_02_r, loc=<-2.2223 -0.3682 0.0907>, rot=<0.0118 -0.013 0.0066 -0.9998>}
 index_metacarpal_r {parent=hand_r, loc=<-1.6535 -0.0183 2.96>, rot=<0.0158 -0.0124 -0.0173 -0.9996>}
 index_01_r {parent=index_metacarpal_r, loc=<-8.3008 -0.0357 0.0524>, rot=<-0.0044 0.0485 0.0924 -0.9945>}
 index_02_r {parent=index_01_r, loc=<-3.3587 0.3004 -0.0452>, rot=<-0.002 0.0027 0.3225 -0.9466>}
 index_03_r {parent=index_02_r, loc=<-2.5578 0.123 0.0252>, rot=<-0.002 0.0067 0.2558 -0.9667>}
 middle_metacarpal_r {parent=hand_r, loc=<-2.1856 0.1725 0.5984>, rot=<0.006 0.0168 -0.0329 -0.9993>}
 middle_01_r {parent=middle_metacarpal_r, loc=<-7.3275 0.2447 -0.0129>, rot=<-0.0208 0.0042 0.0858 -0.9961>}
 middle_02_r {parent=middle_01_r, loc=<-4.0123 -0.1401 -0.1175>, rot=<-0.0218 0.0258 0.3512 -0.9357>}
 middle_03_r {parent=middle_02_r, loc=<-2.6077 -0.0534 -0.1373>, rot=<0.0082 0.0188 0.1708 -0.9851>}
 ring_metacarpal_r {parent=hand_r, loc=<-2.2074 0.3625 -1.3638>, rot=<-0.0284 0.0722 -0.0458 -0.9959>}
 ring_01_r {parent=ring_metacarpal_r, loc=<-7.0295 0.6197 0.1446>, rot=<-0.0489 -0.0252 0.1892 -0.9804>}
 ring_02_r {parent=ring_01_r, loc=<-3.6174 -0.4647 0.1027>, rot=<0 0 0.2806 -0.9598>}
 ring_03_r {parent=ring_02_r, loc=<-2.4152 -0.114 -0.0199>, rot=<0 0 0.1189 -0.9929>}
 pinky_metacarpal_r {parent=hand_r, loc=<-1.7616 0.9753 -3.3814>, rot=<-0.0397 0.0814 -0.069 -0.9935>}
 pinky_01_r {parent=pinky_metacarpal_r, loc=<-7.1709 0.4831 0.1094>, rot=<-0.0654 -0.0035 0.2242 -0.9723>}
 pinky_02_r {parent=pinky_01_r, loc=<-2.8274 -0.0893 0.1043>, rot=<0.0037 0.0011 0.2854 -0.9584>}
 pinky_03_r {parent=pinky_02_r, loc=<-2.2046 -0.0307 0.0078>, rot=<-0.0002 0.0098 0.1445 -0.9895>}
 deform_wrist_dn_base_r {parent=hand_r, loc=<0 3.4172 0>, rot=<0 0 0 -1>}
 deform_wrist_dn_r {parent=deform_wrist_dn_base_r, loc=<0 0 0>, rot=<0 0 0 -1>}
 deform_wrist_up_base_r {parent=hand_r, loc=<-0.0001 -2.7804 0>, rot=<0 0 0 -1>}
 deform_wrist_up_r {parent=deform_wrist_up_base_r, loc=<0 0 0>, rot=<0 0 0 -1>}
 lowerarm_twist_02_r {parent=lowerarm_r, loc=<-9.6199 0 0>, rot=<0 0 0 -1>}
 lowerarm_twist_01_r {parent=lowerarm_r, loc=<-17.8549 0 0>, rot=<0 0 0 -1>}
 deform_elbow_in_base_r {parent=lowerarm_r, loc=<-0.2114 -5.6449 -0.827>, rot=<0 0 0 -1>}
 deform_elbow_in_r {parent=deform_elbow_in_base_r, loc=<0 0 0>, rot=<0 0 0 1>}
 deform_elbow_out_base_r {parent=lowerarm_r, loc=<-0.0001 7.5129 0.0002>, rot=<0 0 0 -1>}
 deform_elbow_out_r {parent=deform_elbow_out_base_r, loc=<0 0 0>, rot=<0 0 0 1>}
 upperarm_twist_01_r {parent=upperarm_r, loc=<-7.9796 -0.0843 0.6195>, rot=<0 0 0 -1>}
 deform_upperarm_bk_base_r {parent=upperarm_twist_01_r, loc=<6.8985 13.3826 1.3257>, rot=<0 0 0 -1>}
 deform_upperarm_bk_r {parent=deform_upperarm_bk_base_r, loc=<0 0 0>, rot=<0 0 0 1>}
 deform_upperarm_fr_base_r {parent=upperarm_twist_01_r, loc=<8.3485 -10.7117 2.1456>, rot=<0 0 0 -1>}
 deform_upperarm_fr_r {parent=deform_upperarm_fr_base_r, loc=<0 0 0>, rot=<0 0 0 1>}
 deform_upperarm_up_base_r {parent=upperarm_twist_01_r, loc=<11.2015 1.2234 -8.7788>, rot=<0 0 0 -1>}
 deform_upperarm_up_r {parent=deform_upperarm_up_base_r, loc=<0 0 0>, rot=<0 0 0 1>}
 upperarm_twist_02_r {parent=upperarm_r, loc=<-19.5059 -0.0843 0.6195>, rot=<0 0 0 -1>}
 deform_bicep_base_r {parent=upperarm_twist_02_r, loc=<0.0002 -8.0494 -0.0003>, rot=<0 0 0 1>}
 deform_bicep_r {parent=deform_bicep_base_r, loc=<0 0 0>, rot=<0 0 0 1>}
 dyn_shoulderPad_r_01 {parent=upperarm_r, loc=<0 0 0>, rot=<0.9528 0.0245 0.302 -0.0177>}
 attach_cape {parent=spine_05, loc=<9.1509 -13.3783 0>, rot=<0.0616 -0.7044 -0.0616 -0.7044>}
 spine_05_null {parent=spine_05, type=Null, loc=<0 0 0>, rot=<0 0 0 1>}
 deform_pec_aim_base_l {parent=spine_04, loc=<5.0227 13.6719 -2.6508>, rot=<0.0672 -0.5201 0.3228 -0.7879>}
 deform_pec_aim_l {parent=deform_pec_aim_base_l, loc=<0 0 0>, rot=<0 0 0 1>}
 deform_pec_base_l {parent=deform_pec_aim_l, loc=<16.3279 0 0>, rot=<0 0 0 1>}
 deform_pec_l {parent=deform_pec_base_l, loc=<0 0 0>, rot=<0 0 0 1>}
 deform_pec_aim_base_r {parent=spine_04, loc=<5.0225 13.6718 2.6508>, rot=<-0.5201 -0.0672 0.7879 0.3228>}
 deform_pec_aim_r {parent=deform_pec_aim_base_r, loc=<0 0 0>, rot=<0 0 0 1>}
 deform_pec_base_r {parent=deform_pec_aim_r, loc=<-16.3279 0 0>, rot=<0 0 0 -1>}
 deform_pec_r {parent=deform_pec_base_r, loc=<0 0 0.0002>, rot=<0 0 0 -1>}
 deform_lat_l_base {parent=spine_03, loc=<2.5561 -10.3716 -17.6918>, rot=<0 0 0 -1>}
 deform_lat_l {parent=deform_lat_l_base, loc=<0 0 0>, rot=<0 0 0 1>}
 deform_lat_r_base {parent=spine_03, loc=<2.5562 -10.3716 17.6918>, rot=<0 0 1 0>}
 deform_lat_r {parent=deform_lat_r_base, loc=<0 0 0>, rot=<0 0 0 -1>}
 thigh_l {parent=pelvis, loc=<-1.8716 -0.0109 -11.9936>, rot=<0.0012 0.0572 0.0155 -0.9982>}
 calf_l {parent=thigh_l, loc=<-41.9178 -0.2246 0.0926>, rot=<0.0006 -0.0161 -0.0391 -0.9991>}
 foot_l {parent=calf_l, loc=<-38.7057 1.2438 0.2452>, rot=<0.0975 -0.0474 0.0317 -0.9936>}
 ball_l {parent=foot_l, loc=<-6.1614 13.1469 0.6807>, rot=<0.0001 -0.0001 -0.7063 -0.7079>}
 calf_twist_02_l {parent=calf_l, loc=<-15.9268 -0.0967 0.0397>, rot=<0.0237 0.0095 -0.0035 -0.9997>}
 calf_twist_01_l {parent=calf_l, loc=<-32.9335 -0.0967 0.0397>, rot=<0.0237 0.0095 -0.0035 -0.9997>}
 deform_knee_in_base_l {parent=calf_l, loc=<1.298 -8.0356 -0.0496>, rot=<0 0 0 1>}
 deform_knee_in_l {parent=deform_knee_in_base_l, loc=<0 0 0>, rot=<0 0 0 1>}
 deform_knee_out_base_l {parent=calf_l, loc=<0 6.8739 0>, rot=<0 0 0 -1>}
 deform_knee_out_l {parent=deform_knee_out_base_l, loc=<0 0 0>, rot=<0 0 0 1>}
 thigh_twist_01_l {parent=thigh_l, loc=<-21.6303 0 0>, rot=<0.0013 0 0 -1>}
 deform_groin_base_l {parent=thigh_l, loc=<-2.1908 8.5661 -1.7442>, rot=<0 0 0 -1>}
 deform_groin_l {parent=deform_groin_base_l, loc=<0 0 0>, rot=<0 0 0 1>}
 thigh_r {parent=pelvis, loc=<-1.8716 -0.0109 11.9936>, rot=<0.0572 -0.0012 0.9982 0.0155>}
 calf_r {parent=thigh_r, loc=<41.9178 0.2246 -0.0926>, rot=<0.0006 -0.0161 -0.0391 -0.9991>}
 foot_r {parent=calf_r, loc=<38.7057 -1.2438 -0.2452>, rot=<0.0975 -0.0474 0.0317 -0.9936>}
 ball_r {parent=foot_r, loc=<6.1614 -13.1469 -0.6807>, rot=<0.0001 -0.0001 -0.7063 -0.7079>}
 calf_twist_02_r {parent=calf_r, loc=<15.9268 0.0967 -0.0397>, rot=<0.0237 0.0095 -0.0035 -0.9997>}
 calf_twist_01_r {parent=calf_r, loc=<32.9335 0.0967 -0.0397>, rot=<0.0237 0.0095 -0.0035 -0.9997>}
 deform_knee_in_base_r {parent=calf_r, loc=<-1.2981 8.0356 0.0496>, rot=<0 0 0 -1>}
 deform_knee_in_r {parent=deform_knee_in_base_r, loc=<0 0 0>, rot=<0 0 0 1>}
 deform_knee_out_base_r {parent=calf_r, loc=<0 -6.8739 0>, rot=<0 0 0 -1>}
 deform_knee_out_r {parent=deform_knee_out_base_r, loc=<0 0 0>, rot=<0 0 0 1>}
 thigh_twist_01_r {parent=thigh_r, loc=<21.6303 0 0>, rot=<0.0013 0 0 -1>}
 deform_groin_base_r {parent=thigh_r, loc=<2.1908 -8.5661 1.7442>, rot=<0 0 0 -1>}
 deform_groin_r {parent=deform_groin_base_r, loc=<0 0 0>, rot=<0 0 0 -1>}
 deform_glute_base_l {parent=pelvis, loc=<-10.1877 -11.6217 -12.899>, rot=<0.0275 0.0501 -0.4472 -0.8926>}
 deform_glute_l {parent=deform_glute_base_l, loc=<0 0 0>, rot=<0 0 0 1>}
 deform_glute_base_r {parent=pelvis, loc=<-10.1877 -11.6218 12.899>, rot=<0.0501 -0.0275 0.8926 -0.4472>}
 deform_glute_r {parent=deform_glute_base_r, loc=<0 0 0>, rot=<0 0 0 1>}
 head_ctrl {controlType=EULER_TRANSFORM, type=Control, loc=<0 0 0>, rot=<0 0 0 1>}
 dyn_hairB_C_01_ctrl_null {type=Null, loc=<0 -11.1192 158.8565>, rot=<0.0434 -0.7058 0.0434 -0.7058>}
 dyn_hairB_C_01_ctrl {controlType=EULER_TRANSFORM, parent=dyn_hairB_C_01_ctrl_null, type=Control, loc=<156.2949 30.5112 0>, rot=<-0.0434 0.7058 -0.0434 -0.7058>}
 dyn_hairB_C_02_ctrl_null {parent=dyn_hairB_C_01_ctrl, type=Null, loc=<6.6158 -0.1196 0>, rot=<0 0 0.208 -0.9781>}
 dyn_hairB_C_02_ctrl {controlType=EULER_TRANSFORM, parent=dyn_hairB_C_02_ctrl_null, type=Control, loc=<-6.0921 -2.5826 0>, rot=<0 0 -0.208 -0.9781>}
 dyn_hairB_C_03_ctrl_null {parent=dyn_hairB_C_02_ctrl, type=Null, loc=<7.061 0 0>, rot=<0 0 -0.0168 -0.9999>}
 dyn_hairB_C_03_ctrl {controlType=EULER_TRANSFORM, parent=dyn_hairB_C_03_ctrl_null, type=Control, loc=<-7.057 0.2372 0>, rot=<0 0 0.0168 -0.9999>}
 dyn_hairB_C_04_ctrl_null {parent=dyn_hairB_C_03_ctrl, type=Null, loc=<7.7305 0 0>, rot=<0 0 -0.1322 -0.9912>}
 dyn_hairB_C_04_ctrl {controlType=EULER_TRANSFORM, parent=dyn_hairB_C_04_ctrl_null, type=Control, loc=<-7.4604 2.0257 0>, rot=<0 0 0.1322 -0.9912>}
 dyn_hairB_C_05_ctrl_null {parent=dyn_hairB_C_04_ctrl, type=Null, loc=<7.2512 0 0>, rot=<0 0 0 1>}
 dyn_hairB_C_05_ctrl {controlType=EULER_TRANSFORM, parent=dyn_hairB_C_05_ctrl_null, type=Control, loc=<-7.2512 0 0>, rot=<0 0 0 1>}
 dyn_hairB_R_01_ctrl_null {type=Null, loc=<-6.6014 -7.8392 158.8795>, rot=<0.0969 -0.7004 0.0969 -0.7004>}
 dyn_hairB_R_01_ctrl {
  controlType=EULER_TRANSFORM
  parent=dyn_hairB_R_01_ctrl_null
  type=Control
  loc=<150.781 50.6879 6.6014>
  rot=<-0.0969 0.7004 -0.0969 -0.7004>
 }
 dyn_hairB_R_02_ctrl_null {parent=dyn_hairB_R_01_ctrl, type=Null, loc=<6.7665 0.9118 0>, rot=<0 0 0.1608 -0.987>}
 dyn_hairB_R_02_ctrl {controlType=EULER_TRANSFORM, parent=dyn_hairB_R_02_ctrl_null, type=Control, loc=<-6.1273 -3.0121 0>, rot=<0 0 -0.1608 -0.987>}
 dyn_hairB_R_03_ctrl_null {parent=dyn_hairB_R_02_ctrl, type=Null, loc=<7.2405 0 0>, rot=<0 0 0.0357 -0.9994>}
 dyn_hairB_R_03_ctrl {controlType=EULER_TRANSFORM, parent=dyn_hairB_R_03_ctrl_null, type=Control, loc=<-7.222 -0.5168 0>, rot=<0 0 -0.0357 -0.9994>}
 dyn_hairB_R_04_ctrl_null {parent=dyn_hairB_R_03_ctrl, type=Null, loc=<7.9383 0 0>, rot=<0 0 -0.1124 -0.9937>}
 dyn_hairB_R_04_ctrl {controlType=EULER_TRANSFORM, parent=dyn_hairB_R_04_ctrl_null, type=Control, loc=<-7.7376 1.7737 0>, rot=<0 0 0.1124 -0.9937>}
 dyn_hairB_R_05_ctrl_null {parent=dyn_hairB_R_04_ctrl, type=Null, loc=<7.4485 0 0>, rot=<0 0 0 1>}
 dyn_hairB_R_05_ctrl {controlType=EULER_TRANSFORM, parent=dyn_hairB_R_05_ctrl_null, type=Control, loc=<-7.4485 0 0>, rot=<0 0 0 1>}
 dyn_hairB_L_01_ctrl_null {type=Null, loc=<7.019 -8.057 158.8791>, rot=<0.0969 -0.7004 0.0969 -0.7004>}
 dyn_hairB_L_01_ctrl {
  controlType=EULER_TRANSFORM
  parent=dyn_hairB_L_01_ctrl_null
  type=Control
  loc=<150.7214 50.8974 -7.019>
  rot=<-0.0969 0.7004 -0.0969 -0.7004>
 }
 dyn_hairB_L_02_ctrl_null {parent=dyn_hairB_L_01_ctrl, type=Null, loc=<6.6782 0.2665 0>, rot=<0 0 0.1785 -0.9839>}
 dyn_hairB_L_02_ctrl {controlType=EULER_TRANSFORM, parent=dyn_hairB_L_02_ctrl_null, type=Control, loc=<-6.1591 -2.5951 0>, rot=<0 0 -0.1785 -0.9839>}
 dyn_hairB_L_03_ctrl_null {parent=dyn_hairB_L_02_ctrl, type=Null, loc=<7.2954 0 0>, rot=<0 0 -0.0403 -0.9992>}
 dyn_hairB_L_03_ctrl {controlType=EULER_TRANSFORM, parent=dyn_hairB_L_03_ctrl_null, type=Control, loc=<-7.2717 0.5868 0>, rot=<0 0 0.0403 -0.9992>}
 dyn_hairB_L_04_ctrl_null {parent=dyn_hairB_L_03_ctrl, type=Null, loc=<7.4829 0 0>, rot=<0 0 -0.0977 -0.9952>}
 dyn_hairB_L_04_ctrl {controlType=EULER_TRANSFORM, parent=dyn_hairB_L_04_ctrl_null, type=Control, loc=<-7.3399 1.4559 0>, rot=<0 0 0.0977 -0.9952>}
 dyn_hairB_L_05_ctrl_null {parent=dyn_hairB_L_04_ctrl, type=Null, loc=<7.4822 0 0>, rot=<0 0 0 1>}
 dyn_hairB_L_05_ctrl {controlType=EULER_TRANSFORM, parent=dyn_hairB_L_05_ctrl_null, type=Control, loc=<-7.4822 0 0>, rot=<0 0 0 1>}
]
variables [
 Aimfloat {type=bool, default=False}
 ArmAimUp {type=double, default=0}
 CloseGroundRemap {type=double, default=0}
 EmotingBool {type=bool, default=False}
 FrontEnd {type=bool, default=False}
 HairBack_CTL {type="TArray<FRigElementKey>"}
 HairBack_CTL_L {type="TArray<FRigElementKey>"}
 HairBack_CTL_R {type="TArray<FRigElementKey>"}
 HairBack_Pose {type=FRigPose, default=(Elements=,HierarchyTopologyVersion=-1,PoseHash=-1)}
 HairBack_Pose_L {type=FRigPose, default=(Elements=,HierarchyTopologyVersion=-1,PoseHash=-1)}
 HairBack_Pose_R {type=FRigPose, default=(Elements=,HierarchyTopologyVersion=-1,PoseHash=-1)}
 HairMovement {type=double, default=0}
 HeadLookDown {type=double, default=0}
 HeadLookLeft {type=double, default=0}
 HeadLookRight {type=double, default=0}
 HeadLookUp {type=double, default=0}
 NotMeele {type=double, default=0}
 Pistol {type=double, default=0}
 RemapHeapUp {type=double, default=0}
 SlyDive {type=bool, default=False}
 UsingCape {type=bool, default=False}
 hairFix {type=double, default=0}
 idlePose {type=double, default=0}
 isCrouching {type=double, default=0}
 isCustomEmotePlaying {type=double, default=0}
 isEmoting {type=double, default=0}
 isHeadUpExtreme {type=double, default=0}
 isTargeting {type=double, default=0}
 movingLeft {type=double, default=0}
]
