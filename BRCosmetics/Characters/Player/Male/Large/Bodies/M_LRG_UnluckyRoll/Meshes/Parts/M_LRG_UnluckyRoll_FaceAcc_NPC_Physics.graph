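# Physics asset for the face accessory mesh of the large male UnluckyRoll NPC body, providing capsule bodies for hair, head, neck, upper torso, arms and shoulder armor.
# Twelve hair pieces each pair a small base anchor locked to the head with a swing- and twist-limited dynamic capsule.
# Center, left and right hairB chains use mostly angularly locked links, with extra locks from left to head and right to spine_05.
# Neck, head and upper-arm joints rotate freely.
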PhysicsAsset /BRCosmetics/Characters/Player/Male/Large/Bodies/M_LRG_UnluckyRoll/Meshes/Parts/M_LRG_UnluckyRoll_FaceAcc_NPC_Physics
bodies {
 armor_shoulder_l [{shapeType=Capsule, length=15.8753, radius=12.0875, shapeName=armor_shoulder_l_capsule, center=<6.1193 0 -5.2048>, rotation=<rot 90 0 0>}]
 armor_shoulder_r [{
   shapeType=Capsule
   length=15.8927
   radius=8.6968
   shapeName=armor_shoulder_l_capsule
   center=<-5.6963 -2.355 -0.4384>
   rotation=<rot 90 -179.7016 2.8229>
  }]
 base_hairPieceA [{shapeType=Capsule, length=0.505, radius=0.505, shapeName=base_hairPieceA_capsule, rotation=<rot 0 -90 0>}]
 base_hairPieceB [{shapeType=Capsule, length=0.505, radius=0.505, shapeName=base_hairPieceB_capsule, rotation=<rot 0 -90 0>}]
 base_hairPieceC [{shapeType=Capsule, length=0.505, radius=0.505, shapeName=base_hairPieceC_capsule, rotation=<rot 0 -90 0>}]
 base_hairPieceD [{shapeType=Capsule, length=0.505, radius=0.505, shapeName=base_hairPieceD_capsule, rotation=<rot 0 -90 0>}]
 base_hairPieceE [{shapeType=Capsule, length=0.505, radius=0.505, shapeName=base_hairPieceE_capsule, rotation=<rot 0 -90 0>}]
 base_hairPieceF [{shapeType=Capsule, length=0.505, radius=0.505, shapeName=base_hairPieceF_capsule, rotation=<rot 0 -90 0>}]
 base_hairPieceG [{shapeType=Capsule, length=0.505, radius=0.505, shapeName=base_hairPieceG_capsule, rotation=<rot 0 -90 0>}]
 base_hairPieceH [{shapeType=Capsule, length=0.505, radius=0.505, shapeName=base_hairPieceH_capsule, rotation=<rot 0 -90 0>}]
 base_hairPieceI [{shapeType=Capsule, length=0.505, radius=0.505, shapeName=base_hairPieceI_capsule, rotation=<rot 0 -90 0>}]
 base_hairPieceJ [{shapeType=Capsule, length=0.505, radius=0.505, shapeName=base_hairPieceJ_capsule, rotation=<rot 0 -90 0>}]
 base_hairPieceK [{shapeType=Capsule, length=0.505, radius=0.505, shapeName=base_hairPieceK_capsule, rotation=<rot 0 -90 0>}]
 base_hairPieceL [{shapeType=Capsule, length=0.505, radius=0.505, shapeName=base_hairPieceL_capsule, rotation=<rot 0 -90 0>}]
 clavicle_l [{shapeType=Capsule, length=5.9428, radius=10.8925, shapeName=clavicle_l_capsule, center=<5.8652 0 0>, rotation=<rot 90 0 0>}]
 clavicle_r [{shapeType=Capsule, length=5.9428, radius=12.7787, shapeName=clavicle_l_capsule, center=<-5.8652 3.1174 1.1024>, rotation=<rot 90 0 180>}]
 dyn_hairB_C_02 [{shapeType=Capsule, length=6.8048, radius=3.853, shapeName=dyn_hairB_C_02_capsule, center=<3.2499 2.0219 0>, rotation=<rot 90 0 0>}]
 dyn_hairB_C_03 [{shapeType=Capsule, length=4.9851, radius=3.3398, shapeName=dyn_hairB_C_03_capsule, center=<5.6312 0.9952 0>, rotation=<rot 90 0 0>}]
 dyn_hairB_C_04 [{shapeType=Capsule, length=4.5529, radius=3.2084, shapeName=dyn_hairB_C_04_capsule, center=<5.6312 0 0>, rotation=<rot 90 0 0>}]
 dyn_hairB_C_05 [{
   shapeType=Capsule
   length=6.1757
   radius=3.2049
   shapeName=dyn_hairB_C_05_capsule
   center=<7.1791 1.244 0.9191>
   rotation=<rot 0.487 105.9296 -92.5671>
  }]
 dyn_hairB_L_02 [{shapeType=Capsule, length=7.1599, radius=3.6503, shapeName=dyn_hairB_L_02_capsule, center=<1.3393 0.9912 0>, rotation=<rot 90 0 0>}]
 dyn_hairB_L_03 [{shapeType=Capsule, length=8.7091, radius=4, shapeName=dyn_hairB_L_03_capsule, center=<5.6312 1.2472 0>, rotation=<rot 90 0 0>}]
 dyn_hairB_L_04 [{shapeType=Capsule, length=4.0102, radius=2.2764, shapeName=dyn_hairB_L_04_capsule, center=<5.6312 0 0>, rotation=<rot 90 0 0>}]
 dyn_hairB_R_02 [{shapeType=Capsule, length=7.1599, radius=3.6503, shapeName=dyn_hairB_R_02_capsule, center=<1.3393 0.9912 0>, rotation=<rot 90 0 0>}]
 dyn_hairB_R_03 [{shapeType=Capsule, length=8.1956, radius=4, shapeName=dyn_hairB_R_03_capsule, center=<5.6312 1.2525 -0.7058>, rotation=<rot 90 0 0>}]
 dyn_hairB_R_04 [{shapeType=Capsule, length=4.3493, radius=2.0318, shapeName=dyn_hairB_R_04_capsule, center=<5.6312 0 0>, rotation=<rot 90 0 0>}]
 dyn_hairPieceA_01 [{shapeType=Capsule, length=10, radius=2.5, shapeName=dyn_hairPieceA_01_capsule, center=<8.2429 0 -1.7483>, rotation=<rot 90 0 0>}]
 dyn_hairPieceB_01 [{shapeType=Capsule, length=10, radius=2.5, shapeName=dyn_hairPieceB_01_capsule, center=<8.2429 0 -1.7483>, rotation=<rot 90 0 0>}]
 dyn_hairPieceC_01 [{shapeType=Capsule, length=5.4286, radius=2.5, shapeName=dyn_hairPieceC_01_capsule, center=<11.8637 0 1.189>, rotation=<rot 90 0 0>}]
 dyn_hairPieceD_01 [{shapeType=Capsule, length=5.6314, radius=1.8563, shapeName=dyn_hairPieceD_01_capsule, center=<7.7257 0 -0.2076>, rotation=<rot 90 0 0>}]
 dyn_hairPieceE_01 [{shapeType=Capsule, length=5, radius=2.0322, shapeName=dyn_hairPieceE_01_capsule, center=<7.5775 0 -0.3132>, rotation=<rot 90 0 0>}]
 dyn_hairPieceF_01 [{shapeType=Capsule, length=10, radius=3, shapeName=dyn_hairPieceF_01_capsule, center=<8.2429 0 1.3909>, rotation=<rot 90 0 0>}]
 dyn_hairPieceG_01 [{shapeType=Capsule, length=10, radius=3, shapeName=dyn_hairPieceG_01_capsule, center=<8.2429 0 1.3909>, rotation=<rot 90 0 0>}]
 dyn_hairPieceH_01 [{shapeType=Capsule, length=10, radius=3, shapeName=dyn_hairPieceH_01_capsule, center=<8.2429 0 1.3909>, rotation=<rot 90 0 0>}]
 dyn_hairPieceI_01 [{shapeType=Capsule, length=8.4546, radius=2.1056, shapeName=dyn_hairPieceI_01_capsule, center=<6.8206 0 1.2549>, rotation=<rot 90 0 0>}]
 dyn_hairPieceJ_01 [{shapeType=Capsule, length=5, radius=2.5, shapeName=dyn_hairPieceJ_01_capsule, center=<4.7398 0 -1.7483>, rotation=<rot 90 0 0>}]
 dyn_hairPieceK_01 [{shapeType=Capsule, length=10, radius=2.5, shapeName=dyn_hairPieceK_01_capsule, center=<8.2429 0 -1.7483>, rotation=<rot 90 0 0>}]
 dyn_hairPieceL_01 [{shapeType=Capsule, length=10, radius=2.5, shapeName=dyn_hairPieceL_01_capsule, center=<8.2429 0 -1.7483>, rotation=<rot 90 0 0>}]
 head [{shapeType=Capsule, length=10, radius=8, shapeName=head_capsule, rotation=<rot 0 90 90>}]
 neck_01 [{shapeType=Capsule, length=3, radius=3, shapeName=neck_01_capsule, rotation=<rot 0 90 90>}]
 neck_02 [{shapeType=Capsule, length=3, radius=3, shapeName=neck_02_capsule, rotation=<rot 0 90 90>}]
 spine_04 [{shapeType=Capsule, length=15.4225, radius=16.0744, shapeName=spine_04_capsule, rotation=<rot 0 -90 0>}]
 spine_05 [{shapeType=Capsule, length=16.3138, radius=15.5845, shapeName=spine_05_capsule, center=<-0.7833 -0.8177 0>, rotation=<rot 0 -90 0>}]
 upperarm_l [{shapeType=Capsule, length=20.2252, radius=8.6369, shapeName=upperarm_r_capsule, center=<6.1205 0 0>, rotation=<rot 90 0 180>}]
 upperarm_r [{shapeType=Capsule, length=20.2252, radius=8.8703, shapeName=upperarm_r_capsule, center=<-6.1205 0 0>, rotation=<rot 90 0 0>}]
}
constraints [
 {
  bone1Name=armor_shoulder_l
  bone2Name=upperarm_l
  swing1LimitDegrees=45
  swing1Motion=Locked
  swing2LimitDegrees=45
  swing2Motion=Locked
  twistLimitDegrees=45
  twistMotion=Locked
 }
 {
  bone1Name=clavicle_l
  bone2Name=spine_05
  swing1LimitDegrees=45
  swing1Motion=Locked
  swing2LimitDegrees=45
  swing2Motion=Locked
  twistLimitDegrees=45
  twistMotion=Locked
 }
 {bone1Name=dyn_hairPieceA_01, bone2Name=base_hairPieceA, swing1LimitDegrees=60, swing2LimitDegrees=60, twistLimitDegrees=20}
 {bone1Name=dyn_hairPieceB_01, bone2Name=base_hairPieceB, swing1LimitDegrees=60, swing2LimitDegrees=60, twistLimitDegrees=20}
 {bone1Name=dyn_hairPieceC_01, bone2Name=base_hairPieceC, swing1LimitDegrees=60, swing2LimitDegrees=60, twistLimitDegrees=20}
 {bone1Name=dyn_hairPieceD_01, bone2Name=base_hairPieceD, swing1LimitDegrees=60, swing2LimitDegrees=60, twistLimitDegrees=20}
 {bone1Name=dyn_hairPieceE_01, bone2Name=base_hairPieceE, swing1LimitDegrees=60, swing2LimitDegrees=60, twistLimitDegrees=20}
 {bone1Name=dyn_hairPieceF_01, bone2Name=base_hairPieceF, swing1LimitDegrees=60, swing2LimitDegrees=40, twistLimitDegrees=20}
 {bone1Name=dyn_hairPieceG_01, bone2Name=base_hairPieceG, swing1LimitDegrees=60, swing2LimitDegrees=40, twistLimitDegrees=20}
 {bone1Name=dyn_hairPieceH_01, bone2Name=base_hairPieceH, swing1LimitDegrees=60, swing2LimitDegrees=40, twistLimitDegrees=20}
 {bone1Name=dyn_hairPieceI_01, bone2Name=base_hairPieceI, swing1LimitDegrees=60, swing2LimitDegrees=60, twistLimitDegrees=20}
 {bone1Name=dyn_hairPieceJ_01, bone2Name=base_hairPieceJ, swing1LimitDegrees=60, swing2LimitDegrees=60, twistLimitDegrees=20}
 {bone1Name=dyn_hairPieceK_01, bone2Name=base_hairPieceK, swing1LimitDegrees=60, swing2LimitDegrees=60, twistLimitDegrees=20}
 {bone1Name=dyn_hairPieceL_01, bone2Name=base_hairPieceL, swing1LimitDegrees=60, swing2LimitDegrees=60, twistLimitDegrees=20}
 {
  bone1Name=spine_05
  bone2Name=spine_04
  swing1LimitDegrees=45
  swing1Motion=Locked
  swing2LimitDegrees=45
  swing2Motion=Locked
  twistLimitDegrees=45
  twistMotion=Locked
 }
 {
  bone1Name=base_hairPieceA
  bone2Name=head
  swing1LimitDegrees=45
  swing1Motion=Locked
  swing2LimitDegrees=45
  swing2Motion=Locked
  twistLimitDegrees=45
  twistMotion=Locked
 }
 {
  bone1Name=base_hairPieceB
  bone2Name=head
  swing1LimitDegrees=45
  swing1Motion=Locked
  swing2LimitDegrees=45
  swing2Motion=Locked
  twistLimitDegrees=45
  twistMotion=Locked
 }
 {
  bone1Name=base_hairPieceC
  bone2Name=head
  swing1LimitDegrees=45
  swing1Motion=Locked
  swing2LimitDegrees=45
  swing2Motion=Locked
  twistLimitDegrees=45
  twistMotion=Locked
 }
 {
  bone1Name=base_hairPieceD
  bone2Name=head
  swing1LimitDegrees=45
  swing1Motion=Locked
  swing2LimitDegrees=45
  swing2Motion=Locked
  twistLimitDegrees=45
  twistMotion=Locked
 }
 {
  bone1Name=base_hairPieceE
  bone2Name=head
  swing1LimitDegrees=45
  swing1Motion=Locked
  swing2LimitDegrees=45
  swing2Motion=Locked
  twistLimitDegrees=45
  twistMotion=Locked
 }
 {
  bone1Name=base_hairPieceF
  bone2Name=head
  swing1LimitDegrees=45
  swing1Motion=Locked
  swing2LimitDegrees=45
  swing2Motion=Locked
  twistLimitDegrees=45
  twistMotion=Locked
 }
 {
  bone1Name=base_hairPieceG
  bone2Name=head
  swing1LimitDegrees=45
  swing1Motion=Locked
  swing2LimitDegrees=45
  swing2Motion=Locked
  twistLimitDegrees=45
  twistMotion=Locked
 }
 {
  bone1Name=base_hairPieceH
  bone2Name=head
  swing1LimitDegrees=45
  swing1Motion=Locked
  swing2LimitDegrees=45
  swing2Motion=Locked
  twistLimitDegrees=45
  twistMotion=Locked
 }
 {
  bone1Name=base_hairPieceI
  bone2Name=head
  swing1LimitDegrees=45
  swing1Motion=Locked
  swing2LimitDegrees=45
  swing2Motion=Locked
  twistLimitDegrees=45
  twistMotion=Locked
 }
 {
  bone1Name=base_hairPieceJ
  bone2Name=head
  swing1LimitDegrees=45
  swing1Motion=Locked
  swing2LimitDegrees=45
  swing2Motion=Locked
  twistLimitDegrees=45
  twistMotion=Locked
 }
 {
  bone1Name=base_hairPieceK
  bone2Name=head
  swing1LimitDegrees=45
  swing1Motion=Locked
  swing2LimitDegrees=45
  swing2Motion=Locked
  twistLimitDegrees=45
  twistMotion=Locked
 }
 {
  bone1Name=base_hairPieceL
  bone2Name=head
  swing1LimitDegrees=45
  swing1Motion=Locked
  swing2LimitDegrees=45
  swing2Motion=Locked
  twistLimitDegrees=45
  twistMotion=Locked
 }
 {
  bone1Name=clavicle_r
  bone2Name=spine_05
  swing1LimitDegrees=45
  swing1Motion=Locked
  swing2LimitDegrees=45
  swing2Motion=Locked
  twistLimitDegrees=45
  twistMotion=Locked
 }
 {
  bone1Name=armor_shoulder_r
  bone2Name=upperarm_r
  swing1LimitDegrees=45
  swing1Motion=Locked
  swing2LimitDegrees=45
  swing2Motion=Locked
  twistLimitDegrees=45
  twistMotion=Locked
 }
 {
  bone1Name=dyn_hairB_C_03
  bone2Name=dyn_hairB_C_02
  swing1LimitDegrees=45
  swing1Motion=Locked
  swing2LimitDegrees=45
  swing2Motion=Locked
  twistLimitDegrees=45
  twistMotion=Locked
 }
 {
  bone1Name=dyn_hairB_C_04
  bone2Name=dyn_hairB_C_03
  swing1LimitDegrees=45
  swing1Motion=Locked
  swing2LimitDegrees=45
  swing2Motion=Locked
  twistLimitDegrees=45
  twistMotion=Locked
 }
 {
  bone1Name=dyn_hairB_L_03
  bone2Name=head
  swing1LimitDegrees=45
  swing1Motion=Locked
  swing2LimitDegrees=45
  swing2Motion=Locked
  twistLimitDegrees=45
  twistMotion=Locked
 }
 {
  bone1Name=dyn_hairB_L_04
  bone2Name=dyn_hairB_L_03
  swing1LimitDegrees=45
  swing1Motion=Locked
  swing2LimitDegrees=45
  swing2Motion=Locked
  twistLimitDegrees=45
  twistMotion=Locked
 }
 {
  bone1Name=dyn_hairB_R_03
  bone2Name=dyn_hairB_R_02
  swing1LimitDegrees=45
  swing1Motion=Locked
  swing2LimitDegrees=45
  swing2Motion=Locked
  twistLimitDegrees=45
  twistMotion=Locked
 }
 {
  bone1Name=dyn_hairB_R_04
  bone2Name=dyn_hairB_R_03
  swing1LimitDegrees=45
  swing1Motion=Locked
  swing2LimitDegrees=45
  swing2Motion=Locked
  twistLimitDegrees=45
  twistMotion=Locked
 }
 {
  bone1Name=upperarm_r
  bone2Name=clavicle_r
  swing1LimitDegrees=45
  swing1Motion=Free
  swing2LimitDegrees=45
  swing2Motion=Free
  twistLimitDegrees=45
  twistMotion=Free
 }
 {
  bone1Name=upperarm_l
  bone2Name=clavicle_l
  swing1LimitDegrees=45
  swing1Motion=Free
  swing2LimitDegrees=45
  swing2Motion=Free
  twistLimitDegrees=45
  twistMotion=Free
 }
 {
  bone1Name=dyn_hairB_C_02
  bone2Name=head
  swing1LimitDegrees=45
  swing1Motion=Free
  swing2LimitDegrees=45
  swing2Motion=Free
  twistLimitDegrees=45
  twistMotion=Free
 }
 {
  bone1Name=dyn_hairB_R_02
  bone2Name=head
  swing1LimitDegrees=45
  swing1Motion=Locked
  swing2LimitDegrees=45
  swing2Motion=Locked
  twistLimitDegrees=45
  twistMotion=Locked
 }
 {
  bone1Name=dyn_hairB_C_05
  bone2Name=dyn_hairB_C_04
  swing1LimitDegrees=45
  swing1Motion=Locked
  swing2LimitDegrees=45
  swing2Motion=Locked
  twistLimitDegrees=45
  twistMotion=Locked
 }
 {
  bone1Name=neck_01
  bone2Name=spine_05
  swing1LimitDegrees=45
  swing1Motion=Free
  swing2LimitDegrees=45
  swing2Motion=Free
  twistLimitDegrees=45
  twistMotion=Free
 }
 {
  bone1Name=neck_02
  bone2Name=neck_01
  swing1LimitDegrees=45
  swing1Motion=Free
  swing2LimitDegrees=45
  swing2Motion=Free
  twistLimitDegrees=45
  twistMotion=Free
 }
 {
  bone1Name=head
  bone2Name=neck_02
  swing1LimitDegrees=45
  swing1Motion=Free
  swing2LimitDegrees=45
  swing2Motion=Free
  twistLimitDegrees=45
  twistMotion=Free
 }
 {
  bone1Name=dyn_hairB_L_03
  bone2Name=dyn_hairB_L_02
  swing1LimitDegrees=45
  swing1Motion=Locked
  swing2LimitDegrees=45
  swing2Motion=Locked
  twistLimitDegrees=45
  twistMotion=Locked
 }
 {
  bone1Name=dyn_hairB_L_02
  bone2Name=head
  swing1LimitDegrees=45
  swing1Motion=Free
  swing2LimitDegrees=45
  swing2Motion=Free
  twistLimitDegrees=45
  twistMotion=Free
 }
 {
  bone1Name=dyn_hairB_R_02
  bone2Name=spine_05
  swing1LimitDegrees=45
  swing1Motion=Locked
  swing2LimitDegrees=45
  swing2Motion=Locked
  twistLimitDegrees=45
  twistMotion=Locked
 }
]
