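# Animation Blueprint for the NPC face-accessory part of M_LRG_UnluckyRoll, animating its hair over the character's pose.
# Copied parent and body poses are layered, then adjusted by a Control Rig fed weapon, movement, targeting, emote and backpack state from the EventGraph.
# Cached AnimDynamics simulates back-hair chains, followed by additive bone offsets, front/side hair dynamics and a head-space RigidBody pass using an NPC physics asset.
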AnimBlueprint /BRCosmetics/Characters/Player/Male/Large/Bodies/M_LRG_UnluckyRoll/Meshes/Parts/M_LRG_UnluckyRoll_FaceAcc_NPC_AnimBP
target_skeleton=&/M_LRG_UnluckyRoll_FaceAcc_Skeleton
graphs [
 AnimGraph {nodes [
   Anim Dynamics - Bone: dyn_hairB_C_05!(ComponentPose, Alpha=0.8, GravityOverride:Vector="0,0,-980", ExternalForce="0,0,0") -> (Pose) {
    alpha=0.8
    angularDampingOverride=0.9999
    angularSpringConstant=5000
    bAngularSpring=true
    bOverrideAngularDamping=true
    bUseGravityOverride=true
    gravityScale=0.35
    simulationSpace=RootRelative
    alphaBoolBlend {blendOption=Linear}
    boundBone {boneName=dyn_hairB_C_05}
    componentAppliedLinearAccClamp=<3500 3500 3500>
    componentLinearAccScale=<0.5 0.5 0.5>
    componentLinearVelScale=<0.5 0.5 0.5>
    gravityOverride {z=-980}
    physicsBodyDefinitions [{
      boundBone {boneName=dyn_hairB_C_05}
      boxExtents=<20 30 30>
      constraintSetup {angularLimitsMax=<0 5 -2>, angularLimitsMin=<0 -5 -30>, angularTarget=<1 0 0>}
      localJointOffset=<4 0 0>
     }]
   }
   Anim Dynamics - Bone: dyn_hairB_R_05!(ComponentPose, Alpha=0.8, GravityOverride:Vector="0,0,-980", ExternalForce="0,0,0") -> (Pose) {
    alpha=0.8
    angularDampingOverride=0.9999
    angularSpringConstant=5000
    bAngularSpring=true
    bOverrideAngularDamping=true
    bUseGravityOverride=true
    gravityScale=0.35
    simulationSpace=RootRelative
    alphaBoolBlend {blendOption=Linear}
    boundBone {boneName=dyn_hairB_R_05}
    componentAppliedLinearAccClamp=<3500 3500 3500>
    componentLinearAccScale=<0.5 0.5 0.5>
    componentLinearVelScale=<0.5 0.5 0.5>
    gravityOverride {z=-980}
    physicsBodyDefinitions [{
      boundBone {boneName=dyn_hairB_R_05}
      boxExtents=<20 30 30>
      constraintSetup {angularLimitsMax=<0 5 -2>, angularLimitsMin=<0 -5 -30>, angularTarget=<1 0 0>}
      localJointOffset=<4 0 0>
     }]
   }
   Anim Dynamics - Bone: dyn_hairB_L_05!(ComponentPose, Alpha=0.8, GravityOverride:Vector="0,0,-980", ExternalForce="0,0,0") -> (Pose) {
    alpha=0.8
    angularDampingOverride=0.9999
    angularSpringConstant=5000
    bAngularSpring=true
    bOverrideAngularDamping=true
    bUseGravityOverride=true
    gravityScale=0.35
    simulationSpace=RootRelative
    alphaBoolBlend {blendOption=Linear}
    boundBone {boneName=dyn_hairB_L_05}
    componentAppliedLinearAccClamp=<3500 3500 3500>
    componentLinearAccScale=<0.5 0.5 0.5>
    componentLinearVelScale=<0.5 0.5 0.5>
    gravityOverride {z=-980}
    physicsBodyDefinitions [{
      boundBone {boneName=dyn_hairB_L_05}
      boxExtents=<20 30 30>
      constraintSetup {angularLimitsMax=<0 5 -2>, angularLimitsMin=<0 -5 -30>, angularTarget=<1 0 0>}
      localJointOffset=<4 0 0>
     }]
   }
   Anim Dynamics - Bone: dyn_hair_front_03_l(Alpha=1, ExternalForce="0,0,0") {
    angularDampingOverride=1
    angularSpringConstant=5000
    bAngularSpring=true
    bOverrideAngularDamping=true
    bUseGravityOverride=true
    gravityScale=0.35
    simulationSpace=RootRelative
    alphaBoolBlend {blendOption=Linear}
    boundBone {boneName=dyn_hair_front_03_l}
    componentLinearAccScale=<1 1 1>
    componentLinearVelScale=<1 1 1>
    gravityOverride {z=-980}
    physicsBodyDefinitions [{
      boundBone {boneName=dyn_hair_front_03_l}
      boxExtents=<30 30 30>
      constraintSetup {angularLimitsMax=<15 15 15>, angularLimitsMin=<-15 -15 -15>, angularTarget=<1 0 0>}
      localJointOffset=<4 0 0>
     }]
   }
   Anim Dynamics - Bone: dyn_hair_front_03_r(Alpha=1, ExternalForce="0,0,0") {
    angularDampingOverride=1
    angularSpringConstant=5000
    bAngularSpring=true
    bOverrideAngularDamping=true
    bUseGravityOverride=true
    gravityScale=0.35
    simulationSpace=RootRelative
    alphaBoolBlend {blendOption=Linear}
    boundBone {boneName=dyn_hair_front_03_r}
    componentLinearAccScale=<1 1 1>
    componentLinearVelScale=<1 1 1>
    gravityOverride {z=-980}
    physicsBodyDefinitions [{
      boundBone {boneName=dyn_hair_front_03_r}
      boxExtents=<30 30 30>
      constraintSetup {angularLimitsMax=<15 15 15>, angularLimitsMin=<-15 -15 -15>, angularTarget=<1 0 0>}
      localJointOffset=<4 0 0>
     }]
   }
   Copy Pose From Mesh(SourceMeshComponent=Get Skel_Mesh()) {bUseAttachedParent=true}
   n1 = Two Way Blend(A=Use cached pose 'CRCache_HairDynPose'(), B=Use cached pose 'CRCache_InputPose'(), Alpha=0.2) {blendNode {alphaBoolBlend {blendOption=Linear}}}
   n2 = Use cached pose 'CRCache_HairDynPose'()
   n3 = "Transform (Modify) Bone - Bone: base_hairB_R"!(
    ComponentPose=Local To Component(LocalPose=Blend Poses by bool(BlendPose_0=@n1, BlendPose_1=@n2) {blendType=Linear})
    bAlphaBoolEnabled:bool=True
    Translation="-3.818,-2.067,-0.000"
    Rotation="-0.184,4.211,-0.389"
    Scale="1,1,1"
   ) -> (Pose) {
    alphaInputType=Bool
    rotationMode=BMM_Additive
    rotationSpace=BCS_BoneSpace
    translationMode=BMM_Additive
    translationSpace=BCS_BoneSpace
    alphaBoolBlend {blendInTime=0.2, blendOption=Linear, blendOutTime=0.2}
    boneToModify {boneName=base_hairB_R}
    rotation=<rot -0.184 4.211 -0.389>
    translation=<-3.8185 -2.0674 0>
   }
   n4 = "Transform (Modify) Bone - Bone: base_hairB_L"!(ComponentPose=@n3, bAlphaBoolEnabled:bool=True, Translation="-6.727,-4.366,-0.000", Rotation="-0.006,7.491,-0.005", Scale="1,1,1") -> (Pose) {
    alphaInputType=Bool
    rotationMode=BMM_Additive
    rotationSpace=BCS_BoneSpace
    translationMode=BMM_Additive
    translationSpace=BCS_BoneSpace
    alphaBoolBlend {blendInTime=0.2, blendOption=Linear, blendOutTime=0.2}
    boneToModify {boneName=base_hairB_L}
    rotation=<rot -0.006 7.491 -0.005>
    translation=<-6.727 -4.366>
   }
   n5 = "Transform (Modify) Bone - Bone: base_hairB_C"!(ComponentPose=@n4, bAlphaBoolEnabled:bool=True, Translation="0,0,0", Rotation="-0.000,-10.763,-0.000", Scale="1,1,1") -> (Pose) {
    alphaInputType=Bool
    rotationMode=BMM_Additive
    rotationSpace=BCS_BoneSpace
    translationMode=BMM_Additive
    translationSpace=BCS_BoneSpace
    alphaBoolBlend {blendInTime=0.2, blendOption=Linear, blendOutTime=0.2}
    boneToModify {boneName=base_hairB_C}
    rotation=<rot 0 -10.763 0>
   }
   n6 = "Transform (Modify) Bone - Bone: dyn_hairB_L_03"!(ComponentPose=@n5, bAlphaBoolEnabled:bool=True, Translation="0,-3,0", Rotation="0,0,0", Scale="1,1,1") -> (Pose) {
    alphaInputType=Bool
    rotationMode=BMM_Additive
    rotationSpace=BCS_BoneSpace
    translationMode=BMM_Additive
    translationSpace=BCS_BoneSpace
    alphaBoolBlend {blendInTime=0.2, blendOption=Linear, blendOutTime=0.2}
    boneToModify {boneName=dyn_hairB_L_03}
    translation {y=-3}
   }
   n7 = "Transform (Modify) Bone - Bone: dyn_hairB_C_04"!(ComponentPose=@n6, bAlphaBoolEnabled:bool=True, Translation="0,0,0", Rotation="0,-10,0", Scale="1,1,1") -> (Pose) {
    alphaInputType=Bool
    rotationMode=BMM_Additive
    rotationSpace=BCS_BoneSpace
    translationMode=BMM_Additive
    translationSpace=BCS_BoneSpace
    alphaBoolBlend {blendInTime=0.2, blendOption=Linear, blendOutTime=0.2}
    boneToModify {boneName=dyn_hairB_C_04}
    rotation {yaw=-10}
   }
   n8 = "Transform (Modify) Bone - Bone: dyn_hairB_R_04"!(ComponentPose=@n7, bAlphaBoolEnabled:bool=True, Translation="0,0,0", Rotation="-0.000,10.405,0.000", Scale="1,1,1") -> (Pose) {
    alphaInputType=Bool
    rotationMode=BMM_Additive
    rotationSpace=BCS_BoneSpace
    translationMode=BMM_Additive
    translationSpace=BCS_BoneSpace
    alphaBoolBlend {blendInTime=0.2, blendOption=Linear, blendOutTime=0.2}
    boneToModify {boneName=dyn_hairB_R_04}
    rotation {yaw=10.405}
   }
   n9 = "Transform (Modify) Bone - Bone: dyn_hairB_C_04"(ComponentPose=@n8, bAlphaBoolEnabled=True, Translation="0,0,0", Rotation="0.000,7.055,-0.000", Scale="1,1,1") {
    alphaInputType=Bool
    rotationMode=BMM_Additive
    rotationSpace=BCS_BoneSpace
    translationMode=BMM_Additive
    translationSpace=BCS_BoneSpace
    alphaBoolBlend {blendInTime=0.2, blendOption=Linear, blendOutTime=0.2}
    boneToModify {boneName=dyn_hairB_C_04}
    rotation=<rot 0 7.0551 0>
   }
   n10 = "Transform (Modify) Bone - Bone: dyn_hairB_L_04"!(ComponentPose=@n9, bAlphaBoolEnabled:bool=True, Translation="0,0,0", Rotation="0.000,13.327,-0.000", Scale="1,1,1") -> (Pose) {
    alphaInputType=Bool
    rotationMode=BMM_Additive
    rotationSpace=BCS_BoneSpace
    translationMode=BMM_Additive
    translationSpace=BCS_BoneSpace
    alphaBoolBlend {blendInTime=0.2, blendOption=Linear, blendOutTime=0.2}
    boneToModify {boneName=dyn_hairB_L_04}
    rotation=<rot 0 13.327 0>
   }
   n11 = Anim Dynamics - Bone: dyn_hair_side_01_l(ComponentPose=@n10, Alpha=1, ExternalForce="0,0,0") {
    angularDampingOverride=1
    angularSpringConstant=5000
    bAngularSpring=true
    bOverrideAngularDamping=true
    bUseGravityOverride=true
    gravityScale=0.35
    simulationSpace=RootRelative
    alphaBoolBlend {blendOption=Linear}
    boundBone {boneName=dyn_hair_side_01_l}
    componentLinearAccScale=<1 1 1>
    componentLinearVelScale=<1 1 1>
    gravityOverride {z=-980}
    physicsBodyDefinitions [{
      boundBone {boneName=dyn_hair_side_01_l}
      boxExtents=<30 30 30>
      constraintSetup {angularLimitsMax=<15 15 15>, angularLimitsMin=<-15 -15 -15>, angularTarget=<1 0 0>}
      localJointOffset=<4 0 0>
     }]
   }
   n12 = Anim Dynamics - Bone: dyn_hair_side_02_l(ComponentPose=@n11, Alpha=1, ExternalForce="0,0,0") {
    angularDampingOverride=1
    angularSpringConstant=5000
    bAngularSpring=true
    bOverrideAngularDamping=true
    bUseGravityOverride=true
    gravityScale=0.35
    simulationSpace=RootRelative
    alphaBoolBlend {blendOption=Linear}
    boundBone {boneName=dyn_hair_side_02_l}
    componentLinearAccScale=<1 1 1>
    componentLinearVelScale=<1 1 1>
    gravityOverride {z=-980}
    physicsBodyDefinitions [{
      boundBone {boneName=dyn_hair_side_02_l}
      boxExtents=<30 30 30>
      constraintSetup {angularLimitsMax=<15 15 15>, angularLimitsMin=<-15 -15 -15>, angularTarget=<1 0 0>}
      localJointOffset=<4 0 0>
     }]
   }
   n13 = Anim Dynamics - Bone: dyn_hair_front_01_l(ComponentPose=@n12, Alpha=1, ExternalForce="0,0,0") {
    angularDampingOverride=1
    angularSpringConstant=5000
    bAngularSpring=true
    bOverrideAngularDamping=true
    bUseGravityOverride=true
    gravityScale=0.35
    simulationSpace=RootRelative
    alphaBoolBlend {blendOption=Linear}
    boundBone {boneName=dyn_hair_front_01_l}
    componentLinearAccScale=<1 1 1>
    componentLinearVelScale=<1 1 1>
    gravityOverride {z=-980}
    physicsBodyDefinitions [{
      boundBone {boneName=dyn_hair_front_01_l}
      boxExtents=<30 30 30>
      constraintSetup {angularLimitsMax=<15 15 15>, angularLimitsMin=<-15 -15 -15>, angularTarget=<1 0 0>}
      localJointOffset=<4 0 0>
     }]
   }
   n14 = Anim Dynamics - Bone: dyn_hair_front_02_l(ComponentPose=@n13, Alpha=1, ExternalForce="0,0,0") {
    angularDampingOverride=1
    angularSpringConstant=5000
    bAngularSpring=true
    bOverrideAngularDamping=true
    bUseGravityOverride=true
    gravityScale=0.35
    simulationSpace=RootRelative
    alphaBoolBlend {blendOption=Linear}
    boundBone {boneName=dyn_hair_front_02_l}
    componentLinearAccScale=<1 1 1>
    componentLinearVelScale=<1 1 1>
    gravityOverride {z=-980}
    physicsBodyDefinitions [{
      boundBone {boneName=dyn_hair_front_02_l}
      boxExtents=<30 30 30>
      constraintSetup {angularLimitsMax=<15 15 15>, angularLimitsMin=<-15 -15 -15>, angularTarget=<1 0 0>}
      localJointOffset=<4 0 0>
     }]
   }
   n15 = Anim Dynamics - Bone: dyn_hair_side_01_r(ComponentPose=@n14, Alpha=1, ExternalForce="0,0,0") {
    angularDampingOverride=1
    angularSpringConstant=5000
    bAngularSpring=true
    bOverrideAngularDamping=true
    bUseGravityOverride=true
    gravityScale=0.35
    simulationSpace=RootRelative
    alphaBoolBlend {blendOption=Linear}
    boundBone {boneName=dyn_hair_side_01_r}
    componentLinearAccScale=<1 1 1>
    componentLinearVelScale=<1 1 1>
    gravityOverride {z=-980}
    physicsBodyDefinitions [{
      boundBone {boneName=dyn_hair_side_01_r}
      boxExtents=<30 30 30>
      constraintSetup {angularLimitsMax=<15 15 15>, angularLimitsMin=<-15 -15 -15>, angularTarget=<1 0 0>}
      localJointOffset=<4 0 0>
     }]
   }
   n16 = Anim Dynamics - Bone: dyn_hair_side_02_r(ComponentPose=@n15, Alpha=1, ExternalForce="0,0,0") {
    angularDampingOverride=1
    angularSpringConstant=5000
    bAngularSpring=true
    bOverrideAngularDamping=true
    bUseGravityOverride=true
    gravityScale=0.35
    simulationSpace=RootRelative
    alphaBoolBlend {blendOption=Linear}
    boundBone {boneName=dyn_hair_side_02_r}
    componentLinearAccScale=<1 1 1>
    componentLinearVelScale=<1 1 1>
    gravityOverride {z=-980}
    physicsBodyDefinitions [{
      boundBone {boneName=dyn_hair_side_02_r}
      boxExtents=<30 30 30>
      constraintSetup {angularLimitsMax=<15 15 15>, angularLimitsMin=<-15 -15 -15>, angularTarget=<1 0 0>}
      localJointOffset=<4 0 0>
     }]
   }
   n17 = Anim Dynamics - Bone: dyn_hair_front_01_r(ComponentPose=@n16, Alpha=1, ExternalForce="0,0,0") {
    angularDampingOverride=1
    angularSpringConstant=5000
    bAngularSpring=true
    bOverrideAngularDamping=true
    bUseGravityOverride=true
    gravityScale=0.35
    simulationSpace=RootRelative
    alphaBoolBlend {blendOption=Linear}
    boundBone {boneName=dyn_hair_front_01_r}
    componentLinearAccScale=<1 1 1>
    componentLinearVelScale=<1 1 1>
    gravityOverride {z=-980}
    physicsBodyDefinitions [{
      boundBone {boneName=dyn_hair_front_01_r}
      boxExtents=<30 30 30>
      constraintSetup {angularLimitsMax=<15 15 15>, angularLimitsMin=<-15 -15 -15>, angularTarget=<1 0 0>}
      localJointOffset=<4 0 0>
     }]
   }
   n18 = Anim Dynamics - Bone: dyn_hair_front_02_r(ComponentPose=@n17, Alpha=1, ExternalForce="0,0,0") {
    angularDampingOverride=1
    angularSpringConstant=5000
    bAngularSpring=true
    bOverrideAngularDamping=true
    bUseGravityOverride=true
    gravityScale=0.35
    simulationSpace=RootRelative
    alphaBoolBlend {blendOption=Linear}
    boundBone {boneName=dyn_hair_front_02_r}
    componentLinearAccScale=<1 1 1>
    componentLinearVelScale=<1 1 1>
    gravityOverride {z=-980}
    physicsBodyDefinitions [{
      boundBone {boneName=dyn_hair_front_02_r}
      boxExtents=<30 30 30>
      constraintSetup {angularLimitsMax=<15 15 15>, angularLimitsMin=<-15 -15 -15>, angularTarget=<1 0 0>}
      localJointOffset=<4 0 0>
     }]
   }
   n19 = "Transform (Modify) Bone - Bone: dyn_hairB_L_01"(ComponentPose=@n18, Alpha=1, Translation="0,0,0.5", Rotation="4.5,0,0", Scale="1,1,1") {
    rotationMode=BMM_Additive
    rotationSpace=BCS_BoneSpace
    translationMode=BMM_Additive
    translationSpace=BCS_BoneSpace
    alphaBoolBlend {blendOption=Linear}
    boneToModify {boneName=dyn_hairB_L_01}
    rotation {pitch=4.5}
    translation {z=0.5}
   }
   n20 = RigidBody!(ComponentPose=@n19, Alpha, OverrideWorldGravity:Vector="0,0,0", ExternalForce) -> (Pose) {
    bDefaultToSkeletalMeshPhysicsAsset=false
    bOverrideWorldGravity=true
    overridePhysicsAsset=&/M_LRG_UnluckyRoll_FaceAcc_NPC_Physics
    simulationSpace=BaseBoneSpace
    alphaBoolBlend {blendOption=Linear}
    baseBoneRef {boneName=head}
    componentAppliedLinearAccClamp=<3000 3000 3000>
    componentLinearAccScale=<0.3 0.3 0.3>
    simSpaceSettings {externalLinearDragV=<0.2 0.2>}
   }
   Output Pose(Result=Component To Local(ComponentPose=@n20)) {layerGroup=None}
   n21 = Layered blend per bone(
    BasePose=Copy Pose From Mesh(SourceMeshComponent=Get Skel_Mesh()) {bUseAttachedParent=true}
    BlendPoses_0=Copy Pose From Mesh(SourceMeshComponent=Get Body_Mesh())
   ) {layerSetup [{branchFilters [{boneName=clavicle_l}, {boneName=clavicle_r}, {boneName=calf_l}, {boneName=calf_r}]}]}
   n22 = Control Rig!(
    Source=@n21
    Alpha
    NotMeele:double=Get MeeleWeapon() -> (MeeleWeapon:float)
    HairMovement:double=Get hairBackmovement() -> (hairBackmovement:double)
    SlyDive:bool=Get bIsSkyDiving()
    Pistol:double=Get pistolAlpha() -> (pistolAlpha:double)
    idlePose:double=Get idlePose() -> (idlePose:double)
    isEmoting:double=Get isEmoting() -> (isEmoting:double)
    isCrouching:double=Get isCrouching() -> (isCrouching:double)
    isTargeting:double=Get IsTargetAlpha() -> (IsTargetAlpha:double)
    UsingCape:bool=Get isWearingPackBool() -> (isWearingPackBool:bool)
    Aimfloat:bool=Get bIsTargeting()
    FrontEnd:bool=Get bIsInFrontEnd()
    EmotingBool:bool=Get EmotingBool() -> (EmotingBool:bool)
    isCustomEmotePlaying:double=Get isCustomEmotePlaying() -> (isCustomEmotePlaying:double)
   ) -> (Pose) {
    alphaBoolBlend {blendOption=Linear}
    alphaScaleBiasClamp {interpSpeedDecreasing=1, interpSpeedIncreasing=1}
    controlRigAssetReference {blueprintRigClass=&/M_LRG_UnluckyRoll_FaceAcc_CtrlRig.M_LRG_UnluckyRoll_FaceAcc_CtrlRig_C}
   }
   Save cached pose 'CRCache_InputPose'(Pose=@n22) {cacheName=CRCache_InputPose}
   n23 = Anim Dynamics - Bone: dyn_hairB_R_02!(
    SimulationSpace:AnimPhysSimSpaceType Enum=Component
    ComponentPose=Local To Component(LocalPose=Use cached pose 'CRCache_InputPose'())
    Alpha=1
    ExternalForce="0,0,0"
   ) -> (Pose) {
    angularDampingOverride=0.9999
    angularSpringConstant=5000
    bAngularSpring=true
    bOverrideAngularDamping=true
    bUseGravityOverride=true
    gravityScale=0.35
    simulationSpace=RootRelative
    alphaBoolBlend {blendOption=Linear}
    boundBone {boneName=dyn_hairB_R_02}
    componentAppliedLinearAccClamp=<3500 3500 3500>
    componentLinearAccScale=<0.5 0.5 0.5>
    componentLinearVelScale=<0.5 0.5 0.5>
    gravityOverride {z=-980}
    physicsBodyDefinitions [{
      boundBone {boneName=dyn_hairB_R_02}
      boxExtents=<20 30 30>
      constraintSetup {angularLimitsMax=<0 5 -5>, angularLimitsMin=<0 -5 -20>, angularTarget=<1 0 0>}
      localJointOffset=<4 0 0>
     }]
   }
   n24 = Anim Dynamics - Bone: dyn_hairB_R_03!(SimulationSpace:AnimPhysSimSpaceType Enum, ComponentPose=@n23, Alpha=0.8, GravityOverride:Vector="0,0,-980", ExternalForce="0,0,0") -> (Pose) {
    alpha=0.8
    angularDampingOverride=0.9999
    angularSpringConstant=5000
    bAngularSpring=true
    bOverrideAngularDamping=true
    bUseGravityOverride=true
    gravityScale=0.35
    simulationSpace=RootRelative
    alphaBoolBlend {blendOption=Linear}
    boundBone {boneName=dyn_hairB_R_03}
    componentAppliedLinearAccClamp=<3500 3500 3500>
    componentLinearAccScale=<0.5 0.5 0.5>
    componentLinearVelScale=<0.5 0.5 0.5>
    gravityOverride {z=-980}
    physicsBodyDefinitions [{
      boundBone {boneName=dyn_hairB_R_03}
      boxExtents=<20 30 30>
      constraintSetup {angularLimitsMax=<0 5 -2>, angularLimitsMin=<0 -5 -20>, angularTarget=<1 0 0>}
      localJointOffset=<4 0 0>
     }]
   }
   n25 = Anim Dynamics - Bone: dyn_hairB_R_04!(SimulationSpace:AnimPhysSimSpaceType Enum, ComponentPose=@n24, Alpha=0.8, GravityOverride:Vector="0,0,-980", ExternalForce="0,0,0") -> (Pose) {
    alpha=0.8
    angularDampingOverride=0.9999
    angularSpringConstant=5000
    bAngularSpring=true
    bOverrideAngularDamping=true
    bUseGravityOverride=true
    gravityScale=0.35
    simulationSpace=RootRelative
    alphaBoolBlend {blendOption=Linear}
    boundBone {boneName=dyn_hairB_R_04}
    componentAppliedLinearAccClamp=<3500 3500 3500>
    componentLinearAccScale=<0.5 0.5 0.5>
    componentLinearVelScale=<0.5 0.5 0.5>
    gravityOverride {z=-980}
    physicsBodyDefinitions [{
      boundBone {boneName=dyn_hairB_R_04}
      boxExtents=<20 30 30>
      constraintSetup {angularLimitsMax=<0 5 -2>, angularLimitsMin=<0 -5 -20>, angularTarget=<1 0 0>}
      localJointOffset=<4 0 0>
     }]
   }
   n26 = Anim Dynamics - Bone: dyn_hairB_L_02!(SimulationSpace:AnimPhysSimSpaceType Enum, ComponentPose=@n25, Alpha=1, ExternalForce="0,0,0") -> (Pose) {
    angularDampingOverride=0.9999
    angularSpringConstant=5000
    bAngularSpring=true
    bOverrideAngularDamping=true
    bUseGravityOverride=true
    gravityScale=0.35
    simulationSpace=RootRelative
    alphaBoolBlend {blendOption=Linear}
    boundBone {boneName=dyn_hairB_L_02}
    componentAppliedLinearAccClamp=<3500 3500 3500>
    componentLinearAccScale=<0.5 0.5 0.5>
    componentLinearVelScale=<0.5 0.5 0.5>
    gravityOverride {z=-980}
    physicsBodyDefinitions [{
      boundBone {boneName=dyn_hairB_L_02}
      boxExtents=<20 30 30>
      constraintSetup {angularLimitsMax=<0 5 -5>, angularLimitsMin=<0 -5 -20>, angularTarget=<1 0 0>}
      localJointOffset=<4 0 0>
     }]
   }
   n27 = Anim Dynamics - Bone: dyn_hairB_L_03!(SimulationSpace:AnimPhysSimSpaceType Enum, ComponentPose=@n26, Alpha=0.8, GravityOverride:Vector="0,0,-980", ExternalForce="0,0,0") -> (Pose) {
    alpha=0.8
    angularDampingOverride=0.9999
    angularSpringConstant=5000
    bAngularSpring=true
    bOverrideAngularDamping=true
    bUseGravityOverride=true
    gravityScale=0.35
    simulationSpace=RootRelative
    alphaBoolBlend {blendOption=Linear}
    boundBone {boneName=dyn_hairB_L_03}
    componentAppliedLinearAccClamp=<3500 3500 3500>
    componentLinearAccScale=<0.5 0.5 0.5>
    componentLinearVelScale=<0.5 0.5 0.5>
    gravityOverride {z=-980}
    physicsBodyDefinitions [{
      boundBone {boneName=dyn_hairB_L_03}
      boxExtents=<20 30 30>
      constraintSetup {angularLimitsMax=<0 5 -2>, angularLimitsMin=<0 -5 -20>, angularTarget=<1 0 0>}
      localJointOffset=<4 0 0>
     }]
   }
   n28 = Anim Dynamics - Bone: dyn_hairB_L_04!(SimulationSpace:AnimPhysSimSpaceType Enum, ComponentPose=@n27, Alpha=0.8, GravityOverride:Vector="0,0,-980", ExternalForce="0,0,0") -> (Pose) {
    alpha=0.8
    angularDampingOverride=0.9999
    angularSpringConstant=5000
    bAngularSpring=true
    bOverrideAngularDamping=true
    bUseGravityOverride=true
    gravityScale=0.35
    simulationSpace=RootRelative
    alphaBoolBlend {blendOption=Linear}
    boundBone {boneName=dyn_hairB_L_04}
    componentAppliedLinearAccClamp=<3500 3500 3500>
    componentLinearAccScale=<0.5 0.5 0.5>
    componentLinearVelScale=<0.5 0.5 0.5>
    gravityOverride {z=-980}
    physicsBodyDefinitions [{
      boundBone {boneName=dyn_hairB_L_04}
      boxExtents=<20 30 30>
      constraintSetup {angularLimitsMax=<0 5 -5>, angularLimitsMin=<0 -5 -20>, angularTarget=<1 0 0>}
      localJointOffset=<4 0 0>
     }]
   }
   n29 = Anim Dynamics - Bone: dyn_hairB_C_02!(SimulationSpace:AnimPhysSimSpaceType Enum, ComponentPose=@n28, Alpha=1, ExternalForce="0,0,0") -> (Pose) {
    angularDampingOverride=0.9999
    angularSpringConstant=5000
    bAngularSpring=true
    bOverrideAngularDamping=true
    bUseGravityOverride=true
    gravityScale=0.35
    simulationSpace=RootRelative
    alphaBoolBlend {blendOption=Linear}
    boundBone {boneName=dyn_hairB_C_02}
    componentAppliedLinearAccClamp=<3500 3500 3500>
    componentLinearAccScale=<0.5 0.5 0.5>
    componentLinearVelScale=<0.5 0.5 0.5>
    gravityOverride {z=-980}
    physicsBodyDefinitions [{
      boundBone {boneName=dyn_hairB_C_02}
      boxExtents=<20 30 30>
      constraintSetup {angularLimitsMax=<0 5 -5>, angularLimitsMin=<0 -5 -20>, angularTarget=<1 0 0>}
      localJointOffset=<4 0 0>
     }]
   }
   n30 = Anim Dynamics - Bone: dyn_hairB_C_03!(SimulationSpace:AnimPhysSimSpaceType Enum, ComponentPose=@n29, Alpha=0.8, GravityOverride:Vector="0,0,-980", ExternalForce="0,0,0") -> (Pose) {
    alpha=0.8
    angularDampingOverride=0.9999
    angularSpringConstant=5000
    bAngularSpring=true
    bOverrideAngularDamping=true
    bUseGravityOverride=true
    gravityScale=0.35
    simulationSpace=RootRelative
    alphaBoolBlend {blendOption=Linear}
    boundBone {boneName=dyn_hairB_C_03}
    componentAppliedLinearAccClamp=<3500 3500 3500>
    componentLinearAccScale=<0.5 0.5 0.5>
    componentLinearVelScale=<0.5 0.5 0.5>
    gravityOverride {z=-980}
    physicsBodyDefinitions [{
      boundBone {boneName=dyn_hairB_C_03}
      boxExtents=<20 30 30>
      constraintSetup {angularLimitsMax=<0 5 -2>, angularLimitsMin=<0 -5 -20>, angularTarget=<1 0 0>}
      localJointOffset=<4 0 0>
     }]
   }
   n31 = Anim Dynamics - Bone: dyn_hairB_C_04!(SimulationSpace:AnimPhysSimSpaceType Enum, ComponentPose=@n30, Alpha=0.8, GravityOverride:Vector="0,0,-980", ExternalForce="0,0,0") -> (Pose) {
    alpha=0.8
    angularDampingOverride=0.9999
    angularSpringConstant=5000
    bAngularSpring=true
    bOverrideAngularDamping=true
    bUseGravityOverride=true
    gravityScale=0.35
    simulationSpace=RootRelative
    alphaBoolBlend {blendOption=Linear}
    boundBone {boneName=dyn_hairB_C_04}
    componentAppliedLinearAccClamp=<3500 3500 3500>
    componentLinearAccScale=<0.5 0.5 0.5>
    componentLinearVelScale=<0.5 0.5 0.5>
    gravityOverride {z=-980}
    physicsBodyDefinitions [{
      boundBone {boneName=dyn_hairB_C_04}
      boxExtents=<20 30 30>
      constraintSetup {angularLimitsMax=<0 5 -2>, angularLimitsMin=<0 -5 -20>, angularTarget=<1 0 0>}
      localJointOffset=<4 0 0>
     }]
   }
   Save cached pose 'CRCache_HairDynPose'(Pose=Component To Local(ComponentPose=@n31)) {cacheName=CRCache_HairDynPose}
  ]}
 EventGraph {nodes [
   n1 = Event BlueprintUpdateAnimation()
   > n2 = Sequence!(execute) -> (then_0, then_1, then_2:Exec, then_3:Exec, then_4:Exec, then_5:Exec, then_6:Exec, then_7:Exec)
   n3 = Branch(execute=@n2.then_4, Condition=Get bIsInFrontEnd())
   > Set Simulation Space(Simulation Space:AnimPhysSimSpaceType Enum=World) -> (Output_Get:AnimPhysSimSpaceType Enum)
   > Set AnimDynAlpha(AnimDynAlpha:double=0.2) -> (Output_Get:double)
   Set Simulation Space(execute=@n3.else)
   n4 = Branch(execute=@n2.then_5, Condition=Get bIsPlayingEmote())
   > Set EmotingBool(EmotingBool:bool=true) -> (Output_Get:bool)
   Set EmotingBool(execute=@n4.else, EmotingBool=false)
   n5 = GetAnimInstance(self=GetOwnerSkeletalMeshComponent())
   Is Valid(exec=@n2.then_7, InputObject=@n5)
   n6 = Montage_IsPlaying(self=@n5, Montage=/BRCosmetics/Animation/Game/MainPlayer/Emotes/CraneAnchor/CLM/Emote_CraneAnchor_CLM_M)
   n7 = Montage_IsPlaying(self=@n5, Montage=/BRCosmetics/Animation/Game/MainPlayer/Emotes/CraneAnchor/CLM/Emote_CraneAnchor_CLM_M_Frontend)
   > Set isCustomEmotePlaying(isCustomEmotePlaying:double=SelectFloat(A=1, B=0, bPickA=OR Boolean(A=@n6, B=@n7))) -> (Output_Get:double)
   n8 = Get isWearingPackBool()
   n9 = Get bIsDBNO()
   Set DBNO_Cape_Float(execute=@n2.then_1, DBNO_Cape_Float:double=SelectFloat(A=1, B=0.2, bPickA=OR Boolean(A=@n8, B=@n9))) -> (Output_Get:double)
   > Set DivePack(DivePack:bool=OR Boolean!(A=Get bIsSkydiveDiveMode(), B=Get bIsSkydiveFloating(), C:bool=Get isWearingPackBool(), D:bool=Get bIsDBNO()) -> (ReturnValue)) -> (Output_Get:bool)
   > Set CrouchPack(CrouchPack:bool=OR Boolean!(A=Get bIsCrouching(), B=Get isWearingPackBool()) -> (ReturnValue)) -> (Output_Get:bool)
   n10 = OR Boolean!(A=AND Boolean(A=Get bIsCrouching(), B=Get bIsMoving2D()), B=Get bIsSurfaceSwimming(), C:bool=Get bIsSkyDiving(), D:bool=Get bIsSprinting()) -> (ReturnValue)
   > Set hairBackmovement(hairBackmovement:double=FInterpTo(Current=Get hairBackmovement(), Target="To Float (Boolean)"(InBool=@n10), DeltaTime=@n1.DeltaTimeX, InterpSpeed=2.5)) -> (Output_Get:double)
   > n11 = Branch(Condition=OR Boolean!(A=Get bIsTargeting(), B=Get bIsCrouching()) -> (ReturnValue))
   > n12 = Set AnimDynAlpha(AnimDynAlpha=0.2)
   n13 = NOT Boolean(A=Get bIsMoving2D())
   Set idlePose(
    execute=[@n17.then, @n12.then]
    idlePose:double=FInterpTo(Current=Get idlePose(), Target="To Float (Boolean)"(InBool=@n13), DeltaTime=@n1.DeltaTimeX, InterpSpeed=2.5)
   ) -> (Output_Get:double)
   > n14 = Branch(Condition=Get bIsCrouching())
   > Set isCrouching(isCrouching:double=1) -> (Output_Get:double)
   n15 = Get bIsTargeting()
   > Set IsTargetAlpha(IsTargetAlpha:double=FInterpTo(Current=Get IsTargetAlpha(), Target="To Float (Boolean)"(InBool=@n15), DeltaTime=@n1.DeltaTimeX, InterpSpeed=10)) -> (Output_Get:double)
   Set isCrouching(execute=@n14.else, isCrouching=0)
   > n16 = Branch(Condition=Get bIsPlayingEmote())
   > Set isEmoting(isEmoting:double=1) -> (Output_Get:double)
   Set isEmoting(execute=@n16.else, isEmoting=0)
   > Set IsTargetAlpha(IsTargetAlpha=0)
   n17 = Set AnimDynAlpha(execute=@n11.else, AnimDynAlpha=0.8)
   n18 = Gravity Override State Function DEPRECATED(
    execute=@n2.then_0
    CharacterPartAnimInstance=Self-Reference()
    GameTimeInSeconds=GetGameTimeInSeconds()
    GravityOverrideParams=Get Gravity Override Params()
   )
   > Set GravityOverride(GravityOverride=@n18)
   n19 = Cast To FortPlayerAnimInstance(execute=@n2.then_2, Object=GetAnimInstance(self=GetOwnerSkeletalMeshComponent()))
   > Set MeeleWeapon(MeeleWeapon:float=1) -> (Output_Get:float)
   > Set pistolAlpha(pistolAlpha:double=1) -> (Output_Get:double)
   > n20 = Switch on EFortWeaponCoreAnimation!(execute, Selection:EFortWeaponCoreAnimation Enum=Get WeaponCoreAnim(self=@n19)) -> (
    Melee:Exec
    Pistol:Exec
    Shotgun:Exec
    PaperBlueprint:Exec
    Rifle:Exec
    MeleeOneHand:Exec
    MachinePistol:Exec
    RocketLauncher:Exec
    GrenadeLauncher:Exec
    GoingCommando:Exec
    AssaultRifle:Exec
    TacticalShotgun:Exec
    SniperRifle:Exec
    TrapPlacement:Exec
    ShoulderLauncher:Exec
    AbilityDecoTool:Exec
    Crossbow:Exec
    C4:Exec
    RemoteControl:Exec
    DualWield:Exec
    AR_BullPup:Exec
    AR_ForwardGrip:Exec
    MedPackPaddles:Exec
    SMG_P90:Exec
    AR_DrumGun:Exec
    Consumable_Small:Exec
    Consumable_Large:Exec
    Balloon:Exec
    MountedTurret:Exec
    CreativeTool:Exec
    ExplosiveBow:Exec
    AshtonIndigo:Exec
    AshtonChicago:Exec
    MeleeDualWield:Exec
    Unarmed:Exec
   )
   Branch(execute=[@n20.DualWield, @n20.Pistol])
   > Set pistolAlpha(pistolAlpha=0.3)
   Branch(execute=[@n20.Melee, @n20.MeleeOneHand])
   > Set MeeleWeapon(MeeleWeapon=0)
   n21 = Get bIsTargeting()
   Set AimFloat(
    execute=@n2.then_3
    AimFloat:double=FInterpTo(Current=Get AimFloat() -> (AimFloat:double), Target="To Float (Boolean)"(InBool=@n21), DeltaTime=@n1.DeltaTimeX, InterpSpeed=10)
   ) -> (Output_Get:double)
   Event OnCharacterCustomizationLoaded()
   > n22 = Cast To PlayerPawn_Athena(Object=TryGetPawnOwner())
   > Set Body_Mesh(Body_Mesh=GetSkeletalMeshForPartType(self=@n22))
   n23 = CheckForBackbling()
   > Set isWearingPackBool(isWearingPackBool:bool=OR Boolean!(A=@n23, B=@n23.bIsBackblingACape, C:bool=@n23.bIsBackblingLong) -> (ReturnValue)) -> (Output_Get:bool)
   n24 = OR Boolean!(A=@n23.bIsBackblingACape, B=@n23.bIsBackblingLong) -> (ReturnValue)
   > Set is_f_WearingCape(is_f_WearingCape:double="To Float (Boolean)"(InBool=@n24)) -> (Output_Get:double)
   > Set isWearingBP(isWearingBP:bool=AND Boolean(A=@n23, B=NOT Boolean(A=@n24))) -> (Output_Get:bool)
   Event BlueprintInitializeAnimation()
   > Set Skel_Mesh(Skel_Mesh=GetOwnerSkeletalMeshComponent())
   n25 = GetAnimInstance(self=GetOwnerSkeletalMeshComponent())
   Montage_IsPlaying(self=@n25, Montage=/BRCosmetics/Animation/Game/MainPlayer/Emotes/CraneAnchor/CMM/Emote_CraneAnchor_CMM_M)
   Reroute Node(InputPin:Wildcard) -> (OutputPin:Wildcard)
   Is Valid(InputObject=@n25)
   Set isCustomEmotePlaying(isCustomEmotePlaying=0)
  ]}
]
variables [
 AimFloat {type=number}
 AnimDynAlpha {type=number}
 Body_Mesh {type=SkeletalMeshComponent}
 CrouchPack {type=boolean}
 DBNO_Cape_Float {type=number}
 DivePack {type=boolean}
 EmotingBool {type=boolean}
 Gravity Override Params {type=GravityOverrideParamsStruct}
 GravityOverride {type=Vector}
 hairBackmovement {type=number}
 idlePose {type=number}
 Is Target Alpha {type=number}
 is_f_WearingCape {type=number}
 isCrouching {type=number}
 isCustomEmotePlaying {type=number}
 isEmoting {type=number}
 IsTargetAlpha {type=number}
 isWearingBP {type=boolean}
 isWearingPackBool {type=boolean}
 MeeleWeapon {type=number}
 pistolAlpha {type=number}
 Simulation Space {type=AnimPhysSimSpaceType, default=RootRelative}
 Skel_Mesh {type=SkeletalMeshComponent}
]
